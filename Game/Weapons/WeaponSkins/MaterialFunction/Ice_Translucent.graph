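# Material function for weapon skins that layers translucent ice over input material attributes through an Alpha input.
# Fresnel drives a gradient texture for color and emissive glow plus opacity, while EdgeMap tints edges and a tiled stone normal blends with the input normal.
# A static switch adds procedural snow, masked by pre-skinned normal Z and noise, so it stays fixed to the mesh.
# Several nodes, including depth fade, are unconnected.
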MaterialFunction /Game/Weapons/WeaponSkins/MaterialFunction/Ice_Translucent
graph {nodes [
  n1 = TextureCoordinate()
  n2 = ScalarParameter() {defaultValue=5, group=Ice, parameterName=Ice Normal Tiling}
  n3 = ScalarParameter() {group=Ice, parameterName=IceNormalIntensity}
  n4 = MaterialFunctionCall(
   Normal=TextureSample(UVs=Multiply(A=@n1, B=@n2)) {samplerType=SAMPLERTYPE_Normal, texture=/Game/Weapons/WeaponSkins/Textures/StoneNormal}.RGB
   Flatness=Saturate(None=OneMinus(None=@n3))
  ) {materialFunction=&FlattenNormal}
  n5 = ComponentMask(None=VertexInterpolator(VS=PreSkinnedNormal()).PS) {b=true}
  Add(
   A=Multiply(A=ComponentMask(None=@n4.Result) {g=true}, B=ScalarParameter() {defaultValue=0.3, group=Ice, parameterName=RockNormalSnowIntensity})
   B=@n5
  )
  n6 = FunctionInput() {inputName=Material, inputType=FunctionInput_MaterialAttributes}
  n7 = GetMaterialAttributes(None=@n6) {attributeGetTypes [#1, #2, #3]}
  n8 = Fresnel(ExponentIn=ScalarParameter() {defaultValue=0.8, group=Ice, parameterName=Crystal_FresEX}, Normal=Transform(None=@n7.Normal)) {baseReflectFraction=0}
  n9 = Power(Base=@n8, Exp=ScalarParameter() {defaultValue=1, group=Ice, parameterName=Ice Fresnel})
  n10 = LinearInterpolate(A=ScalarParameter() {defaultValue=0.4, group=Ice, parameterName=Ice Crystal Opacity}, Alpha=@n9)
  n11 = Multiply(
   A=TextureSampleParameter2D(UVs=Saturate(None=@n8)) {group=Ice, parameterName=IceGradient, texture=/Game/Global/Textures/Default/Masks_and_Gradients/T_IceGradient}.RGB
   B=ScalarParameter() {defaultValue=1, group=Ice, parameterName=ice Brightness}
  )
  n12 = ScalarParameter() {group=Ice, parameterName=Ice Emissive Brightness}
  n13 = MakeMaterialAttributes!(
   BaseColor=LinearInterpolate(
    A=@n11
    B=VectorParameter() {group=Ice, parameterName=Ice Edge Tint, defaultValue=<rgba 0.5876 0.6719 0.6652 1>}.RGB
    Alpha=FunctionInput() {inputName=EdgeMap, inputType=FunctionInput_Scalar}
   )
   Metallic=ScalarParameter() {defaultValue=0.4, group=Ice, parameterName=IceMetallic}
   Specular=@n7.Specular
   Roughness=ScalarParameter() {defaultValue=0.05, group=Ice, parameterName=IceRoughness}
   Anisotropy
   EmissiveColor=Multiply(A=@n11, B=Multiply(A=@n12, B=@n9))
   Opacity=@n10
   OpacityMask=@n10
   Normal=MaterialFunctionCall(BaseNormal=@n7.Normal, AdditionalNormal=@n4.Result) {materialFunction=&BlendAngleCorrectedNormals}.Result
   Tangent
   WorldPositionOffset
   SubsurfaceColor
   ClearCoat
   ClearCoatRoughness
   AmbientOcclusion
   Refraction
   CustomizedUVs_0
   CustomizedUVs_1
   CustomizedUVs_2
   CustomizedUVs_3
   CustomizedUVs_4
   CustomizedUVs_5
   CustomizedUVs_6
   CustomizedUVs_7
   PixelDepthOffset
   ShadingModel
   Displacement
  )
  n14 = Constant()
  n15 = Constant() {r=1}
  n16 = Multiply(A=TextureCoordinate(), B=ScalarParameter() {defaultValue=12, group=Ice, parameterName=SnowTiling})
  n17 = TextureSample(UVs=@n16) {samplerType=SAMPLERTYPE_LinearColor, texture=/Game/Global/Textures/Default/Noise/T_FN_FluffyCloth}
  n18 = Power(
   Base=Multiply(A=@n5, B=ScalarParameter() {defaultValue=1, group=Ice, parameterName=SnowAmount})
   Exp=ScalarParameter() {defaultValue=2, group=Ice, parameterName=SnowPower}
  )
  n19 = BlendMaterialAttributes(
   A=@n13
   B=MakeMaterialAttributes(
    BaseColor=VectorParameter() {parameterName=SnowTint, defaultValue=<rgba 0.97 0.97 0.97 1>}.RGB
    Metallic=Constant()
    Specular=Constant() {r=0.35}
    Roughness=Constant() {r=0.7}
    EmissiveColor=@n14
    Opacity=@n15
    OpacityMask=@n15
    Normal=TextureSample(UVs=@n16) {samplerType=SAMPLERTYPE_Normal, texture=/Game/Weapons/WeaponSkins/Textures/Snow_N}.RGB
    SubsurfaceColor=@n14
   )
   Alpha=Add(A=Subtract(A=@n17.R), B=Saturate(None=@n18))
  ) {vertexAttributeBlendType=UseA}
  FunctionOutput(None=BlendMaterialAttributes(
    A=@n6
    B=StaticSwitchParameter(True=@n19, False=@n13) {group=Ice, parameterName=Use Procedural Snow}
    Alpha=FunctionInput() {bUsePreviewValueAsDefault=true, inputName=Alpha, inputType=FunctionInput_Scalar, sortPriority=1}
   ))
  n20 = DepthFade(FadeDistance=ScalarParameter() {defaultValue=9.5587, parameterName=FadeDistance})
  LinearInterpolate(
   A=QualitySwitch(Default=@n20, Low=Constant() {r=0.4}, High=@n20, Medium=@n20)
   B=ScalarParameter() {defaultValue=0.5, parameterName=Bubble Opacity}
  )
  MaterialFunctionCall(
   In=Multiply(A=@n8, B=ScalarParameter() {defaultValue=3.0567, parameterName=Multasdfa})
   Contrast=ScalarParameter() {defaultValue=1, parameterName=asdfasdf}
  ) {materialFunction=&CheapContrast}
  Multiply(A=@n16) {constB=0.5}
  ScalarParameter() {parameterName=Ice isTranslucent}
  StaticBoolParameter() {parameterName=IsTranslucent}
 ]}
