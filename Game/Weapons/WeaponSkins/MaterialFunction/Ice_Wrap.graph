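# Material function for weapon skins that wraps incoming material attributes in ice, with optional procedural snow, blended by an Alpha input.
# Ice base color comes from a Fresnel-driven gradient tinted and lerped to an edge tint by EdgeMap, with an added detail normal.
# View-dependent emissive is modulated by panning caustic noise projected from pre-skinned local position.
# A static switch adds snow, height-lerped by pre-skinned normal direction and a tiled noise mask.
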MaterialFunction /Game/Weapons/WeaponSkins/MaterialFunction/Ice_Wrap
graph {nodes [
  n1 = FunctionInput() {inputName=Material, inputType=FunctionInput_MaterialAttributes}
  n2 = GetMaterialAttributes(None=@n1) {attributeGetTypes [#1, #2, #3]}
  n3 = Transform(None=@n2.Normal)
  n4 = CameraVectorWS()
  n5 = Power(Base=Clamp(None=DotProduct(A=@n3, B=@n4)), Exp=ScalarParameter() {defaultValue=4, parameterName=IceEmissiveExponent})
  n6 = VectorParameter() {parameterName=IceEmissiveColor, defaultValue=<rgba 0 0.0278 0.0677 1>}
  n7 = PreSkinnedPosition()
  n8 = ComponentMask(None=FunctionInput(Preview=VertexInterpolator(VS=@n7).PS) {inputName=PreSkinLocalPosition}) {b=true, g=true}
  n9 = Multiply(
   A=MaterialFunctionCall(Float2=@n8) {materialFunction=&BreakOutFloat2Components}.G
   B=Constant2Vector() {r=1}
  )
  n10 = ScalarParameter() {defaultValue=0.5, parameterName=CausticSkew}
  n11 = Multiply(A=Add(A=@n8, B=Multiply(A=@n9, B=@n10)), B=ScalarParameter() {defaultValue=0.01, parameterName=CausticScale})
  n12 = Constant2Vector() {g=0.01, r=1}
  n13 = ScalarParameter() {defaultValue=1, parameterName=CausticSpeed}
  n14 = Constant2Vector() {g=1}
  n15 = TextureSample(UVs=Panner(Coordinate=Multiply(A=@n11, B=@n12), Speed=Multiply(A=@n13, B=@n14)) {bFractionalPart=true, speedY=0.1}) {samplerSource=SSM_Wrap_WorldGroupSettings, samplerType=SAMPLERTYPE_LinearColor, texture=/Game/Global/Textures/Default/Noise/testDiffClouds}
  n16 = ScalarParameter() {defaultValue=1, group=Ice, parameterName=Ice Crystal Opacity}
  n17 = Fresnel(ExponentIn=ScalarParameter() {defaultValue=0.8, group=Ice, parameterName=Crystal_FresEX}, Normal=@n3) {baseReflectFraction=0}
  n18 = Power(Base=@n17, Exp=ScalarParameter() {defaultValue=1, group=Ice, parameterName=Ice Fresnel})
  n19 = TextureSampleParameter2D(UVs=Saturate(None=@n17)) {
   group=Ice
   parameterName=IceGradient
   samplerSource=SSM_Clamp_WorldGroupSettings
   texture=/Game/Global/Textures/Default/Masks_and_Gradients/T_IceGradient
  }
  n20 = Multiply(A=@n19.RGB, B=ScalarParameter() {defaultValue=1, group=Ice, parameterName=ice Brightness})
  n21 = VectorParameter() {parameterName=IceTint, defaultValue=<rgba 1 1 1 1>}
  n22 = TextureSampleParameter2D(UVs=Multiply(A=TextureCoordinate(), B=ScalarParameter() {defaultValue=5, group=Ice, parameterName=Ice Normal Tiling})) {parameterName=Ice Normal, samplerType=SAMPLERTYPE_Normal, texture=/Game/Weapons/WeaponSkins/Textures/StoneNormal}
  n23 = Multiply(A=Saturate(None=ScalarParameter() {group=Ice, parameterName=IceNormalIntensity}), B=Constant3Vector() {constant=<rgba 1 1 0 0>})
  n24 = MakeMaterialAttributes!(
   BaseColor=LinearInterpolate(
    A=Multiply(A=@n20, B=@n21.RGB)
    B=VectorParameter() {group=Ice, parameterName=Ice Edge Tint, defaultValue=<rgba 0.5876 0.6719 0.6652 1>}.RGB
    Alpha=FunctionInput() {inputName=EdgeMap, inputType=FunctionInput_Scalar}
   )
   Metallic=ScalarParameter() {defaultValue=0.4, group=Ice, parameterName=IceMetallic}
   Specular=ScalarParameter() {defaultValue=1, parameterName=Specular}
   Roughness=ScalarParameter() {defaultValue=0.05, group=Ice, parameterName=IceRoughness}
   Anisotropy
   EmissiveColor=Multiply(A=Multiply(A=@n5, B=@n6.RGB), B=Multiply(A=@n15.R))
   Opacity=StaticSwitch(True=LinearInterpolate(A=@n16, Alpha=@n18), False=@n16)
   OpacityMask
   Normal=Add(A=@n2.Normal, B=Multiply(A=@n22.RGB, B=@n23))
   Tangent
   WorldPositionOffset
   SubsurfaceColor
   ClearCoat
   ClearCoatRoughness
   AmbientOcclusion
   Refraction
   CustomizedUVs_0
   CustomizedUVs_1
   CustomizedUVs_2
   CustomizedUVs_3
   CustomizedUVs_4
   CustomizedUVs_5
   CustomizedUVs_6
   CustomizedUVs_7
   PixelDepthOffset
   ShadingModel
   Displacement
  )
  n25 = Constant()
  n26 = Multiply(A=TextureCoordinate(), B=ScalarParameter() {defaultValue=1, group=Ice, parameterName=SnowMaskTiling})
  n27 = TextureSample(UVs=@n26) {samplerType=SAMPLERTYPE_Normal, texture=/Game/Global/Textures/Default/Noise/T_FN_FluffyCloth_N}
  n28 = ScalarParameter() {defaultValue=-1, parameterName=SnowFlatness}
  n29 = VertexInterpolator(VS=PreSkinnedNormal())
  n30 = Clamp(None=ComponentMask(None=FunctionInput(Preview=@n29.PS) {bUsePreviewValueAsDefault=true, inputName=PreSkinnedLocalNormal}) {b=true})
  n31 = ScalarParameter() {defaultValue=1, group=Ice, parameterName=SnowAmount}
  n32 = Saturate(None=Power(Base=Multiply(A=@n30, B=@n31), Exp=ScalarParameter() {defaultValue=2, group=Ice, parameterName=SnowPower}))
  n33 = LinearInterpolate(
   A=ScalarParameter() {parameterName=SnowMaskMin}
   B=ScalarParameter() {parameterName=SnowMaskMax}
   Alpha=TextureSample(UVs=@n26) {samplerType=SAMPLERTYPE_LinearColor, texture=/Game/Global/Textures/Default/Noise/T_FN_FluffyCloth}.R
  )
  n34 = BlendMaterialAttributes(
   A=@n24
   B=MakeMaterialAttributes(
    BaseColor=VectorParameter() {parameterName=SnowTint, defaultValue=<rgba 0.97 0.97 0.97 1>}.RGB
    Metallic=Constant()
    Specular=ScalarParameter() {defaultValue=0.2, parameterName=SnowSpecular}
    Roughness=ScalarParameter() {defaultValue=2, parameterName=SnowRoughness}
    EmissiveColor=@n25
    Opacity=@n25
    Normal=MaterialFunctionCall(Normal=@n27.RGB, Flatness=@n28) {materialFunction=&FlattenNormal}.Result
    SubsurfaceColor=@n25
   )
   Alpha=MaterialFunctionCall(Transition Phase=@n32, Height Texture=@n33, Contrast=ScalarParameter() {parameterName=SnowMaskContrast}) {materialFunction=&HeightLerp}.Alpha
  ) {vertexAttributeBlendType=UseA}
  FunctionOutput(None=BlendMaterialAttributes(
    A=@n1
    B=StaticSwitchParameter(True=@n34, False=@n24) {group=Ice, parameterName=Use Procedural Snow}
    Alpha=FunctionInput() {bUsePreviewValueAsDefault=true, inputName=Alpha, inputType=FunctionInput_Scalar, sortPriority=1}
   ))
  Reroute(None=@n20)
  Saturate(None=Add(A=@n33, B=@n32))
  TextureSample(UVs=Multiply(A=TextureCoordinate(), B=ScalarParameter() {defaultValue=12, group=Ice, parameterName=SnowTiling})) {samplerSource=SSM_Wrap_WorldGroupSettings, samplerType=SAMPLERTYPE_Normal, texture=/Game/Weapons/WeaponSkins/Textures/Snow_N}
 ]}
