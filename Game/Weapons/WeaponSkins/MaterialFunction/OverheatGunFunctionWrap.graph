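# Weapon-skin material function that wraps base material attributes with a glowing barrel-heat effect around the muzzle.
# A muzzle sphere mask drives a HeightLerp, using base color as height to break up the glow.
# Its alpha shapes a falloff-to-core color, boosted by power, intensity and temperature into saturated emissive, blended in by the mask and UseBarrelHeat.
# The "Use Overheat" static switch passes base attributes through when true, applying heat only when false.
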MaterialFunction /Game/Weapons/WeaponSkins/MaterialFunction/OverheatGunFunctionWrap
graph {nodes [
  n1 = FunctionInput() {bUsePreviewValueAsDefault=true, inputName=Base MA, inputType=FunctionInput_MaterialAttributes}
  n2 = GetMaterialAttributes(None=@n1) {attributeGetTypes [#1, #2]}
  n3 = MaterialFunctionCall!(
   MuzzleFalloff=Multiply(A=ScalarParameter() {defaultValue=50, group=Weapon_Muzzle, parameterName=Heat MuzzleFalloff}) {constB=1.4}
   Pre-Skinned Local Position=FunctionInput() {bUsePreviewValueAsDefault=true, inputName=Pre-Skinned Local Position, sortPriority=41}
  ) {materialFunction=/Game/Packages/Fortress_SharedMaterials/Base_Material_Functions/BaseCustomizationFunctions/MF_BaseWeapon_MuzzleSphereMask}
  n4 = OneMinus(None=@n2.BaseColor) {desc=diffuse}
  n5 = Constant() {r=0.555}
  n6 = MaterialFunctionCall(
   Transition Phase=Multiply(A=@n3.Muzzle_SphereMask, B=ScalarParameter() {defaultValue=0.6255, group=Weapon_Heat, parameterName=Heat TransitionMultiplier}) {desc=sphere mask}
   Height Texture=OneMinus(None=Multiply(A=@n4, B=@n5) {constB=0.2881})
  ) {materialFunction=&HeightLerp}
  n7 = Multiply(
   A=@n6.Alpha
   B=LinearInterpolate(
    A=VectorParameter() {group=Weapon_Heat, parameterName=Heat Falloff Color, defaultValue=<rgba 0.645 0.4322 0.4322 1>}.RGB
    B=VectorParameter() {group=Weapon_Heat, parameterName=Heat Core Color, defaultValue=<rgba 0.8824 0.3741 0.2296 1>}.RGB
    Alpha=@n6.Alpha
   )
  )
  n8 = Multiply(A=@n7, B=@n7)
  n9 = ScalarParameter() {defaultValue=2.9669, group=Weapon_Heat, parameterName=Heat Intensity}
  n10 = Multiply(
   A=Power(Base=Multiply(A=@n8, B=@n9) {constB=2}, Exp=ScalarParameter() {defaultValue=1.6394, group=Weapon_Heat, parameterName=Heat Final Power})
   B=ScalarParameter() {defaultValue=5, group=Weapon_Heat, parameterName=Heat Temp}
  )
  n11 = SetMaterialAttributes!(
   MaterialAttributes=@n2.MaterialAttributes
   BaseColor=@n2.BaseColor
   Specular=@n2.Specular
   EmissiveColor=Saturate(None=Multiply(A=@n6.Alpha, B=@n10))
   Opacity=Constant() {r=1}
  ) {attributeSetTypes [
    #1
    #2
    #3
    #4
   ]}
  n12 = BlendMaterialAttributes(A=@n1, B=@n11, Alpha=@n3.Muzzle_SphereMask) {vertexAttributeBlendType=UseA}
  n13 = Saturate(None=ScalarParameter() {group=Weapon_Heat, parameterName=UseBarrelHeat, sortPriority=0})
  FunctionOutput(None=StaticSwitch(
    True=@n1
    False=BlendMaterialAttributes(A=@n1, B=@n12, Alpha=@n13) {vertexAttributeBlendType=UseA}
    Value=StaticBoolParameter() {group=Weapon_Elemental, parameterName=Use Overheat}
   ))
  LinearInterpolate()
 ]}
