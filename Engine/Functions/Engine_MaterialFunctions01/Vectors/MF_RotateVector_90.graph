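# Material function, exposed to the library, that rotates an input vector 90 degrees around a cardinal axis in right-hand coordinates.
# It masks the input into X, Y and Z components and negates each with a multiply.
# Six outputs give clockwise and counterclockwise results for each axis by re-appending swapped and negated components, so the rotation needs no trigonometry or matrix.
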MaterialFunction /Engine/Functions/Engine_MaterialFunctions01/Vectors/MF_RotateVector_90
graph {
 properties {bExposeToLibrary=true, description="Rotates vectors along any cardinal axis (Right hand coordinates)"}
 nodes [
  n1 = FunctionInput() {inputName=Vector}
  n2 = ComponentMask(None=@n1) {r=true}
  n3 = ComponentMask(None=@n1) {b=true}
  n4 = Multiply(A=@n3) {constB=-1}
  n5 = ComponentMask(None=@n1) {g=true}
  FunctionOutput(None=AppendVector(A=AppendVector(A=@n2, B=@n4), B=@n5)) {outputName=CW X}
  n6 = Multiply(A=@n2) {constB=-1}
  FunctionOutput(None=AppendVector(A=AppendVector(A=@n5, B=@n6), B=@n3)) {outputName=CW Z, sortPriority=4}
  n7 = Multiply(A=@n5) {constB=-1}
  FunctionOutput(None=AppendVector(A=AppendVector(A=@n7, B=@n2), B=@n3)) {outputName=CCW Z, sortPriority=5}
  FunctionOutput(None=AppendVector(A=AppendVector(A=@n2, B=@n3), B=@n7)) {outputName=CCW X, sortPriority=1}
  FunctionOutput(None=AppendVector(A=AppendVector(A=@n4, B=@n5), B=@n2)) {outputName=CW Y, sortPriority=2}
  FunctionOutput(None=AppendVector(A=AppendVector(A=@n3, B=@n5), B=@n6)) {outputName=CCW Y, sortPriority=3}
 ]
}
