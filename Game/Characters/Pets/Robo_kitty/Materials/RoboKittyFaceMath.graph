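# Material function for the Robo Kitty pet's face that converts facial pose scalar parameters, grouped under Flipbook, into a single expression index.
# A chain of If nodes tests each pose against a fixed threshold and outputs its constant index, or zero when none exceeds it.
# The nesting sets priority: right blink wins, then left brow up, glabella down, glabella up and right brow up, so poses never blend.
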MaterialFunction /Game/Characters/Pets/Robo_kitty/Materials/RoboKittyFaceMath
graph {nodes [
  Constant()
  n1 = Constant()
  n2 = If(
   A=ScalarParameter() {group=Flipbook, parameterName=R_brow_up_pose}
   B=Constant() {r=0.2}
   "A > B"=Constant() {r=1}
   "A == B"=@n1
   "A < B"=@n1
  )
  n3 = If(
   A=ScalarParameter() {group=Flipbook, parameterName=C_glabella_up_pose}
   B=Constant() {r=0.2}
   "A > B"=Constant() {r=2}
   "A == B"=@n2
   "A < B"=@n2
  )
  n4 = If(
   A=ScalarParameter() {group=Flipbook, parameterName=C_glabella_down_pose}
   B=Constant() {r=0.2}
   "A > B"=Constant() {r=4}
   "A == B"=@n3
   "A < B"=@n3
  )
  n5 = If(
   A=ScalarParameter() {group=Flipbook, parameterName=L_brow_up_pose}
   B=Constant() {r=0.2}
   "A > B"=Constant() {r=5}
   "A == B"=@n4
   "A < B"=@n4
  )
  FunctionOutput(None=If(A=ScalarParameter() {group=Flipbook, parameterName=R_blink_pose}, B=Constant() {r=0.2}, "A > B"=Constant() {r=3}, "A == B"=@n5, "A < B"=@n5))
 ]}
