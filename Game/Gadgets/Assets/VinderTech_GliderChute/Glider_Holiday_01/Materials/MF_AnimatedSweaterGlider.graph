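# Material function for the holiday sweater glider that adds an animated, glowing knit design to incoming material attributes.
# A sweeping smoothstep band gates blurred noise times a panning FX texture, plus a camera-vector crystal reflection, feeding base color and emissive.
# A cropped mask's green channel confines this, design tint, red-orange Fresnel and knit specular/metallic/roughness.
# The sine mask blend has identical inputs, so does nothing; unmasked metallic takes incoming specular.
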MaterialFunction /Game/Gadgets/Assets/VinderTech_GliderChute/Glider_Holiday_01/Materials/MF_AnimatedSweaterGlider
graph {nodes [
  n1 = Rotator(Coordinate=TextureCoordinate() {uTiling=0.7, vTiling=0.6})
  n2 = TextureSample(UVs=@n1) {samplerType=SAMPLERTYPE_Grayscale, texture=&LowResBlurredNoise}
  n3 = Rotator(Coordinate=TextureCoordinate(), Time=Constant() {r=1.9976})
  n4 = Frac(None=Panner(Coordinate=@n3) {speedX=0.75})
  n5 = ComponentMask(None=Abs(None=Add(A=@n4) {constB=-0.5})) {r=true}
  n6 = Constant() {r=0.2}
  n7 = Constant() {r=0.55}
  n8 = Multiply(
   A=Multiply(A=Power(Base=@n2.R)) {constB=3}
   B=Clamp(None=MaterialFunctionCall(Alpha=@n5, Min=@n6, Max=@n7) {materialFunction=&SmoothStep}.Result)
  )
  n9 = TextureSample(UVs=Panner(Coordinate=@n3) {speedX=-0.1, speedY=1}) {samplerType=SAMPLERTYPE_LinearColor, texture=/Game/Gadgets/Assets/VinderTech_GliderChute/Glider_Holiday_01/Textures/T_Glider_Holiday_01_fx}
  n10 = Add(A=CameraVectorWS()) {constB=-0.5}
  n11 = TextureSample(UVs=ComponentMask(None=Multiply(A=@n10) {constA=0.5, constB=2}) {g=true, r=true}) {texture=/Game/Environments/Props/Rocks/Textures/T_Props_CrystalEnv_Purple}
  n12 = VectorParameter() {parameterName=Cropped UV Coordinates, defaultValue=<rgba 0.019 0.018 0.591 0.59>}
  n13 = TextureSample(UVs=MaterialFunctionCall!(
    UVs=TextureCoordinate()
    TextureIn
    UpperLeftCorner=AppendVector(A=@n12.R, B=@n12.R)
    LowerRightCorner=AppendVector(A=@n12.B, B=@n12.B)
    ExportFloat4
   ) {materialFunction=/Engine/Functions/Engine_MaterialFunctions01/Texturing/TextureCropping}.Crop UVs) {samplerType=SAMPLERTYPE_Masks, texture=/Game/Gadgets/Assets/VinderTech_GliderChute/Glider_Holiday_01/Textures/T_Glider_Holiday_01_M}
  n14 = MaterialFunctionCall!(Sine Phase=Time(), Value 1=Constant() {r=-1}, Value 2=Constant() {r=1}) {materialFunction=/Engine/Functions/Engine_MaterialFunctions02/Utility/Sine_Remapped}
  n15 = LinearInterpolate(A=@n13.G, B=@n13.G, Alpha=Clamp(None=Ceil(None=@n14.Remapped Sine)))
  n16 = Multiply(A=Add(A=Multiply(A=@n8, B=@n9.RGB), B=Multiply(A=@n11.RGB) {constB=0.2}), B=@n15)
  n17 = BreakMaterialAttributes!(Attr=FunctionInput() {inputType=FunctionInput_MaterialAttributes})
  n18 = LinearInterpolate(A=@n17.BaseColor, B=VectorParameter() {parameterName=Design Color, defaultValue=<rgba 0.7188 0.6589 0.6027 1>}.RGB, Alpha=@n15)
  n19 = Constant3Vector() {constant=<rgba 1 0.0658 0 0>}
  n20 = Fresnel() {exponent=3}
  n21 = LinearInterpolate(A=@n18, B=Clamp(None=LinearInterpolate(A=@n18, B=@n19, Alpha=@n20)), Alpha=@n15)
  n22 = LinearInterpolate(
   A=AppendVector(A=AppendVector(A=@n17.Specular, B=@n17.Specular), B=@n17.Roughness)
   B=VectorParameter() {parameterName=Knit Pattern Spec, defaultValue=<rgba 0.851 1 0.3804 1>}.RGB
   Alpha=@n15
  )
  FunctionOutput(None=MakeMaterialAttributes!(
    BaseColor=Add(A=@n16, B=@n21)
    Metallic=ComponentMask(None=@n22) {g=true}
    Specular=ComponentMask(None=@n22) {r=true}
    Roughness=ComponentMask(None=@n22) {b=true}
    Anisotropy
    EmissiveColor=Add(A=@n16, B=@n17.EmissiveColor)
    Opacity
    OpacityMask
    Normal=@n17.Normal
    Tangent
    WorldPositionOffset
    SubsurfaceColor
    ClearCoat
    ClearCoatRoughness
    AmbientOcclusion
    Refraction
    CustomizedUVs_0
    CustomizedUVs_1
    CustomizedUVs_2
    CustomizedUVs_3
    CustomizedUVs_4
    CustomizedUVs_5
    CustomizedUVs_6
    CustomizedUVs_7
    PixelDepthOffset
    ShadingModel
    Displacement
   ))
  Panner(Coordinate=@n1) {speedX=0.2, speedY=-0.1}
 ]}
