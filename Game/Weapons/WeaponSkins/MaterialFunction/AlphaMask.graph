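# Material function for weapon skins that masks a material's opacity mask by a scalar input.
# It takes a Material Attributes input, multiplies its Opacity Mask by the scalar and saturates the result.
# The result is written back into the attributes, and all other attributes pass through unchanged.
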MaterialFunction /Game/Weapons/WeaponSkins/MaterialFunction/AlphaMask
graph {nodes [
  n1 = GetMaterialAttributes(None=FunctionInput() {bUsePreviewValueAsDefault=true, inputName=Material, inputType=FunctionInput_MaterialAttributes}) {attributeGetTypes [#1]}
  n2 = Multiply(A=@n1.OpacityMask, B=FunctionInput() {bUsePreviewValueAsDefault=true, inputType=FunctionInput_Scalar})
  FunctionOutput(None=SetMaterialAttributes!(MaterialAttributes=@n1.MaterialAttributes, OpacityMask=Saturate(None=@n2)) {attributeSetTypes [#1]})
 ]}
