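# Material function for weapon skins that adds an animated, pulsing emissive glow to masked elements.
# A tint color and intensity parameter are multiplied by the mask input to form the glow.
# A time-driven sine, scaled by pulse speed, drives a clamped lerp whose lower value derives from pulse intensity, modulating that glow.
# The result is added to the incoming material attributes' existing emissive color, and the updated attributes are output.
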MaterialFunction /Game/Weapons/WeaponSkins/MaterialFunction/AnimatedElementGLow
graph {nodes [
  n1 = GetMaterialAttributes(None=FunctionInput() {bUsePreviewValueAsDefault=true, inputName=Material, inputType=FunctionInput_MaterialAttributes}) {attributeGetTypes [#1]}
  n2 = VectorParameter() {group=Animated Element Glow, parameterName=GlowingElementColor, defaultValue=<rgba 1 0 0 1>}
  n3 = ScalarParameter() {group=Animated Element Glow, parameterName=Glowing Element Intensity}
  n4 = Subtract(B=ScalarParameter() {defaultValue=0.5, group=Animated Element Glow, parameterName=Glowing Element Pulse Intensity})
  n5 = Sine(None=Multiply(A=Time(), B=ScalarParameter() {defaultValue=100, group=Animated Element Glow, parameterName=Glowing Element Pulse Speed}))
  n6 = Multiply(
   A=Multiply(A=Multiply(A=@n2.RGB, B=@n3), B=FunctionInput() {bUsePreviewValueAsDefault=true, inputName=GlowingElementMask})
   B=Clamp(None=LinearInterpolate(A=@n4, Alpha=@n5))
  )
  FunctionOutput(None=SetMaterialAttributes!(MaterialAttributes=@n1.MaterialAttributes, EmissiveColor=Add(A=@n1.EmissiveColor, B=@n6)) {attributeSetTypes [#1]})
 ]}
