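# Material function for weapon skins that fakes anisotropic shading by perturbing the normal of incoming Material Attributes.
# DitherTemporalAA drives a signed jitter along tangent-space X, which is combined with an Anisotropic Offset parameter and an up vector.
# The result is merged into the existing normal via BlendAngleCorrectedNormals, then blended in by the MASK input.
# A quality switch applies the effect at High and Default, while Low and Medium pass the attributes through unchanged.
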MaterialFunction /Game/Weapons/WeaponSkins/MaterialFunction/Anisotropic
graph {nodes [
  n1 = FunctionInput() {bUsePreviewValueAsDefault=true, inputType=FunctionInput_MaterialAttributes}
  n2 = GetMaterialAttributes(None=@n1) {attributeGetTypes [#1]}
  n3 = Multiply(
   A=LinearInterpolate(Alpha=MaterialFunctionCall() {materialFunction=&DitherTemporalAA}.Result) {constA=-1}
   B=Multiply(A=Constant2Vector() {r=1}) {constB=0.7}
  ) {constB=0.9}
  n4 = Constant()
  n5 = Constant3Vector() {constant=<rgba 0 0 10 0>}
  n6 = Add(
   A=Add(A=AppendVector(A=@n3, B=@n4), B=VectorParameter() {parameterName=Anisotropic Offset, defaultValue=<rgba 0 0 1 1>}.RGB)
   B=Normalize(VectorInput=Add(B=@n5))
  )
  n7 = BlendMaterialAttributes(
   A=@n1
   B=SetMaterialAttributes!(
    MaterialAttributes=@n2.MaterialAttributes
    Normal=MaterialFunctionCall(BaseNormal=@n2.Normal, AdditionalNormal=@n6) {materialFunction=&BlendAngleCorrectedNormals}.Result
   ) {attributeSetTypes [#1]}
   Alpha=FunctionInput() {bUsePreviewValueAsDefault=true, inputName=MASK, inputType=FunctionInput_Scalar}
  )
  FunctionOutput(None=QualitySwitch(Default=@n7, Low=@n2.MaterialAttributes, High=@n7, Medium=@n2.MaterialAttributes))
 ]}
