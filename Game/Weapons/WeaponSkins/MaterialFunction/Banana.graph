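# Material function for the banana (Peely) skin that swaps facial expressions and ripens the banana over match time, editing World Position Offset and Base Color.
# Pose parameters pass thresholds to pick vertex-color mask values, collapsing unwanted face elements toward the object position.
# A static switch selects combined eye/brow logic or separate masks with a PeelyBlenderFace dizzy override.
# MatchTime shifts color from green tint through base to brown texture, masked by FX texture channels.
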MaterialFunction /Game/Weapons/WeaponSkins/MaterialFunction/Banana
graph {nodes [
  Add()
  Constant()
  Constant() {r=1}
  n1 = GetMaterialAttributes(None=FunctionInput() {inputName=Material, inputType=FunctionInput_MaterialAttributes}) {attributeGetTypes [#1, #2]}
  n2 = WorldPosition()
  n3 = ObjectPositionWS()
  n4 = Constant() {r=0.95}
  n5 = Constant() {r=1}
  n6 = ScalarParameter() {defaultValue=0.1, group=Banana, parameterName=Threshold_R_blink_pose}
  n7 = VertexColor()
  n8 = ChannelMaskParameter(None=@n7) {group=Banana, parameterName=Banana_Former Red}
  n9 = ScalarParameter() {group=Banana_Poses, parameterName=C_glabella_up_pose}
  n10 = ScalarParameter() {defaultValue=0.1, group=Banana, parameterName=Threshold_Glabella_up}
  n11 = Constant() {r=0.1}
  n12 = ChannelMaskParameter(None=@n7) {group=Banana, maskChannel=Green, parameterName=Banana_Former Green, defaultValue=<rgba 0 1 0 0>}
  n13 = ScalarParameter() {group=Banana_Poses, parameterName=C_glabella_down_pose}
  n14 = ScalarParameter() {defaultValue=0.1, group=Banana, parameterName=Threshold_Glabella_down}
  n15 = ScalarParameter() {defaultValue=0.2, group=Banana, parameterName="GlabellaDownValue(G)"}
  n16 = ScalarParameter() {group=Banana_Poses, parameterName=R_brow_up_pose}
  n17 = ScalarParameter() {defaultValue=0.1, group=Banana, parameterName=Threshold_R_brow_up_pose}
  n18 = ScalarParameter() {defaultValue=0.4, group=Banana, parameterName="BananaBrowUpValue(R)"}
  n19 = ScalarParameter() {defaultValue=0.375, group=Banana, parameterName=Threshold_squint_inner_pose}
  n20 = MaterialFunctionCall!(
   Hide Element Toggle=Constant() {r=1}
   VertexColorChannel=@n8.R
   Vertex Value to hide=ScalarParameter() {defaultValue=0.3, group=Banana, parameterName="BananaHappyPoseValue(R)"}
  ) {materialFunction=/Game/Packages/Fortress_SharedMaterials/Base_Material_Functions/VertexColorMask}
  n21 = If(
   A=ScalarParameter() {group=Banana_Poses, parameterName=R_squint_inner_pose}
   B=@n19
   "A > B"=MaterialFunctionCall(
    Hide Element Toggle=Constant() {r=1}
    VertexColorChannel=@n8.R
    Vertex Value to hide=ScalarParameter() {defaultValue=0.5, group=Banana, parameterName="BananaSquintPoseValue(R)"}
   ) {materialFunction=&VertexColorMask}.Result
   "A == B"=@n20.Result
   "A < B"=@n20.Result
  )
  n22 = If(
   A=@n16
   B=@n17
   "A > B"=MaterialFunctionCall(Hide Element Toggle=Constant() {r=1}, VertexColorChannel=@n12.R, Vertex Value to hide=@n18) {materialFunction=&VertexColorMask}.Result
   "A == B"=@n21
   "A < B"=@n21
  )
  n23 = If(
   A=@n13
   B=@n14
   "A > B"=MaterialFunctionCall(Hide Element Toggle=Constant() {r=1}, VertexColorChannel=@n12.R, Vertex Value to hide=@n15) {materialFunction=&VertexColorMask}.Result
   "A == B"=@n22
   "A < B"=@n22
  )
  n24 = If(
   A=@n9
   B=@n10
   "A > B"=MaterialFunctionCall(Hide Element Toggle=Constant() {r=1}, VertexColorChannel=@n12.R, Vertex Value to hide=@n11) {materialFunction=&VertexColorMask}.Result
   "A == B"=@n23
   "A < B"=@n23
  )
  n25 = If(
   A=ScalarParameter() {group=Banana_Poses, parameterName=R_blink_pose}
   B=@n6
   "A > B"=MaterialFunctionCall(
    Hide Element Toggle=Constant() {r=1}
    VertexColorChannel=@n8.R
    Vertex Value to hide=ScalarParameter() {defaultValue=0.6, group=Banana, parameterName="BananaBlinkPoseValue(R)"}
   ) {materialFunction=&VertexColorMask}.Result
   "A == B"=@n24
   "A < B"=@n24
  )
  n26 = Max(A=@n8.R, B=@n12.R)
  n27 = Clamp(None=OneMinus(None=If(A=@n26, B=@n4, "A > B"=@n5, "A == B"=@n25, "A < B"=@n25)))
  n28 = MaterialFunctionCall(
   Hide Element Toggle=Constant() {r=1}
   VertexColorChannel=@n8.R
   Vertex Value to hide=ScalarParameter() {group=Banana, parameterName="DizzyFaceValue(R)"}
  ) {materialFunction=&VertexColorMask}
  n29 = MaterialFunctionCall(
   Hide Element Toggle=Constant() {r=1}
   VertexColorChannel=@n8.R
   Vertex Value to hide=ScalarParameter() {defaultValue=0.3, group=Banana, parameterName="BananaHappyPoseValue(R)"}
  ) {materialFunction=&VertexColorMask}
  n30 = If(
   A=ScalarParameter() {group=Banana_Poses, parameterName=R_squint_inner_pose}
   B=@n19
   "A > B"=MaterialFunctionCall(
    Hide Element Toggle=Constant() {r=1}
    VertexColorChannel=@n8.R
    Vertex Value to hide=ScalarParameter() {defaultValue=0.5, group=Banana, parameterName="BananaSquintPoseValue(R)"}
   ) {materialFunction=&VertexColorMask}.Result
   "A == B"=@n29.Result
   "A < B"=@n29.Result
  )
  n31 = If(
   A=ScalarParameter() {group=Banana_Poses, parameterName=R_blink_pose}
   B=@n6
   "A > B"=MaterialFunctionCall(
    Hide Element Toggle=Constant() {r=1}
    VertexColorChannel=@n8.R
    Vertex Value to hide=ScalarParameter() {defaultValue=0.6, group=Banana, parameterName="BananaBlinkPoseValue(R)"}
   ) {materialFunction=&VertexColorMask}.Result
   "A == B"=@n30
   "A < B"=@n30
  )
  n32 = If(A=@n8.R, B=Constant() {r=0.95}, "A > B"=Constant() {r=1}, "A == B"=@n31, "A < B"=@n31)
  n33 = ScalarParameter() {defaultValue=1, group=Banana, parameterName=sdfaafafafafa}
  n34 = If(A=@n16, B=@n17, "A > B"=@n18, "A == B"=@n33, "A < B"=@n33)
  n35 = If(A=@n13, B=@n14, "A > B"=@n15, "A == B"=@n33, "A < B"=@n33)
  n36 = MaterialFunctionCall(
   Hide Element Toggle=Constant() {r=1}
   VertexColorChannel=@n12.R
   Vertex Value to hide=Min(A=Min(A=@n34, B=@n35), B=If(A=@n9, B=@n10, "A > B"=@n11, "A == B"=@n33, "A < B"=@n33))
  ) {materialFunction=&VertexColorMask}
  n37 = If(A=@n12.R, B=Constant() {r=0.95}, "A > B"=Constant() {r=1}, "A == B"=@n36.Result, "A < B"=@n36.Result)
  n38 = OneMinus(None=Max(A=@n37, B=@n32))
  n39 = If(
   A=Saturate(None=ScalarParameter() {parameterName=PeelyBlenderFace})
   B=@n8.R
   "A > B"=Saturate(None=OneMinus(None=@n28.Result))
   "A == B"=@n38
   "A < B"=@n38
  )
  n40 = Multiply(
   A=Multiply(A=Subtract(A=@n2.XYZ, B=@n3)) {constB=-1}
   B=Saturate(None=StaticSwitchParameter(True=@n27, False=@n39) {group=Banana, parameterName="OverrideEyebrowLogicWithEye(BuffCat)"})
  )
  n41 = ScalarParameter() {defaultValue=0.06, group=Banana, parameterName=BananaDesat}
  n42 = TextureSampleParameter2D() {
   group=Banana
   parameterName=Banana FX
   samplerSource=SSM_Wrap_WorldGroupSettings
   samplerType=SAMPLERTYPE_Masks
   texture=/Game/Characters/Player/Male/Medium/Bodies/M_MED_Banana/Textures/T_Male_Commando_Banana2_Body_FX
  }
  n43 = ScalarParameter() {group=Banana, parameterName=MatchTime}
  n44 = LinearInterpolate(
   A=@n1.BaseColor
   B=LinearInterpolate(
    A=Desaturation(None=@n1.BaseColor, Fraction=@n41)
    B=VectorParameter() {group=Banana, parameterName=Green, defaultValue=<rgba 0.104 0.3229 0 1>}.RGB
    Alpha=@n42.G
   )
   Alpha=Saturate(None=LinearInterpolate(Alpha=@n43) {constA=1, constB=-1.1})
  )
  n45 = TextureSample(UVs=FunctionInput() {bUsePreviewValueAsDefault=true, inputName=Gmap, inputType=FunctionInput_Scalar}) {samplerSource=SSM_Wrap_WorldGroupSettings, texture=/Game/Characters/Player/Male/Medium/Bodies/M_MED_Banana/Textures/T_BananaBrown}
  n46 = Saturate(None=LinearInterpolate(Alpha=@n43) {constA=-1.1})
  n47 = ScalarParameter() {defaultValue=2, group=Banana, parameterName=BlendRate}
  n48 = Saturate(None=Add(A=Subtract(A=@n42.B), B=Multiply(A=@n46, B=@n47)))
  n49 = ScalarParameter() {defaultValue=4, group=Banana, parameterName=BananaBlendPower}
  n50 = Multiply(A=@n42.R, B=Saturate(None=Power(Base=@n48, Exp=@n49)))
  FunctionOutput(None=SetMaterialAttributes!(
    MaterialAttributes=@n1.MaterialAttributes
    WorldPositionOffset=Add(A=@n1.WorldPositionOffset, B=@n40)
    BaseColor=LinearInterpolate(A=@n44, B=@n45.RGB, Alpha=@n50)
   ) {attributeSetTypes [#1, #2]})
  n51 = Constant()
  If(A=Divide(A=Add(A=@n8.R, B=@n12.R)), B=Constant() {r=1}, "A > B"=Add(A=@n37, B=@n31), "A == B"=@n51, "A < B"=@n51)
  If()
  LinearInterpolate()
  Max(A=@n1.WorldPositionOffset, B=@n40)
  Multiply(A=@n37, B=@n32)
  Reroute(None=@n8.R)
  ScalarParameter() {defaultValue=0.1, group=Banana, parameterName=BananaEyeThreshold}
 ]}
