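# Material function for weapon skins that scales an emissive input by time of day.
# A Full TOD Control static switch picks either morning, day, evening and night powers via Scalar_Time_of_Day, or a sine remap between day and night values, phased from the Fortnite collection's TimeOfDay offset by PhaseAlign.
# An optional static switch divides the result by eye adaptation, with mobile substituting a constant 1.
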MaterialFunction /Game/Weapons/WeaponSkins/MaterialFunction/BoostEmissivebyTOD
graph {nodes [
  n1 = ShadingPathSwitch(Default=EyeAdaptation().EyeAdaptation, Mobile=Constant() {r=1})
  n2 = CollectionParameter() {collection=&FortniteMaterialParameters, parameterName=TimeOfDay_0-1}
  n3 = ScalarParameter() {defaultValue=0.65, group=TOD Emissive, parameterName=PhaseAlign}
  n4 = StaticSwitchParameter(
   True=MaterialFunctionCall(
    Value_Morning=ScalarParameter() {defaultValue=0.15, group=TOD Emissive, parameterName=Emissive Power Morning, sortPriority=1}
    Value_Day=ScalarParameter() {group=TOD Emissive, parameterName=Emissive Power Day, sortPriority=2}
    Value_Evening=ScalarParameter() {defaultValue=0.15, group=TOD Emissive, parameterName=Emissive Power Evening, sortPriority=3}
    Value_Night=ScalarParameter() {defaultValue=1, group=TOD Emissive, parameterName=Emissive Power Night, sortPriority=4}
   ) {materialFunction=&Scalar_Time_of_Day}.Result
   False=MaterialFunctionCall!(
    Sine Phase=Subtract(A=@n2, B=@n3)
    Value 1=ScalarParameter() {defaultValue=1, group=TOD Emissive, parameterName=Emissive_Day}
    Value 2=ScalarParameter() {defaultValue=1, group=TOD Emissive, parameterName=Emissive_Night}
   ) {materialFunction=/Engine/Functions/Engine_MaterialFunctions02/Utility/Sine_Remapped}.Remapped Sine
  ) {desc=Allows you to fully control emissive power based on 6 hour increments., group=TOD Emissive, parameterName=Full TOD Control}
  n5 = Multiply(A=FunctionInput(), B=@n4)
  FunctionOutput(None=StaticSwitchParameter(True=Divide(A=@n5, B=@n1), False=@n5) {group=TOD Emissive, parameterName=UseEmissiveEyeAdaption})
 ]}
