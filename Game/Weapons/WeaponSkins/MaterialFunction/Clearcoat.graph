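# Material function for weapon skins that adds a clear coat layer to incoming material attributes, masked by an Alpha input.
# It sets the Clear Coat shading model at full coat strength, scales Metallic, and uses one parameter to swap the input roughness and a constant roughness between base and coat.
# The result blends over the original attributes by Alpha.
# A quality switch skips the effect on Low and Medium, passing the input through unchanged.
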MaterialFunction /Game/Weapons/WeaponSkins/MaterialFunction/Clearcoat
graph {nodes [
  n1 = FunctionInput() {inputType=FunctionInput_MaterialAttributes}
  n2 = GetMaterialAttributes(None=@n1) {attributeGetTypes [#1, #2]}
  n3 = ScalarParameter() {defaultValue=0.1, group=Clearcoat, parameterName=Undercoat Roughness}
  n4 = ScalarParameter() {group=Clearcoat, parameterName=Roughness Map Now affects Clearcoat roughness}
  n5 = SetMaterialAttributes!(
   MaterialAttributes=@n2.MaterialAttributes
   CustomData0=Constant() {r=1}
   CustomData1=LinearInterpolate(A=@n3, B=@n2.Roughness, Alpha=@n4)
   Roughness=LinearInterpolate(A=@n2.Roughness, B=@n3, Alpha=@n4)
   Metallic=Multiply(A=@n2.Metallic, B=ScalarParameter() {defaultValue=1, group=Clearcoat, parameterName=Undercoat Metallic Multiplier})
   ShadingModel=ShadingModel() {shadingModel=MSM_ClearCoat}
  ) {attributeSetTypes [
    #3
    #4
    #1
    #2
    #5
   ]}
  n6 = BlendMaterialAttributes(A=@n2.MaterialAttributes, B=@n5, Alpha=FunctionInput() {bUsePreviewValueAsDefault=true, inputName=Alpha, inputType=FunctionInput_Scalar})
  FunctionOutput(None=QualitySwitch(Default=@n6, Low=@n1, High=@n6, Medium=@n1))
 ]}
