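# DayNightLerp is a library-exposed material function in the weapon skins folder that outputs a day/night blend value, described as transitioning between black and white at 7 and 21 time of day.
# It offsets the normalized TimeOfDay value from FortniteMaterialParameters, then remaps its cosine to drive a lerp between the Bottom and Top parameters.
# Default endpoints overshoot the final clamp, flattening the cosine into plateaus with shorter transitions.
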MaterialFunction /Game/Weapons/WeaponSkins/MaterialFunction/DayNightLerp
graph {
 properties {bExposeToLibrary=true}
 nodes [
  n1 = ScalarParameter() {defaultValue=-1, parameterName=Bottom}
  n2 = ScalarParameter() {defaultValue=1.7, parameterName=Top}
  n3 = CollectionParameter() {collection=&FortniteMaterialParameters, parameterName=TimeOfDay_0-1}
  n4 = ScalarParameter() {defaultValue=0.48, parameterName=Tod_Offset}
  n5 = ConstantBiasScale(None=Cosine(None=Add(A=@n3, B=@n4)))
  FunctionOutput(None=Clamp(None=LinearInterpolate(A=@n1, B=@n2, Alpha=@n5))) {description=Transitions between Black and White at 7 and 21 TOD}
 ]
}
