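# Material function for weapon skins that outputs a sun-facing shading term, a dot product against the world-space pixel normal.
# It reads SunlightDirectionalVector from the FortniteMaterialParameters collection, breaks out its components, builds a float3, normalizes it and dots it with the normal.
# Only the R component feeds all three float3 axes, so the compared direction is diagonal rather than the full sun vector.
# The result is unclamped, with no view-angle input despite the Fresnel name.
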MaterialFunction /Game/Weapons/WeaponSkins/MaterialFunction/FortniteSunnySideFresnel
graph {nodes [
  n1 = CollectionParameter() {
   collection=&FortniteMaterialParameters
   parameterName=SunlightDirectionalVector
  }
  n2 = MaterialFunctionCall(Float4=@n1) {materialFunction=&BreakOutFloat4Components}
  n3 = MaterialFunctionCall(X=@n2.R, Y=@n2.R, Z=@n2.R) {materialFunction=&MakeFloat3}
  FunctionOutput(None=DotProduct(A=Normalize(VectorInput=@n3.Result), B=PixelNormalWS()))
 ]}
