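# Material function for weapon skins that turns incoming material attributes into a glowing crystal look, blended by an Alpha input.
# A Fresnel-indexed gradient sets base color, with optional source tint and bump-offset cracks that also affect roughness.
# Emissive fakes refraction with a reflection-sampled sky texture, adding edge-to-core glow from Fresnel and an edge map.
# Sun direction and color drive a sheen, and sunny-side and day-night functions modulate brightness.
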MaterialFunction /Game/Weapons/WeaponSkins/MaterialFunction/Crystal
graph {nodes [
  n1 = ScalarParameter() {group=Crystal, parameterName=CrystalMetallic}
  n2 = TextureCoordinate()
  n3 = ScalarParameter() {defaultValue=8, group=Crystal, parameterName=CrackTiling}
  n4 = TextureSample(UVs=BumpOffset(Coordinate=Multiply(A=@n2, B=@n3), Height=ScalarParameter() {defaultValue=-0.5, group=Crystal, parameterName=CrackHeight})) {samplerType=SAMPLERTYPE_LinearColor, texture=/Game/Global/Textures/Default/Masks_and_Gradients/T_Cracks}
  n5 = FunctionInput() {inputName=Material, inputType=FunctionInput_MaterialAttributes}
  n6 = GetMaterialAttributes(None=@n5) {attributeGetTypes [
    #1
    #2
    #3
    #4
   ]}
  n7 = ScalarParameter() {group=Crystal, parameterName=CrackIntensity}
  n8 = LinearInterpolate(A=@n6.Roughness, B=LinearInterpolate(A=@n6.Roughness, B=Constant() {r=0.7}, Alpha=@n4.R), Alpha=@n7)
  n9 = ScalarParameter() {defaultValue=1, group=Crystal, parameterName=CrystalSubsurfaceOpacity}
  n10 = Constant()
  n11 = Transform(None=@n6.Normal)
  n12 = Fresnel(ExponentIn=ScalarParameter() {defaultValue=0.8, group=Crystal, parameterName=Crystal_FresEX}, Normal=@n11) {baseReflectFraction=0}
  n13 = TextureSampleParameter2D(UVs=Saturate(None=@n12)) {group=Crystal, parameterName=CrystalGradient, texture=/Game/Global/Textures/Default/Masks_and_Gradients/HeistColorGradients/CrystalGradient}
  n14 = Multiply(
   A=LinearInterpolate(A=@n13.RGB, B=Multiply(A=@n6.BaseColor, B=@n13.RGB), Alpha=ScalarParameter() {group=Crystal, parameterName=SourceBaseColorStrength})
   B=ScalarParameter() {defaultValue=1, parameterName=Crystal Brightness}
  )
  n15 = LinearInterpolate(
   A=@n14
   B=LinearInterpolate(A=@n14, B=VectorParameter() {group=Crystal, parameterName=CrackColor, defaultValue=<rgba 0 0 0 1>}.RGB, Alpha=@n4.R)
   Alpha=@n7
  )
  n16 = Multiply(
   A=TextureSample(UVs=ReflectionVectorWS(CustomWorldNormal=@n11) {bNormalizeCustomWorldNormal=true}) {samplerType=SAMPLERTYPE_LinearColor, texture=/Game/Packages/Fortress_Sky/TexturesHDR/T_AthenaSkylight_LowRes1}.RGB
   B=ScalarParameter() {defaultValue=3.3537, group=Crystal, parameterName=Refraction Intensity}
  )
  n17 = VectorParameter() {group=Crystal, parameterName=Crystal EdgeEmissive, defaultValue=<rgba 0 0.0106 0.04 1>}
  n18 = VectorParameter() {group=Crystal, parameterName=Crystal Core Emissive, defaultValue=<rgba 0 0.3044 3.5 1>}
  n19 = Saturate(None=Power(Base=@n12))
  n20 = ScalarParameter() {defaultValue=2, group=Crystal, parameterName=CoreEmissive_Power}
  n21 = FunctionInput(Preview=TextureSample() {texture=/Game/Accessories/FORT_Backpacks/M_MED_Heist/Textures/T_FN_Character_Heist_FX}.G) {bUsePreviewValueAsDefault=true, inputName=Edge Map, inputType=FunctionInput_Scalar}
  n22 = ScalarParameter() {defaultValue=0.4, group=Crystal, parameterName=EdgeMask_Intensity}
  n23 = Saturate(None=Add(A=Power(Base=@n19, Exp=@n20), B=Multiply(A=@n21, B=@n22)))
  n24 = ScalarParameter() {defaultValue=1.5, group=Crystal, parameterName=Emissive_Sunny}
  n25 = ScalarParameter() {defaultValue=0.08, group=Crystal, parameterName=Emissive_Shade}
  n26 = MaterialFunctionCall() {materialFunction=&/FortniteSunnySideFresnel}
  n27 = Add(
   A=Power(Base=Multiply(A=@n16) {constB=0.5}) {constExponent=1.5}
   B=Multiply(A=LinearInterpolate(A=@n17.RGB, B=@n18.RGB, Alpha=@n23), B=LinearInterpolate(A=@n24, B=@n25, Alpha=@n26.Result))
  )
  n28 = ComponentMask(None=CollectionParameter() {collection=&FortniteMaterialParameters, parameterName=SunLightColor}) {b=true, g=true, r=true}
  n29 = Constant() {r=0.3}
  n30 = Saturate(None=Multiply(A=Desaturation(None=@n28, Fraction=@n29), B=ScalarParameter() {defaultValue=2, group=Crystal, parameterName=SheenBrightness}))
  n31 = Multiply(A=@n27, B=@n30)
  n32 = ScalarParameter() {group=Crystal, parameterName=SheenIntensity}
  n33 = ScalarParameter() {defaultValue=3.5, group=Crystal, parameterName=Sheen_Power}
  n34 = CollectionParameter() {
   collection=&FortniteMaterialParameters
   parameterName=SunlightDirectionalVector
  }
  n35 = OneMinus(None=@n34)
  n36 = Power(
   Base=DotProduct(A=ComponentMask(None=@n35) {b=true, g=true, r=true}, B=ReflectionVectorWS())
   Exp=ScalarParameter() {defaultValue=8, group=Crystal, parameterName=VectorPower}
  )
  n37 = LinearInterpolate(
   B=ScalarParameter() {defaultValue=25, group=Crystal, parameterName=Sheen Brightness}
   Alpha=Multiply(A=Power(Base=@n12, Exp=@n33) {constExponent=4}, B=Saturate(None=@n36))
  ) {constA=-1}
  n38 = Multiply(
   A=LinearInterpolate(A=@n27, B=LinearInterpolate(A=@n31, B=@n30, Alpha=@n32), Alpha=Saturate(None=@n37))
   B=LinearInterpolate(Alpha=MaterialFunctionCall() {materialFunction=&/DayNightLerp}.Result) {constA=0.33}
  )
  FunctionOutput(None=BlendMaterialAttributes(
    A=@n5
    B=MakeMaterialAttributes!(
     BaseColor=@n15
     Metallic=@n1
     Specular=@n6.Specular
     Roughness=@n8
     Anisotropy
     EmissiveColor=@n38
     Opacity=@n9
     OpacityMask
     Normal=@n6.Normal
     Tangent
     WorldPositionOffset
     SubsurfaceColor=@n10
     ClearCoat=@n10
     ClearCoatRoughness
     AmbientOcclusion
     Refraction
     CustomizedUVs_0
     CustomizedUVs_1
     CustomizedUVs_2
     CustomizedUVs_3
     CustomizedUVs_4
     CustomizedUVs_5
     CustomizedUVs_6
     CustomizedUVs_7
     PixelDepthOffset
     ShadingModel
     Displacement
    )
    Alpha=FunctionInput() {bUsePreviewValueAsDefault=true, inputName=Alpha, inputType=FunctionInput_Scalar, sortPriority=1}
   ))
 ]}
