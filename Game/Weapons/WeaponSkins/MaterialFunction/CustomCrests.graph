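# Material function for weapon skins that stamps a parameterized crest texture onto incoming material attributes.
# Crest UVs are rotated, offset and scaled by parameters, with a 0-1 mask confining the stamp.
# A static bool picks opacity and color from the crest's own RGBA, or from a fixed mask texture adding an outline color and constant surface values.
# A MultiplyBaseColor switch instead multiplies the crest into the underlying base color before blending.
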MaterialFunction /Game/Weapons/WeaponSkins/MaterialFunction/CustomCrests
graph {nodes [
  n1 = FunctionInput() {inputName=Materials, inputType=FunctionInput_MaterialAttributes}
  n2 = GetMaterialAttributes(None=@n1) {attributeGetTypes [#1]}
  n3 = TextureCoordinate()
  n4 = Constant() {r=0.5}
  n5 = ScalarParameter() {group=Custom Crest, parameterName=Crest Rotation Angle}
  n6 = ScalarParameter() {defaultValue=0.2695, group=Custom Crest, parameterName=Crest_X}
  n7 = ScalarParameter() {defaultValue=0.0619, group=Custom Crest, parameterName=Crest_Y}
  n8 = MaterialFunctionCall(
   UVs=Add(
    A=MaterialFunctionCall(UVs=@n3, Rotation Center=@n4, "Rotation Angle (0-1)"=@n5) {materialFunction=&CustomRotator}.Rotated Values
    B=AppendVector(A=@n6, B=@n7)
   )
   Texture Scale=ScalarParameter() {defaultValue=0.04, group=Custom Crest, parameterName=CrestScale}
  ) {materialFunction=&ScaleUVsByCenter}
  n9 = TextureSampleParameter2D(UVs=@n8.UVs) {group=Custom Crest, parameterName=T_Crest, texture=/Game/Characters/Player/Male/Medium/Bodies/M_MED_Soccer/Textures/cRESTS/USA}
  n10 = SetMaterialAttributes!(MaterialAttributes=@n2.MaterialAttributes, BaseColor=Multiply(A=@n2.BaseColor, B=@n9.RGB)) {attributeSetTypes [#1]}
  n11 = Multiply(A=@n9.A, B=@n8.0-1 mask)
  n12 = TextureSample(UVs=@n8.UVs) {samplerType=SAMPLERTYPE_Masks, texture=/Game/Characters/Player/Male/Medium/Bodies/M_MED_Soccer/Textures/cRESTS/Crests}
  n13 = Add(A=@n12.R, B=@n12.R)
  n14 = StaticBoolParameter() {group=Custom Crest, parameterName=use Alpha Channel for Opacity}
  n15 = StaticSwitch(True=@n11, False=Multiply(A=Saturate(None=@n13), B=@n8.0-1 mask), Value=@n14)
  n16 = SetMaterialAttributes!(
   MaterialAttributes=@n1
   BaseColor=LinearInterpolate(A=VectorParameter() {group=Custom Crest, parameterName=CC_OutlineColor, defaultValue=<rgba 0 0 0 1>}.RGB, B=@n9.RGB, Alpha=@n12.G)
   Metallic=Constant()
   Specular=Constant() {r=0.45}
   Roughness=Constant() {r=0.65}
  ) {attributeSetTypes [
    #1
    #2
    #3
    #4
   ]}
  n17 = StaticSwitch(
   True=SetMaterialAttributes!(MaterialAttributes=@n1, BaseColor=@n9.RGB) {attributeSetTypes [#1]}
   False=@n16
   Value=@n14
  )
  FunctionOutput(None=StaticSwitchParameter(True=BlendMaterialAttributes(A=@n1, B=@n10, Alpha=@n15), False=BlendMaterialAttributes(A=@n1, B=@n17, Alpha=@n15)) {group=Custom Crest, parameterName=MultiplyBaseColor})
  Multiply()
  Multiply()
  Reroute(None=@n11)
 ]}
