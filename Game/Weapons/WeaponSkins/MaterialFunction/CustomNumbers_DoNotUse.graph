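# Material function in the weapon skins folder, named DoNotUse, that stamps a customizable one- or two-digit number onto input material attributes.
# Digits come from a jersey-number SubUV atlas on UV1, with a first digit of zero or less switching to single-digit placement.
# The red mask blends outline color into a fill chosen by outline brightness (red counted twice, green omitted) or forced tint.
# A parameter blends toward raw texture color.
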MaterialFunction /Game/Weapons/WeaponSkins/MaterialFunction/CustomNumbers_DoNotUse
graph {nodes [
  n1 = FunctionInput() {inputName=Material, inputType=FunctionInput_MaterialAttributes}
  n2 = VectorParameter() {parameterName=CC_OutlineColor, defaultValue=<rgba 0 0 0 1>}
  n3 = Add(A=Add(A=@n2.R, B=@n2.R), B=@n2.B)
  n4 = Desaturation(None=OneMinus(None=@n2.RGB), Fraction=Constant() {r=0.9})
  n5 = LinearInterpolate(
   A=If(
    A=Divide(A=@n3) {constB=3}
    B=Constant() {r=0.4}
    "A > B"=VectorParameter() {parameterName=CC_TritaryColor, defaultValue=<rgba 0 0 0 1>}.RGB
    "A == B"=@n4
    "A < B"=@n4
   )
   B=VectorParameter() {group=Custom Numbers, parameterName=ForcedNumberTint, defaultValue=<rgba 0 0 0 1>}.RGB
   Alpha=ScalarParameter() {group=Custom Numbers, parameterName=OverrideNumberTint}
  )
  n6 = TextureCoordinate() {coordinateIndex=1}
  n7 = ScalarParameter() {defaultValue=0.125, group=Custom Numbers, parameterName=First Digit X}
  n8 = ScalarParameter() {group=Custom Numbers, parameterName=First Digit Y}
  n9 = ScalarParameter() {defaultValue=0.6, group=Custom Numbers, parameterName=Number Scale}
  n10 = MaterialFunctionCall(UVs=Add(A=@n6, B=AppendVector(A=@n7, B=@n8)), Texture Scale=@n9) {materialFunction=&ScaleUVsByCenter}
  n11 = TextureObjectParameter() {
   group=Custom Numbers
   parameterName=Numbers Texture
   samplerType=SAMPLERTYPE_Masks
   texture=/Game/Characters/Player/Male/Medium/Bodies/M_MED_Soccer/Textures/T_JerseyNumbers
  }
  n12 = Constant() {r=4}
  n13 = ScalarParameter() {group=Custom Numbers, parameterName=First Digit}
  n14 = Constant() {r=12}
  n15 = MaterialFunctionCall!(Texture=@n11, UVs=@n10.UVs, SubImages=@n12, Frame=If(A=@n13, B=Constant(), "A > B"=@n13, "A == B"=@n14, "A < B"=@n14)) {materialFunction=&/CustomNumbers_SubUV}
  n16 = AppendVector(
   A=ScalarParameter() {defaultValue=-0.125, group=Custom Numbers, parameterName=Second Digit X}
   B=ScalarParameter() {group=Custom Numbers, parameterName=Second Digit Y}
  )
  n17 = Add(
   A=@n6
   B=AppendVector(
    A=ScalarParameter() {group=Custom Numbers, parameterName=Second Digit X_Single}
    B=ScalarParameter() {group=Custom Numbers, parameterName=Second Digit Y_Single}
   )
  )
  n18 = MaterialFunctionCall(UVs=If(A=@n13, B=Constant(), "A > B"=Add(A=@n6, B=@n16), "A == B"=@n17, "A < B"=@n17), Texture Scale=@n9) {materialFunction=&ScaleUVsByCenter}
  n19 = MaterialFunctionCall(Texture=@n11, UVs=@n18.UVs, SubImages=@n12, Frame=ScalarParameter() {defaultValue=5, group=Custom Numbers, parameterName=Second Digit}) {materialFunction=&CustomNumbers_SubUV}
  n20 = Saturate(None=Max(A=Multiply(A=@n10.0-1 mask, B=@n15.RGB), B=Multiply(A=@n18.0-1 mask, B=@n19.RGB)))
  n21 = ComponentMask(None=@n20) {r=true}
  n22 = Constant() {r=0.45}
  n23 = ScalarParameter() {defaultValue=0.4, group=Custom Numbers, parameterName=Number_Stroke_Roughness}
  n24 = ScalarParameter() {defaultValue=0.6, group=Custom Numbers, parameterName=Number_Internal_Roughness}
  n25 = BlendMaterialAttributes(
   A=SetMaterialAttributes!(
    MaterialAttributes=@n1
    BaseColor=LinearInterpolate(A=@n2.RGB, B=@n5, Alpha=@n21)
    Specular=@n22
    Roughness=LinearInterpolate(A=@n23, B=@n24, Alpha=@n21)
   ) {attributeSetTypes [#1, #2, #3]}
   B=SetMaterialAttributes(MaterialAttributes=@n1, BaseColor=Add(A=@n15.RGB, B=@n19.RGB), Specular=@n22, Roughness=@n24) {attributeSetTypes [#1, #2, #3]}
   Alpha=ScalarParameter() {parameterName=UseCustomDiffuseAndAlpha}
  )
  n26 = Saturate(None=Add(A=Multiply(A=@n15.Alpha, B=@n10.0-1 mask), B=Multiply(A=@n18.0-1 mask, B=@n19.Alpha)))
  FunctionOutput(None=BlendMaterialAttributes(
    A=@n1
    B=BlendMaterialAttributes(A=@n1, B=@n25, Alpha=@n26)
    Alpha=FunctionInput() {bUsePreviewValueAsDefault=true, inputName=Alpha, inputType=FunctionInput_Scalar}
   ))
  Reroute(None=@n20)
 ]}
