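# Material function that composites a customizable two-digit number, such as a jersey number, onto a weapon skin's material attributes.
# Each digit samples an atlas through scaled, offset UV1, and a zero first digit is hidden with kerning removed.
# An optional UV0 sphere mask gives the back independent layout controls and a debug view.
# Atlas channels drive custom colors with an outline-brightness fallback, then blend in with roughness and optional metalness.
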MaterialFunction /Game/Weapons/WeaponSkins/MaterialFunction/CustomNumbers_Optimized
graph {nodes [
  n1 = FunctionInput() {inputName=Material, inputType=FunctionInput_MaterialAttributes}
  n2 = ScalarParameter() {defaultValue=8, group=Numbers, parameterName=First Digit}
  n3 = Constant() {r=11}
  n4 = Constant()
  n5 = Constant() {r=1}
  n6 = If(A=@n2, B=@n4, "A > B"=@n4, "A == B"=@n5, "A < B"=@n5)
  n7 = Constant() {r=11}
  n8 = Divide(A=Add(A=LinearInterpolate(A=@n2, B=@n3, Alpha=@n6)) {constB=6}, B=@n7)
  n9 = TextureCoordinate() {coordinateIndex=1}
  n10 = ScalarParameter() {defaultValue=-0.65, group=Numbers, parameterName=Kerning}
  n11 = SphereMask(
   A=TextureCoordinate()
   B=AppendVector(
    A=ScalarParameter() {defaultValue=0.5, group=Numbers_Back, parameterName="Back_MaskPosition_U(CHNL.0)", sliderMax=1, sortPriority=1}
    B=ScalarParameter() {defaultValue=0.5, group=Numbers_Back, parameterName="Back_MaskPosition_V(CHNL.0)", sliderMax=1, sortPriority=2}
   )
   Radius=ScalarParameter() {defaultValue=0.12, group=Numbers_Back, parameterName=Back_MaskRadius, sliderMax=1, sortPriority=3}
   Hardness=Constant() {r=0.9}
  )
  n12 = Saturate(None=OneMinus(None=@n11))
  n13 = StaticSwitchParameter(
   True=LinearInterpolate(A=ScalarParameter() {defaultValue=-0.65, group=Numbers_Back, parameterName=Back_Kerning}, B=@n10, Alpha=@n12)
   False=@n10
  ) {group=Numbers, parameterName=use Front/Back Independent Control?}
  n14 = ScalarParameter() {defaultValue=0.4, group=Numbers, parameterName=Offset X}
  n15 = StaticSwitchParameter(
   True=LinearInterpolate(A=ScalarParameter() {defaultValue=0.4, group=Numbers_Back, parameterName=Back_Offset X}, B=@n14, Alpha=@n12)
   False=@n14
  ) {group=Numbers, parameterName=use Front/Back Independent Control?}
  n16 = Add(A=@n13, B=@n15)
  n17 = ScalarParameter() {defaultValue=0.4, group=Numbers, parameterName=Scale_X}
  n18 = StaticSwitchParameter(
   True=LinearInterpolate(A=ScalarParameter() {defaultValue=0.4, group=Numbers_Back, parameterName=Back_Scale_X}, B=@n17, Alpha=@n12)
   False=@n17
  ) {group=Numbers, parameterName=use Front/Back Independent Control?}
  n19 = ScalarParameter() {group=Numbers, parameterName=Offset Y}
  n20 = StaticSwitchParameter(True=LinearInterpolate(A=ScalarParameter() {group=Numbers_Back, parameterName=Back_Offset Y}, B=@n19, Alpha=@n12), False=@n19) {group=Numbers, parameterName=use Front/Back Independent Control?}
  n21 = Add(A=@n9, B=AppendVector(A=Multiply(A=@n16, B=@n18), B=@n20))
  n22 = ScalarParameter() {defaultValue=0.4, group=Numbers_Back, parameterName=Back_Scale_Y}
  n23 = ScalarParameter() {defaultValue=0.4, group=Numbers, parameterName=Scale_Y}
  n24 = AppendVector(
   A=Multiply(A=@n7, B=@n18)
   B=StaticSwitchParameter(True=LinearInterpolate(A=@n22, B=@n23, Alpha=@n12), False=@n23) {group=Numbers, parameterName=use Front/Back Independent Control?}
  )
  n25 = MaterialFunctionCall(UVs=@n21, Texture Scale=@n24) {materialFunction=&ScaleUVsByCenter}
  n26 = ComponentMask(None=@n25.UVs) {r=true}
  n27 = TextureObjectParameter() {
   parameterName=NumbersTexture
   samplerType=SAMPLERTYPE_Masks
   texture=/Game/Characters/Player/Male/Medium/Bodies/M_MED_Soccer/Textures/T_Numbers_Soccer
  }
  n28 = TextureSample(UVs=AppendVector(A=Add(A=@n8, B=@n26), B=ComponentMask(None=@n25.UVs) {g=true}), Tex=@n27)
  n29 = Constant() {r=0.5}
  n30 = Add(A=Divide(A=@n21, B=@n18), B=@n29) {constB=0.5}
  n31 = Divide(A=@n29, B=@n18)
  n32 = Constant() {r=1}
  n33 = ScalarParameter() {defaultValue=0.4, group=Numbers, parameterName=Pinch}
  n34 = StaticSwitchParameter(True=LinearInterpolate(A=ScalarParameter() {defaultValue=0.4, group=Numbers_Back, parameterName=Back_Pinch}, B=@n33, Alpha=@n12), False=@n33) {group=Numbers, parameterName=use Front/Back Independent Control?}
  n35 = MaterialFunctionCall(Float2=Multiply(A=Subtract(A=@n30, B=@n31), B=AppendVector(A=@n32, B=@n34))) {materialFunction=&BreakOutFloat2Components}
  n36 = Multiply(A=@n25.0-1 mask, B=Multiply(A=SphereMask(A=@n29, B=@n35.R) {attenuationRadius=0.5}, B=SphereMask(A=@n35.G, B=@n29) {attenuationRadius=0.5}))
  n37 = Multiply(A=@n36, B=@n28.A)
  n38 = Multiply(A=@n28.RGB, B=@n37)
  n39 = Divide(A=Add(A=ScalarParameter() {defaultValue=4, group=Numbers, parameterName=Second Digit}) {constB=6}, B=@n7)
  n40 = Multiply(A=@n18, B=LinearInterpolate(A=Subtract(A=@n15, B=@n13), B=@n15, Alpha=@n6) {desc="Remove Kerning if First Digit = 0"})
  n41 = Add(A=@n9, B=AppendVector(A=ComponentMask(None=@n40) {r=true}, B=@n20))
  n42 = MaterialFunctionCall(UVs=@n41, Texture Scale=@n24) {materialFunction=&ScaleUVsByCenter}
  n43 = ComponentMask(None=@n42.UVs) {r=true}
  n44 = TextureSample(UVs=AppendVector(A=Add(A=@n39, B=@n43), B=ComponentMask(None=@n42.UVs) {g=true}), Tex=@n27)
  n45 = Constant() {r=0.5}
  n46 = Add(A=Divide(A=@n41, B=@n18), B=@n45) {constB=0.5}
  n47 = Divide(A=@n45, B=@n18)
  n48 = Constant() {r=1}
  n49 = MaterialFunctionCall(Float2=Multiply(A=Subtract(A=@n46, B=@n47), B=AppendVector(A=@n48, B=@n34))) {materialFunction=&BreakOutFloat2Components}
  n50 = SphereMask(A=@n45, B=@n49.R) {attenuationRadius=0.5}
  n51 = SphereMask(A=@n49.G, B=@n45) {attenuationRadius=0.5}
  n52 = Multiply(A=Multiply(A=@n42.0-1 mask, B=Multiply(A=@n50, B=@n51)), B=@n44.A)
  n53 = Multiply(A=@n44.RGB, B=@n52)
  n54 = LinearInterpolate(A=Max(A=@n38, B=@n53), B=@n53, Alpha=@n6)
  n55 = VectorParameter() {group=Numbers, parameterName=CC_OutlineColor, defaultValue=<rgba 0 0 0 1>}
  n56 = Divide(A=Add(A=Add(A=@n55.R, B=@n55.R), B=@n55.B)) {constB=3}
  n57 = ScalarParameter() {defaultValue=0.4, group=Numbers, parameterName=Fallback Number Color Brightness}
  n58 = VectorParameter() {group=Numbers, parameterName=CC_TritaryColor, defaultValue=<rgba 0.5312 0.0898 0 1>}
  n59 = Desaturation(None=OneMinus(None=@n55.RGB), Fraction=Constant() {r=0.9})
  n60 = LinearInterpolate(
   A=LinearInterpolate(A=@n55.RGB, B=If(A=@n56, B=@n57, "A > B"=@n58.RGB, "A == B"=@n59, "A < B"=@n59), Alpha=ComponentMask(None=@n54) {r=true})
   B=VectorParameter() {group=Numbers, parameterName=CC_SecondaryColor, defaultValue=<rgba 0 0.8497 1 1>}.RGB
   Alpha=ComponentMask(None=@n54) {b=true}
  )
  n61 = ScalarParameter() {defaultValue=1, group=Numbers, parameterName=Generate Custom Colors}
  n62 = GetMaterialAttributes(None=@n1) {attributeGetTypes [#1]}
  n63 = ScalarParameter() {group=Numbers, parameterName=Number_CustomMetalness}
  n64 = ScalarParameter() {group=Numbers, parameterName=Number_OverrideMetalness}
  n65 = Max(A=@n37, B=@n52)
  n66 = BlendMaterialAttributes(
   A=@n1
   B=SetMaterialAttributes!(
    MaterialAttributes=@n1
    BaseColor=LinearInterpolate(A=@n54, B=@n60, Alpha=@n61)
    Roughness=ScalarParameter() {defaultValue=0.6, group=Numbers, parameterName=Number_Roughness}
    Metallic=LinearInterpolate(A=@n62.Metallic, B=@n63, Alpha=@n64)
   ) {attributeSetTypes [#2, #3, #1]}
   Alpha=Multiply(A=LinearInterpolate(A=@n65, B=@n52, Alpha=@n6), B=FunctionInput() {inputName=Alpha, inputType=FunctionInput_Scalar})
  )
  n67 = SetMaterialAttributes(MaterialAttributes=@n66, BaseColor=@n11) {attributeSetTypes [#2, #3, #1]}
  FunctionOutput(None=StaticSwitchParameter(True=StaticSwitchParameter(True=@n67, False=@n66) {group=Numbers_Back, parameterName=Mask Visualization, sortPriority=0}, False=@n66) {group=Numbers, parameterName=use Front/Back Independent Control?})
  LinearInterpolate(A=Multiply(A=@n28.RGB, B=@n36), B=@n38, Alpha=@n61)
  LinearInterpolate(A=Multiply(A=@n44.RGB, B=@n36), B=@n53, Alpha=@n61)
 ]}
