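# Material function for weapon skins that recolors an input material's base color and metallic with four parameter sets.
# A mask texture's RGBA channels layer primary, secondary, tertiary and outline values.
# GMap, raised to a contrast power and squared, drives hue- and saturation-shifted shading in dark areas.
# The final blend and "All Colorized" output use a max of R, B and A, omitting G; an OverlayBalance branch is unconnected.
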MaterialFunction /Game/Weapons/WeaponSkins/MaterialFunction/CustomizeableColors
graph {nodes [
  n1 = GetMaterialAttributes(None=FunctionInput() {inputName=Material, inputType=FunctionInput_MaterialAttributes}) {attributeGetTypes [#1, #2, #3]}
  n2 = TextureSampleParameter2D() {group=CustomColors, parameterName=CustomColor_mask, samplerType=SAMPLERTYPE_Masks, texture=&T_BlackMask}
  n3 = Saturate(None=LinearInterpolate(
    A=@n1.BaseColor
    B=VectorParameter() {group=CustomColors, parameterName=CC_PrimaryColor, defaultValue=<rgba 0 0 0 1>}.RGB
    Alpha=Saturate(None=@n2.R)
   ))
  n4 = VectorParameter() {group=CustomColors, parameterName=CC_SecondaryColor, defaultValue=<rgba 0 0 0 1>}
  n5 = Saturate(None=@n2.G)
  n6 = ScalarParameter() {defaultValue=1, group=CustomColors, parameterName=UseAlphaFromTexture-CustomColor_mask}
  n7 = LinearInterpolate(
   A=LinearInterpolate(
    A=LinearInterpolate(A=@n3, B=@n4.RGB, Alpha=@n5)
    B=VectorParameter() {group=CustomColors, parameterName=CC_TritaryColor, defaultValue=<rgba 0 0 0 1>}.RGB
    Alpha=Saturate(None=@n2.B)
   )
   B=VectorParameter() {group=CustomColors, parameterName=CC_OutlineColor, defaultValue=<rgba 0 0 0 1>}.RGB
   Alpha=Saturate(None=Multiply(A=@n2.A, B=@n6) {constB=1.5})
  )
  n8 = Saturate(None=@n7)
  n9 = Saturate(None=@n8)
  n10 = MaterialFunctionCall(Hue Shift Percentage=ScalarParameter() {defaultValue=-0.02, group=CustomColors, parameterName=ShadowHueShift}, Texture=@n8) {materialFunction=&HueShift}
  n11 = ScalarParameter() {defaultValue=-0.2, group=CustomColors, parameterName=Shadow_Saturation}
  n12 = Max(A=Max(A=Max(A=@n2.R, B=@n2.R), B=@n2.B), B=@n2.A)
  n13 = Multiply(
   A=Power(
    Base=FunctionInput() {inputName=GMap, inputType=FunctionInput_Scalar}
    Exp=ScalarParameter() {defaultValue=1, group=CustomColors, parameterName=Gmap_Contrast}
   )
   B=@n12
  )
  n14 = Multiply(A=@n13, B=@n13)
  n15 = LinearInterpolate(A=@n9, B=LinearInterpolate(A=Desaturation(None=@n10.Result, Fraction=@n11), B=@n9, Alpha=Saturate(None=@n14)), Alpha=Saturate(None=@n12))
  n16 = LinearInterpolate(A=@n1.Metallic, B=ScalarParameter() {group=CustomColors, parameterName=CC_PrimaryMetallic}, Alpha=@n2.R)
  n17 = ScalarParameter() {group=CustomColors, parameterName=CC_SecondaryMetallic}
  n18 = LinearInterpolate(
   A=LinearInterpolate(A=LinearInterpolate(A=@n16, B=@n17, Alpha=@n2.G), B=ScalarParameter() {group=CustomColors, parameterName=CC_TritaryMetallic}, Alpha=@n2.B)
   B=ScalarParameter() {group=CustomColors, parameterName=CC_OutlineMetallic}
   Alpha=@n2.A
  )
  FunctionOutput(None=BlendMaterialAttributes(
    A=@n1.MaterialAttributes
    B=SetMaterialAttributes!(MaterialAttributes=@n1.MaterialAttributes, BaseColor=@n15, Metallic=@n18, Roughness=@n1.Roughness) {attributeSetTypes [#1, #2, #3]}
    Alpha=@n12
   ))
  FunctionOutput(None=@n12) {outputName=All Colorized}
  Multiply(
   A=Multiply(A=@n7, B=ScalarParameter() {defaultValue=1.5, group=CustomColors, parameterName=OverlayBalance}) {constB=1.5}
   B=Saturate(None=@n13)
  )
  Reroute()
 ]}
