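# Weapon skin material function that blends an animated, glowing disco floor look over incoming material attributes by an Alpha input.
# A static switch picks between two variants.
# The new one drives hue shift and a sphere mask over a tiled mask with time, changing radius when Emote_EnvTime is active.
# The other pulses and thresholds optionally falling tile masks, colored by gradient or hue shift, with derived HDR emissive and a cheaper low-quality path.
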MaterialFunction /Game/Weapons/WeaponSkins/MaterialFunction/DiscoFloorFunction
graph {nodes [
  n1 = FunctionInput() {bUsePreviewValueAsDefault=true, inputName=Material, inputType=FunctionInput_MaterialAttributes}
  n2 = GetMaterialAttributes(None=@n1) {attributeGetTypes [#1]}
  n3 = Multiply(A=Time(), B=ScalarParameter() {defaultValue=0.25, group=DiscoLights, parameterName=Disco Floor Tempo})
  n4 = MaterialFunctionCall(Value=@n3) {materialFunction=&LinearSine}
  n5 = Multiply(A=TextureCoordinate(), B=ScalarParameter() {defaultValue=8, group=DiscoLights, parameterName=DiscoLights_Tiling})
  n6 = TextureSampleParameter2D(UVs=@n5) {
   desc=Turn to Texture Parameter
   group=DiscoLights
   parameterName=NewDiscoEmissiveMask
   samplerType=SAMPLERTYPE_Masks
   texture=/Game/Global/Textures/Default/Tileable_Patterns/DiscoFloors_M
  }
  n7 = Saturate(None=If(
    A=CollectionParameter() {collection=&FortniteMaterialParameters, parameterName=Emote_EnvTime}
    B=Constant() {r=0.001}
    "A > B"=Constant() {r=1}
    "A < B"=Constant()
   ))
  n8 = Add(A=@n4."Rounded Linear Sine ", B=LinearInterpolate(A=@n6.R, B=@n6.R, Alpha=@n7))
  n9 = VectorParameter() {group=DiscoLights, parameterName=Default Hue, defaultValue=<rgba 1 0.0092 0 1>}
  n10 = LinearInterpolate(
   A=ScalarParameter() {defaultValue=0.2, group=DiscoLights, parameterName=Disco Sphere Radius}
   B=ScalarParameter() {defaultValue=0.6, group=DiscoLights, parameterName=Disco Sphere Radius Emote}
   Alpha=@n7
  )
  n11 = SphereMask(
   A=LinearInterpolate(A=@n6.R, B=@n6.R, Alpha=@n7)
   B=LinearInterpolate(A=@n10, Alpha=Saturate(None=@n4.Linear Sine)) {constB=0.99}
   Radius=@n10
   Hardness=Constant() {r=1}
  )
  n12 = Multiply(
   A=ScalarParameter() {defaultValue=15, group=DiscoLights, parameterName=DiscoLights_EmissiveBrightness}
   B=Multiply(
    A=MaterialFunctionCall(Hue Shift Percentage=@n8, Texture=@n9.RGB) {materialFunction=&HueShift}.Result
    B=Multiply(A=@n6.G, B=@n11)
   )
  )
  n13 = TextureSampleParameter2D(UVs=@n5) {
   group=DiscoLights
   parameterName=NewDiscoTilingNormal
   samplerType=SAMPLERTYPE_Normal
   texture=/Game/Global/Textures/Default/Tileable_Patterns/DiscoFloors_N
  }
  n14 = SetMaterialAttributes!(MaterialAttributes=@n2.MaterialAttributes, EmissiveColor=@n12, Normal=@n13.RGB) {attributeSetTypes [#2, #1]}
  n15 = VectorParameter() {group=DiscoLights, parameterName=DiscoLights_BaseColor, defaultValue=<rgba 0 0 0 1>}
  n16 = ScalarParameter() {group=DiscoLights, parameterName=DiscoLights_BaseColorEmissiveAdd}
  n17 = Multiply(A=Time(), B=ScalarParameter() {defaultValue=0.5, group=DiscoLights, parameterName=DiscoLights_Speed})
  n18 = ScalarParameter() {defaultValue=0.125, group=DiscoLights, parameterName=HueShiftSpeed}
  n19 = Multiply(A=TextureCoordinate(), B=ScalarParameter() {defaultValue=8, group=DiscoLights, parameterName=DiscoLights_Tiling})
  n20 = Multiply(
   A=Time()
   B=AppendVector(
    A=ScalarParameter() {group=DiscoLights, parameterName=FallSpeed_X}
    B=ScalarParameter() {defaultValue=2, group=DiscoLights, parameterName=FallSpeed_Y}
   )
  )
  n21 = Frac(None=LinearInterpolate(A=@n19, B=Add(A=@n19, B=@n20), Alpha=ScalarParameter() {group=DiscoLights, parameterName=isFalling}))
  n22 = TextureSampleParameter2D(UVs=@n21) {
   desc=Turn to Texture Parameter
   group=DiscoLights
   parameterName=DiscoEmissiveMask
   samplerType=SAMPLERTYPE_LinearColor
   texture=/Game/Global/Textures/Default/Tileable_Patterns/DiscoTiles_M
  }
  n23 = ScalarParameter() {desc=New, group=DiscoLights, parameterName=ThresholdOffset}
  n24 = Abs(None=Cosine(None=@n17))
  n25 = Abs(None=Cosine(None=Multiply(A=@n17) {constB=0.25}))
  n26 = Abs(None=MaterialFunctionCall(Value=Multiply(A=@n17) {constB=0.5}) {materialFunction=&LinearSine}.Linear Sine)
  n27 = Multiply(A=@n17, B=ScalarParameter() {defaultValue=0.25, group=DiscoLights, parameterName=AlternateFreq}) {constB=0.25}
  n28 = ScalarParameter() {defaultValue=10, group=DiscoLights, parameterName=PulsePower}
  n29 = Power(Base=Abs(None=Cosine(None=@n27)), Exp=@n28)
  n30 = LinearInterpolate(
   A=LinearInterpolate(
    A=Add(A=@n22.R, B=@n23) {constB=0.6, desc=New}
    B=Add(A=@n22.B, B=@n23) {constB=0.6, desc=New}
    Alpha=LinearInterpolate(A=@n24, B=@n25, Alpha=@n26)
   )
   B=LinearInterpolate(A=@n22.R, B=@n22.G, Alpha=Saturate(None=@n29))
   Alpha=ScalarParameter() {group=DiscoLights, parameterName=AlternateEmissiveMasks}
  )
  n31 = Add(A=Add(A=Multiply(A=@n17, B=@n18), B=@n22.R), B=@n30)
  n32 = TextureCoordinate() {desc=New}
  n33 = ComponentMask(None=VectorParameter() {desc=New, group=DiscoLights, parameterName=GradientMapUV_Tiling, defaultValue=<rgba 1 1 0 0>}.RGB) {desc=New, g=true, r=true}
  n34 = TextureSampleParameter2D(UVs=StaticSwitchParameter(True=Sine(None=@n31), False=Multiply(A=@n32, B=@n33) {desc=New}) {defaultValue=true, desc=New, group=DiscoLights, parameterName=UseGradientRandomColor}) {group=DiscoLights, parameterName=Gradient, texture=/Game/Global/Textures/Default/Masks_and_Gradients/WeaponGradients/SpiritGradient}
  n35 = StaticSwitchParameter(
   True=@n34.RGB
   False=MaterialFunctionCall(
    Hue Shift Percentage=@n31
    Texture=VectorParameter() {group=DiscoLights, parameterName=DiscoLights_CoreEmissive, defaultValue=<rgba 0 0.3167 1 1>}.RGB
   ) {materialFunction=&HueShift}.Result
  ) {group=DiscoLights, parameterName=UseGradientMapForColor}
  n36 = Multiply(A=@n17) {constB=0.125}
  n37 = Clamp(None=Abs(None=Frac(None=@n36)))
  n38 = Clamp(None=Power(Base=OneMinus(None=@n30), Exp=LinearInterpolate(Alpha=@n37) {constA=2, constB=4}))
  n39 = ScalarParameter() {group=DiscoLights, parameterName=DontInvertMask}
  n40 = Power(Base=@n24, Exp=@n28)
  n41 = Power(
   Base=LinearInterpolate(A=Power(Base=@n30), B=LinearInterpolate(A=@n38, B=@n30, Alpha=@n39), Alpha=Saturate(None=@n40))
   Exp=ScalarParameter() {defaultValue=3, group=DiscoLights, parameterName=Light Frequency}
  ) {constExponent=4}
  n42 = ScalarParameter() {defaultValue=0.05, group=DiscoLights, parameterName=DiscoThreshold_Min}
  n43 = Constant() {r=1}
  n44 = Constant()
  n45 = Multiply(A=@n35, B=Saturate(None=If(A=@n41, B=@n42, "A > B"=@n43, "A < B"=@n44)))
  n46 = MaterialFunctionCall(
   LDR Input=@n45
   HDR Tint=@n35
   HDR Intensity Multiplier=ScalarParameter() {defaultValue=8, group=DiscoLights, parameterName=HDRIntensity}
   Derive HDR Power=ScalarParameter() {defaultValue=1.5, group=DiscoLights, parameterName=HDRPower}
   Desaturate Luminance=ScalarParameter() {defaultValue=1, group=DiscoLights, parameterName=LuminanceDesat}
  ) {materialFunction=&DeriveHDRfromLDR}
  n47 = Multiply(A=@n45, B=ScalarParameter() {defaultValue=4, group=DiscoLights, parameterName=LowEndBrightness})
  n48 = Multiply(
   A=Add(
    A=LinearInterpolate(B=@n15.RGB, Alpha=@n16)
    B=QualitySwitch(Default=@n46.RemappedRange, Low=@n47, High=@n46.RemappedRange, Medium=@n46.RemappedRange)
   )
   B=ScalarParameter() {defaultValue=15, group=DiscoLights, parameterName=DiscoLights_EmissiveBrightness}
  )
  n49 = TextureSampleParameter2D(UVs=@n21) {
   group=DiscoLights
   parameterName=DiscoTilingNormal
   samplerType=SAMPLERTYPE_Normal
   texture=/Game/Global/Textures/Default/Tileable_Patterns/DiscoTiles_N
  }
  n50 = LinearInterpolate(
   A=Constant3Vector() {constant=<rgba 0.5 0.5 1 0>}
   B=@n49.RGB
   Alpha=ScalarParameter() {defaultValue=1, group=DiscoLights, parameterName=NormalIntensity}
  )
  n51 = StaticSwitchParameter(
   True=MaterialFunctionCall!(Material=@n2.MaterialAttributes, Emissive=@n48) {desc=New, materialFunction=/Engine/Functions/MaterialLayerFunctions/MatLayerBlend_Emissive}.Blended Material
   False=MakeMaterialAttributes!(
    BaseColor=@n15.RGB
    Metallic
    Specular=Constant() {r=0.8}
    Roughness=ScalarParameter() {defaultValue=0.1, group=DiscoLights, parameterName=DiscoLights_Roughness}
    Anisotropy
    EmissiveColor=@n48
    Opacity
    OpacityMask
    Normal=MaterialFunctionCall(BaseNormal=@n2.Normal, AdditionalNormal=@n50) {materialFunction=&BlendAngleCorrectedNormals}.Result
    Tangent
    WorldPositionOffset
    SubsurfaceColor
    ClearCoat
    ClearCoatRoughness
    AmbientOcclusion
    Refraction
    CustomizedUVs_0
    CustomizedUVs_1
    CustomizedUVs_2
    CustomizedUVs_3
    CustomizedUVs_4
    CustomizedUVs_5
    CustomizedUVs_6
    CustomizedUVs_7
    PixelDepthOffset
    ShadingModel
    Displacement
   )
  ) {desc=New, group=DiscoLights, parameterName=Use Original Material Attributes}
  n52 = FunctionInput() {bUsePreviewValueAsDefault=true, inputName=Alpha, inputType=FunctionInput_Scalar}
  FunctionOutput(None=BlendMaterialAttributes(
    A=@n1
    B=StaticSwitchParameter(True=@n14, False=@n51) {desc=New, group=DiscoLights, parameterName=Use New Disco Floor}
    Alpha=Saturate(None=@n52)
   ))
  n53 = Multiply(A=@n3) {constB=4}
  Multiply(A=@n4.Linear Sine, B=ConstantBiasScale(None=Sine(None=@n53)) {scale=0.1})
  Multiply(B=@n12)
 ]}
