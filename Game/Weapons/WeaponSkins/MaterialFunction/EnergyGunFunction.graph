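# Weapon skin material function that adds a glowing energy effect around the muzzle, toggled by the Use Energy static switch.
# A scaled muzzle sphere mask drives a HeightLerp, using inverted, remapped base color as the height texture to break up the glow.
# That alpha blends falloff and core colors into powered emissive over black base color and dimmed specular.
# The sphere mask blends this over the input attributes, keeping their vertex attributes.
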MaterialFunction /Game/Weapons/WeaponSkins/MaterialFunction/EnergyGunFunction
graph {nodes [
  n1 = FunctionInput() {bUsePreviewValueAsDefault=true, inputName=Base MA, inputType=FunctionInput_MaterialAttributes}
  n2 = GetMaterialAttributes(None=@n1) {attributeGetTypes [#1, #2]}
  n3 = MaterialFunctionCall!(
   MuzzleFalloff=Multiply(
    A=ScalarParameter() {defaultValue=50, group=Weapon_Muzzle, parameterName=MuzzleFalloff}
    B=ScalarParameter() {defaultValue=8, group=Weapon_Elemental_Energy, parameterName=MuzzleFalloffScale}
   ) {constB=1.5}
   Pre-Skinned Local Position=FunctionInput() {bUsePreviewValueAsDefault=true, inputName=Pre-Skinned Local Position, sortPriority=41}
  ) {materialFunction=/Game/Packages/Fortress_SharedMaterials/Base_Material_Functions/BaseCustomizationFunctions/MF_BaseWeapon_MuzzleSphereMask}
  n4 = MaterialFunctionCall(
   Transition Phase=Multiply(A=@n3.Muzzle_SphereMask, B=ScalarParameter() {defaultValue=0.9, group=Weapon_Elemental_Energy, parameterName=TransitionMultiplier}) {desc=sphere mask}
   Height Texture=LinearInterpolate(A=Constant3Vector() {constant=<rgba 1 1 1 1>}, B=Constant3Vector() {constant=<rgba 0.5 0.5 0.5 1>}, Alpha=@n2.BaseColor)
  ) {materialFunction=&HeightLerp}
  n5 = VectorParameter() {group=Weapon_Elemental_Energy, parameterName=Energy Falloff Color, defaultValue=<rgba 0.0089 0.086 0.445 1>}
  n6 = VectorParameter() {group=Weapon_Elemental_Energy, parameterName=Energy Core Color, defaultValue=<rgba 0.0442 0.2898 0.9 1>}
  n7 = Multiply(A=@n4.Alpha, B=LinearInterpolate(A=@n5.RGB, B=@n6.RGB, Alpha=@n4.Alpha))
  n8 = ScalarParameter() {defaultValue=1.5, group=Weapon_Elemental_Energy, parameterName=Final Power}
  n9 = SetMaterialAttributes!(
   MaterialAttributes=@n2.MaterialAttributes
   BaseColor=Constant()
   Specular=Multiply(A=@n2.Specular) {constB=0.2}
   EmissiveColor=Multiply(A=@n4.Alpha, B=Power(Base=@n7, Exp=@n8))
   Opacity=Constant() {r=1}
  ) {attributeSetTypes [
    #1
    #2
    #3
    #4
   ]}
  FunctionOutput(None=StaticSwitch(
    True=BlendMaterialAttributes(A=@n1, B=@n9, Alpha=@n3.Muzzle_SphereMask) {vertexAttributeBlendType=UseA}
    False=@n1
    Value=StaticBoolParameter() {group=Weapon_Elemental, parameterName=Use Energy}
   ))
 ]}
