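# Material function that layers a frosted ice look around a weapon's muzzle, toggled by the "Use Ice" static switch.
# A muzzle sphere mask drives a HeightLerp against a triplanar snowflake pattern, producing the ice coverage and the masked blend of snowflake normals into the mesh normal.
# A Fresnel on that normal samples an ice gradient for base color, with an added frost tint, and modulates opacity when blending the glossy, non-metallic ice over the incoming material attributes.
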MaterialFunction /Game/Weapons/WeaponSkins/MaterialFunction/IceFunction
graph {nodes [
  n1 = Constant3Vector() {constant=<rgba 0 0 1 0>}
  n2 = FunctionInput() {bUsePreviewValueAsDefault=true, inputName=Pre-Skinned Local Position, sortPriority=41}
  n3 = FunctionInput() {bUsePreviewValueAsDefault=true, inputType=FunctionInput_MaterialAttributes}
  n4 = GetMaterialAttributes(None=@n3) {attributeGetTypes [#1, #2]}
  n5 = MaterialFunctionCall!(
   Material Normals=@n4.Normal
   TextureObject_MASK
   TextureObject_Diffuse=TextureObject() {samplerType=SAMPLERTYPE_Masks, texture=/Game/Global/Textures/Default/Tileable_Patterns/SnowFlakeTest}
   TextureObject_Normal=TextureObject() {samplerType=SAMPLERTYPE_Normal, texture=/Game/Global/Textures/Default/Tileable_Patterns/SnowFlakeTest_normal}
   TextureScale=Constant() {desc=Ice Tiling Size, r=14}
   Normal Texture Intensity=Constant() {r=0.6}
   Pre-Skinned Local Position=@n2
   Pre-Skinned Local Normal=FunctionInput() {bUsePreviewValueAsDefault=true, inputName=Pre-Skinned Local Normal, sortPriority=42}
   Rotation Z Axis
   Rotation X Axis
   Triplanner_Offset
   isMech
   Export Float 4
   ProjectionTransitionContrast
   IsStatsB
   isCharacter
  ) {materialFunction=/Game/Packages/Fortress_SharedMaterials/Base_Material_Functions/BaseCustomizationFunctions/MF_Base_Customization_Triplanar}
  n6 = ComponentMask(None=@n5.Diffuse Texture) {r=true}
  n7 = MaterialFunctionCall!(
   MuzzleFalloff=Multiply(
    A=ScalarParameter() {defaultValue=1.25, group=Weapon_Elemental_Ice, parameterName=MuzzleFalloffScale}
    B=ScalarParameter() {defaultValue=50, group=Weapon_Muzzle, parameterName=MuzzleFalloff}
   )
   Pre-Skinned Local Position=@n2
  ) {materialFunction=/Game/Packages/Fortress_SharedMaterials/Base_Material_Functions/BaseCustomizationFunctions/MF_BaseWeapon_MuzzleSphereMask}
  n8 = MaterialFunctionCall(Transition Phase=@n7.Muzzle_SphereMask, Height Texture=Saturate(None=OneMinus(None=@n6)), Contrast=Constant() {r=0.18}) {materialFunction=&HeightLerp}
  n9 = Saturate(None=Multiply(A=LinearInterpolate(Alpha=@n8.Alpha) {constAlpha=0, constB=5}, B=Constant() {r=1.5}) {constB=2})
  n10 = Constant() {r=-5.5}
  n11 = Constant() {r=2}
  n12 = Add(A=@n8.Lerp Alpha No Contrast, B=Saturate(None=LinearInterpolate(A=@n10, B=@n11, Alpha=@n7.Muzzle_SphereMask)))
  n13 = Constant() {r=0.8}
  n14 = Multiply(A=@n9, B=OneMinus(None=Clamp(None=@n12, Max=@n13)))
  n15 = Normalize(VectorInput=LinearInterpolate(A=@n4.Normal, B=LinearInterpolate(A=@n1, B=@n5.Normal Texture, Alpha=@n8.Lerp Alpha No Contrast), Alpha=Saturate(None=@n14)))
  n16 = Constant() {
   desc=Tiling Normal Influence on Elemental Ice Color. Lerp that controls the blend between the tiling ice normal and the mesh normal texture. Alters normal influence on Ice base/emissive color Fresnel.
   r=2.5
  }
  n17 = Fresnel(ExponentIn=Constant() {desc=.8, r=0.8}, Normal=Transform(None=LinearInterpolate(A=@n1, B=@n15, Alpha=@n16))) {baseReflectFraction=0, exponent=2.9525}
  n18 = Saturate(None=@n17)
  n19 = LinearInterpolate(A=Constant() {r=-1}, B=Constant3Vector() {constant=<rgba 0.2664 0.43 0.55 1>}, Alpha=Saturate(None=@n12))
  n20 = SetMaterialAttributes!(
   MaterialAttributes=@n4.MaterialAttributes
   BaseColor=Add(
    A=TextureSample(UVs=@n18) {texture=/Game/Global/Textures/Default/Masks_and_Gradients/WeaponGradients/T_ElementalWeapon_Ice_Gradient}.RGB
    B=Saturate(None=@n19)
   )
   Metallic=Constant()
   Specular=Constant() {r=1}
   Roughness=Constant() {r=0.25}
   Normal=@n15
  ) {attributeSetTypes [
    #3
    #4
    #1
    #5
    #2
   ]}
  n21 = LinearInterpolate(A=Constant() {desc=Ice opacity, r=-0.6}, B=Constant() {r=1}, Alpha=@n17)
  n22 = LinearInterpolate(B=Saturate(None=Add(A=@n9, B=@n21) {constB=0.1}), Alpha=@n9)
  FunctionOutput(None=StaticSwitch(
    True=BlendMaterialAttributes(A=@n3, B=@n20, Alpha=@n22) {vertexAttributeBlendType=UseA}
    False=@n3
    Value=StaticBoolParameter() {group=Weapon_Elemental, parameterName=Use Ice}
   ))
 ]}
