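# Weapon skin material function that adds animated lightning emissive to input material attributes when the Use Lightning static switch is on.
# A time- and Fresnel-scrolled digital noise glow, masked near the muzzle, combines with gradient-texture bolts.
# Ceil-stepped noise jitters bolt UVs, while looping sphere masks and HeightLerp reveal them.
# The player electrical color tints the result, which is raised to the fourth power and scaled by Specular.
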MaterialFunction /Game/Weapons/WeaponSkins/MaterialFunction/LightningFunction
graph {nodes [
  n1 = FunctionInput() {bUsePreviewValueAsDefault=true, inputName=Base MA, inputType=FunctionInput_MaterialAttributes}
  n2 = Time()
  n3 = ScalarParameter() {defaultValue=-0.1062, group=Weapon_Elemental_Lightning, parameterName=Time}
  n4 = MaterialFunctionCall() {materialFunction=/Game/Packages/Fortress_SharedMaterials/Base_Material_Functions/BaseCustomizationFunctions/MF_BaseWeapon_RandomFloat}
  n5 = GetMaterialAttributes(None=@n1) {attributeGetTypes [#1, #2]}
  n6 = Transform(None=@n5.Normal)
  n7 = MaterialFunctionCall(Float3=CameraVectorWS()) {materialFunction=&BreakOutFloat3Components}
  n8 = Constant() {r=-1}
  n9 = Constant() {r=15}
  n10 = AppendVector(A=AppendVector(A=@n7.R, B=@n7.R), B=Add(A=Add(A=@n7.G, B=@n8), B=Multiply(A=@n7.B, B=@n9)))
  n11 = Add(
   A=Add(A=Multiply(A=@n2, B=@n3) {constB=0.05}, B=@n4.randomFloat)
   B=Multiply(
    A=MaterialFunctionCall(Normal Vector=@n6, Camera Vector=@n10) {materialFunction=&Fresnel_Function}.Result
    B=Constant() {r=0.002}
   ) {constB=0.001}
  )
  n12 = Constant2Vector() {g=0.0697, r=1.0667}
  n13 = ConstantBiasScale(None=TextureSample(UVs=Multiply(A=@n11, B=@n12)) {texture=/Game/Global/Textures/Default/Noise/NoiseDigital_large}.R) {bias=0.5962, scale=2.5}
  n14 = MaterialFunctionCall!(
   MuzzleFalloff
   Pre-Skinned Local Position=FunctionInput() {bUsePreviewValueAsDefault=true, inputName=Pre-Skinned Local Position, sortPriority=41}
  ) {materialFunction=/Game/Packages/Fortress_SharedMaterials/Base_Material_Functions/BaseCustomizationFunctions/MF_BaseWeapon_MuzzleSphereMask}
  n15 = Add(A=@n4.randomFloat, B=Time())
  n16 = Multiply(A=@n15) {constB=16}
  n17 = Panner(Coordinate=TextureCoordinate() {uTiling=0.46, vTiling=0.2476}, Time=Divide(A=Ceil(None=@n16)) {constB=4}) {speedX=0.4, speedY=0.4}
  n18 = Subtract(A=ComponentMask(None=TextureSample(UVs=@n17) {samplerType=SAMPLERTYPE_LinearColor, texture=/Game/Global/Textures/Default/Noise/TilingNoise06_N_uncompressed}.RGB) {g=true, r=true}) {constB=0.5}
  n19 = TextureCoordinate() {uTiling=2.4539, vTiling=2.8057}
  n20 = TextureSample(UVs=Add(A=Multiply(A=@n18) {constB=0.1}, B=Panner(Coordinate=@n19, Time=@n15) {speedX=-0.2301, speedY=0.0354}) {constB=0}) {samplerType=SAMPLERTYPE_LinearColor, texture=/Game/Global/Textures/Default/Masks_and_Gradients/TestLightningGradient}
  n21 = MaterialFunctionCall(In=@n14.Muzzle_SphereMask, Contrast=Constant() {r=1.4586}) {materialFunction=&CheapContrast}
  n22 = Frac(None=Add(A=@n15, B=@n20.B))
  n23 = SphereMask(A=@n20.G, B=Multiply(A=Subtract(A=@n22) {constB=0.2}) {constB=10}) {attenuationRadius=1, desc="! 50 perc", hardnessPercent=57.9646}
  n24 = Clamp(None=Multiply(A=Multiply(A=@n21.Result, B=@n23)) {constB=20})
  n25 = Multiply(
   A=@n20.R
   B=MaterialFunctionCall(Transition Phase=OneMinus(None=@n24), Height Texture=OneMinus(None=@n20.A)) {materialFunction=&HeightLerp}.Results
  )
  n26 = CollectionParameter() {collection=&FortniteMaterialParameters, parameterName=PlayerElectricalColor}
  n27 = Multiply(
   A=ComponentMask(None=@n26) {b=true, g=true, r=true}
   B=ScalarParameter() {defaultValue=5.2296, group=Weapon_Elemental_Lightning, parameterName=Lightning Bolt intensity}
  ) {constB=4.2596}
  n28 = CollectionParameter() {collection=&FortniteMaterialParameters, parameterName=PlayerElectricalColor}
  n29 = Multiply(
   A=Add(A=Multiply(A=@n13, B=@n14.Muzzle_SphereMask), B=Multiply(A=@n25, B=@n27)) {constB=0}
   B=ComponentMask(None=@n28) {b=true, g=true, r=true}
  )
  n30 = Multiply(A=@n29, B=@n29)
  n31 = Multiply(A=@n5.Specular, B=Multiply(A=@n30, B=@n30))
  FunctionOutput(None=StaticSwitch(
    True=MaterialFunctionCall!(Material=@n1, Emissive=@n31) {materialFunction=/Engine/Functions/MaterialLayerFunctions/MatLayerBlend_Emissive}.Blended Material
    False=@n1
    Value=StaticBoolParameter() {group=Weapon_Elemental, parameterName=Use Lightning}
   ))
  n32 = MaterialFunctionCall(Float3=CameraVectorWS()) {materialFunction=&BreakOutFloat3Components}
  n33 = Multiply(A=@n32.R, B=Constant() {r=1.5})
  n34 = Constant() {r=1}
  MaterialFunctionCall(Normal Vector=Transform(None=@n5.Normal), Camera Vector=AppendVector(A=AppendVector(A=@n33, B=@n32.G), B=Add(A=@n32.B, B=@n34))) {materialFunction=&Fresnel_Function}
 ]}
