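# Material function that layers elemental weapon skin effects onto an input set of material attributes.
# It chains the Ice, Lightning, FireGun and EnergyGun functions in that order, each passing its attributes to the next.
# All four share a pre-skinned local position input, while only Ice also takes the local normal.
# Both inputs default to pixel shader outputs of MF_Base_LocalPosition.
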MaterialFunction /Game/Weapons/WeaponSkins/MaterialFunction/ElementalWeaponsFunction
graph {nodes [
  n1 = MaterialFunctionCall() {materialFunction=/Game/Packages/Fortress_SharedMaterials/Base_Material_Functions/BaseCustomizationFunctions/MF_Base_LocalPosition}
  n2 = FunctionInput(Preview=@n1.Pre-Skinned Local Position Pixel Shader) {bUsePreviewValueAsDefault=true, inputName=Pre-Skinned Local Position, sortPriority=41}
  n3 = MaterialFunctionCall!(
   Base MA=MaterialFunctionCall!(
    In=FunctionInput() {bUsePreviewValueAsDefault=true, inputType=FunctionInput_MaterialAttributes}
    Pre-Skinned Local Position=@n2
    Pre-Skinned Local Normal=FunctionInput(Preview=@n1.Pre-Skinned Local Normal Pixel Shader) {bUsePreviewValueAsDefault=true, inputName=Pre-Skinned Local Normal, sortPriority=42}
   ) {materialFunction=&/IceFunction}.Result
   Pre-Skinned Local Position=@n2
  ) {materialFunction=&/LightningFunction}
  FunctionOutput(None=MaterialFunctionCall!(
    Base MA=MaterialFunctionCall!(Base MA=@n3.Result, Pre-Skinned Local Position=@n2) {materialFunction=&/FireGunFunction}.Result
    Pre-Skinned Local Position=@n2
   ) {materialFunction=&/EnergyGunFunction}.Result)
 ]}
