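# Material function for weapon skins that layers a generic glowing, clear-coated gem look onto an input material.
# A tintable glow blends edge and core brightness by an invertible Fresnel, adds a sine-shaped Fresnel band, then adds onto the input emissive.
# The gem attributes blend over the input by MASK.
# A static switch optionally multiplies MASK by a sphere mask comparing the Color Change Sequence input to GemColorTransition, gating where the gem appears.
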MaterialFunction /Game/Weapons/WeaponSkins/MaterialFunction/GenericGem
graph {nodes [
  n1 = FunctionInput() {bUsePreviewValueAsDefault=true, inputName=Material, inputType=FunctionInput_MaterialAttributes}
  n2 = GetMaterialAttributes(None=@n1) {attributeGetTypes [
    #1
    #2
    #3
    #4
    #5
    #6
   ]}
  n3 = VectorParameter() {group=Generic Gem, parameterName=Gem Slight Glow Color, defaultValue=<rgba 1 1 1 1>}
  n4 = Multiply(A=@n3.RGB, B=ScalarParameter() {defaultValue=0.1, group=Generic Gem, parameterName=Edge Brightness})
  n5 = Multiply(A=@n3.RGB, B=ScalarParameter() {defaultValue=2, group=Generic Gem, parameterName=Core Brightness})
  n6 = Fresnel(ExponentIn=ScalarParameter() {defaultValue=1, group=Generic Gem, parameterName=Gem Fresnel EX}) {baseReflectFraction=0}
  n7 = Saturate(None=LinearInterpolate(A=@n6, B=OneMinus(None=@n6), Alpha=ScalarParameter() {group=Generic Gem, parameterName=Invert Gem Fresnel, sortPriority=50}))
  n8 = Abs(None=ConstantBiasScale(None=Sine(None=@n6)) {bias=0.1, scale=2})
  n9 = Add(A=@n2.EmissiveColor, B=Add(A=LinearInterpolate(A=@n4, B=@n5, Alpha=@n7), B=Multiply(A=@n3.RGB, B=@n8)))
  n10 = Constant() {r=1}
  n11 = Constant() {r=0.1}
  n12 = FunctionInput() {inputName=MASK, inputType=FunctionInput_Scalar}
  n13 = Multiply(
   A=@n12
   B=SphereMask(
    A=FunctionInput() {bUsePreviewValueAsDefault=true, inputName=Color Change Sequence, inputType=FunctionInput_Scalar}
    B=ScalarParameter() {parameterName=GemColorTransition}
    Radius=ScalarParameter() {defaultValue=0.2, parameterName=SPhereRadius_transition}
    Hardness=Constant() {r=1}
   )
  )
  FunctionOutput(None=BlendMaterialAttributes(
    A=@n1
    B=MakeMaterialAttributes!(
     BaseColor=@n2.BaseColor
     Metallic=@n2.BaseColor
     Specular=@n2.BaseColor
     Roughness=@n2.BaseColor
     Anisotropy
     EmissiveColor=@n9
     Opacity
     OpacityMask
     Normal=@n2.BaseColor
     Tangent
     WorldPositionOffset
     SubsurfaceColor
     ClearCoat=@n10
     ClearCoatRoughness=@n11
     AmbientOcclusion
     Refraction
     CustomizedUVs_0
     CustomizedUVs_1
     CustomizedUVs_2
     CustomizedUVs_3
     CustomizedUVs_4
     CustomizedUVs_5
     CustomizedUVs_6
     CustomizedUVs_7
     PixelDepthOffset
     ShadingModel
     Displacement
    )
    Alpha=StaticSwitchParameter(True=@n13, False=@n12) {parameterName=Use Color Change Sequence}
   ) {vertexAttributeBlendType=UseA})
  PixelNormalWS()
  Transform(None=@n2.Normal)
  VertexNormalWS()
 ]}
