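# HopRockGem is a weapon-skin material function that blends a glossy, metallic gem look over incoming material attributes by an Alpha input.
# Base color adds Fresnel-driven thin-film iridescence, while the incoming normal passes through.
# Emissive combines a pulsing edge glow, limited by an actor-relative sphere mask, with parallax HopRock-mask glow tinted by panning smoke and pulsing color, remapped to HDR.
# A rotating-normal static switch is present but unconnected.
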MaterialFunction /Game/Weapons/WeaponSkins/MaterialFunction/HopRockGem
graph {nodes [
  n1 = FunctionInput() {inputName=Material, inputType=FunctionInput_MaterialAttributes}
  n2 = VectorParameter() {parameterName=Color_1, defaultValue=<rgba 0.0845 0.0877 0.1 1>}
  n3 = TextureSampleParameter2D(UVs=Fresnel(ExponentIn=ScalarParameter() {defaultValue=0.571, parameterName=EX_Fres_Thinfilm_HopRock}) {baseReflectFraction=0, exponent=0.5}) {group=ThinFilm, parameterName=Thin Film Texture_1, texture=/Game/Global/Textures/Default/Thinfilm/T_ThinFilm_Spectrum_COLOR}
  n4 = ScalarParameter() {defaultValue=0.8, group=ThinFilm, parameterName=ThinFilm_Intensity_1}
  n5 = GetMaterialAttributes(None=@n1) {attributeGetTypes [#1]}
  n6 = LinearInterpolate(
   A=@n2.RGB
   B=Add(A=Multiply(A=@n3.RGB, B=@n4), B=@n2.RGB)
   Alpha=Fresnel(ExponentIn=ScalarParameter() {defaultValue=5.8257, group=ThinFilm, parameterName=ThinFilm_Visibility}, Normal=@n5.Normal) {baseReflectFraction=0}
  )
  n7 = Constant() {r=0.8}
  n8 = Constant() {r=0.02}
  n9 = VectorParameter() {parameterName=EdgeGlow, defaultValue=<rgba 10 10 10 1>}
  n10 = ComponentMask(None=@n5.Normal) {g=true}
  n11 = Cosine(None=Multiply(A=Time(), B=ScalarParameter() {defaultValue=0.846, parameterName=EdgePulseSpeed})) {period=10}
  n12 = WorldPosition()
  n13 = ActorPositionWS()
  n14 = SphereMask(
   A=Multiply(A=Subtract(A=@n12.XYZ, B=@n13), B=VectorParameter() {parameterName=SpheremaskScaleMult_1, defaultValue=<rgba 1 1 1 1>}.RGB)
   B=VectorParameter() {parameterName=SphereMaskOffset_HopRock, defaultValue=<rgba 17.3159 257.2161 0 1>}.RGB
   Radius=ScalarParameter() {defaultValue=246.8589, parameterName=GlowWorldRadius_1}
   Hardness=ScalarParameter() {defaultValue=-2, parameterName=Hardness}
  ) {attenuationRadius=11, hardnessPercent=0.1}
  n15 = Multiply(
   A=@n10
   B=TextureSample(UVs=Multiply(A=@n11, B=@n14)) {samplerType=SAMPLERTYPE_LinearColor, texture=/Game/Global/Textures/Default/Masks_and_Gradients/T_VertGradient}.R
  )
  n16 = TextureObjectParameter() {
   parameterName=HopRock Mask
   samplerType=SAMPLERTYPE_Masks
   texture=/Game/Weapons/FORT_TRAPS/Trap_Jump_FullTile/Textures/T_BouncePad_MI_Floor_Jump_FullTile_FX
  }
  n17 = TextureSample(Tex=@n16) {samplerType=SAMPLERTYPE_Masks, texture=&T_BouncePad_MI_Floor_Jump_FullTile_FX}
  n18 = ScalarParameter() {defaultValue=0.2, parameterName=sdfgsdfg}
  n19 = Multiply(A=@n18) {constB=4}
  n20 = Time()
  n21 = Clamp(None=Abs(None=Sine(None=@n20) {period=13}))
  n22 = Multiply(A=@n14, B=Fresnel(ExponentIn=LinearInterpolate(A=@n18, B=@n19, Alpha=@n21)) {baseReflectFraction=0})
  n23 = TextureSample(UVs=BumpOffset(Coordinate=TextureCoordinate(), Height=ScalarParameter() {defaultValue=-0.2, parameterName=Glow Bump}), Tex=@n16) {samplerType=SAMPLERTYPE_Masks, texture=/Game/Weapons/FORT_TRAPS/BouncePad/Textures/T_BouncePad_M_BouncePad_FX}
  n24 = Divide(A=@n10)
  n25 = TextureCoordinate() {uTiling=5, vTiling=5}
  n26 = ScalarParameter() {defaultValue=-2, parameterName=BumpOffset_1}
  n27 = TextureSample(UVs=Panner(Coordinate=BumpOffset(Coordinate=@n25, Height=@n26)) {speedX=-0.25, speedY=-0.16}) {texture=&T_Inky_Smoke_Tile_Chromatic}
  n28 = VectorParameter() {parameterName=GlowColor_1, defaultValue=<rgba 0.6783 1.4419 5 0>}
  n29 = Time()
  n30 = Clamp(None=Abs(None=Sine(None=@n29) {period=10}))
  n31 = Multiply(
   A=Multiply(A=Subtract(A=@n23.G, B=@n24), B=ScalarParameter() {defaultValue=0.05, parameterName=Glow Intensity})
   B=Add(A=Multiply(A=@n27.RGB) {constB=4}, B=Multiply(A=@n28.RGB, B=@n30))
  )
  n32 = Add(
   A=Multiply(A=Multiply(A=@n9.RGB, B=@n15), B=Multiply(A=@n17.B, B=@n22))
   B=MaterialFunctionCall(
    LDR Input=Clamp(None=@n31)
    HDR Tint=@n31
    HDR Intensity Multiplier=ScalarParameter() {defaultValue=4.9089, parameterName=HDRIntensityMult_1}
   ) {materialFunction=&DeriveHDRfromLDR}.RemappedRange
  )
  FunctionOutput(None=BlendMaterialAttributes(
    A=@n1
    B=MakeMaterialAttributes!(
     BaseColor=@n6
     Metallic=@n7
     Specular=@n7
     Roughness=@n8
     Anisotropy
     EmissiveColor=@n32
     Opacity
     OpacityMask
     Normal=@n5.Normal
     Tangent
     WorldPositionOffset
     SubsurfaceColor
     ClearCoat
     ClearCoatRoughness
     AmbientOcclusion
     Refraction
     CustomizedUVs_0
     CustomizedUVs_1
     CustomizedUVs_2
     CustomizedUVs_3
     CustomizedUVs_4
     CustomizedUVs_5
     CustomizedUVs_6
     CustomizedUVs_7
     PixelDepthOffset
     ShadingModel
     Displacement
    )
    Alpha=FunctionInput() {inputName=Alpha, inputType=FunctionInput_Scalar}
   ))
  n33 = Multiply(A=ScalarParameter() {defaultValue=1, parameterName=Random}) {constB=0.2}
  n34 = Time()
  StaticSwitchParameter(
   True=MaterialFunctionCall(Rotation Axis=Constant3Vector() {constant=<rgba 0 0 1 0>}, Rotation Angle=Multiply(A=@n34, B=@n33) {constB=0.1}) {materialFunction=&FixRotateAboutAxisNormals}.New World Space Vertex Normal
   False=PixelNormalWS()
  ) {defaultValue=true, parameterName=UseWPO_1}
 ]}
