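# Material function for animated weapon-skin hair, adding world position offset to incoming material attributes behind the UseGlowyAnimatedHair static switch.
# Panning Perlin noise, projected on three pre-skinned position planes and blended by normal, drives displacement that shifts from an ADS amount to a standard amount with camera distance.
# An influence mask from normalized bounds position or a gradient texture scales the offset.
# Fresnel glow, sine-wave and hair AO branches exist but do not reach the output.
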MaterialFunction /Game/Weapons/WeaponSkins/MaterialFunction/MF_AnimatedHair
graph {nodes [
  n1 = ObjectPositionWS()
  n2 = WorldPosition()
  n3 = TextureSampleParameter2D() {parameterName=HairInfluenceGradient, texture=&T_BlacksRGB}
  n4 = Divide(A=PreSkinnedPosition(), B=PreSkinnedLocalBounds().Max)
  n5 = StaticSwitchParameter(True=@n3.RGB, False=OneMinus(None=@n4)) {parameterName=UseInfluenceGradientTexture}
  n6 = ChannelMaskParameter(None=@n5) {group=AnimatedHair, maskChannel=Blue, parameterName=AH_GradientChannel, defaultValue=<rgba 0 0 1 0>}
  n7 = Multiply(
   A=Add(A=@n6.R, B=ScalarParameter() {defaultValue=-0.05, group=AnimatedHair, parameterName=AH_GradientOffset})
   B=ScalarParameter() {defaultValue=1.25, group=AnimatedHair, parameterName=AH_GradientMultiplier}
  ) {constB=2}
  n8 = PreSkinnedPosition()
  n9 = Divide(A=@n8, B=VectorParameter() {group=AnimatedHair, parameterName=AH_WPONoiseTiling, defaultValue=<rgba 12 12 12 1>}.RGB)
  n10 = VectorParameter() {group=AnimatedHair, parameterName=AH_WPONoisePanSpeedAndDirection, defaultValue=<rgba 0.5 -0.5 0 1>}
  n11 = Panner(Coordinate=ComponentMask(None=@n9) {g=true, r=true}, Speed=ComponentMask(None=@n10.RGB) {g=true, r=true})
  n12 = TextureObjectParameter() {parameterName=DistorntionNoise, texture=/Game/Global/Textures/Default/Noise/3xPerlin_M}
  n13 = Panner(Coordinate=ComponentMask(None=@n9) {b=true, r=true}, Speed=ComponentMask(None=@n10.RGB) {b=true, r=true})
  n14 = PreSkinnedNormal()
  n15 = Abs(None=@n14)
  n16 = ComponentMask(None=@n9) {b=true, g=true}
  n17 = ComponentMask(None=@n10.RGB) {b=true, g=true}
  n18 = LinearInterpolate(
   A=LinearInterpolate(A=TextureSample(UVs=@n11, Tex=@n12).RGB, B=TextureSample(UVs=@n13, Tex=@n12).RGB, Alpha=ComponentMask(None=@n15) {g=true})
   B=TextureSample(UVs=Panner(Coordinate=@n16, Speed=@n17), Tex=@n12).RGB
   Alpha=ComponentMask(None=@n15) {r=true}
  )
  n19 = Divide(
   A=Distance(A=CameraPositionWS(), B=WorldPosition().XYZ)
   B=ScalarParameter() {defaultValue=256, group=AnimatedHair, parameterName=AH_DistanceBias}
  )
  n20 = Multiply(
   A=@n18
   B=LinearInterpolate(
    A=VectorParameter() {group=AnimatedHair, parameterName=AH_ADS_NoiseDisplacement_Amt, defaultValue=<rgba 0 2 0 1>}.RGB
    B=VectorParameter() {group=AnimatedHair, parameterName=AH_NoiseDisplacement_Amt, defaultValue=<rgba 0 4 0 1>}.RGB
    Alpha=Clamp(None=@n19)
   )
  )
  n21 = Multiply(A=Clamp(None=Power(Base=@n7)), B=@n20)
  Add(A=Normalize(VectorInput=Subtract(A=@n1, B=@n2.XYZ)), B=@n21)
  n22 = VectorParameter() {group=AnimatedHair, parameterName=AH_PanSpeed, defaultValue=<rgba 0 0 -0.2 1>}
  n23 = Time()
  n24 = ChannelMaskParameter(None=Add(A=@n5, B=Multiply(A=@n22.RGB, B=@n23))) {group=AnimatedHair, maskChannel=Blue, parameterName=AH_SineWaveChannel, defaultValue=<rgba 0 0 1 0>}
  n25 = Multiply(A=@n24.R, B=ScalarParameter() {defaultValue=4, group=AnimatedHair, parameterName=AH_SineWaveFrequency})
  Add(
   A=Multiply(A=Sine(None=@n25), B=VectorParameter() {group=AnimatedHair, parameterName=AH_Displacement_Amt, defaultValue=<rgba 0 4 0 1>}.RGB)
   B=@n20
  )
  n26 = GetMaterialAttributes(None=FunctionInput() {inputName=Material, inputType=FunctionInput_MaterialAttributes}) {attributeGetTypes [#1, #2]}
  n27 = VectorParameter() {group=AnimatedHair, parameterName=AH_GlowCOlor, defaultValue=<rgba 0 1 0.9145 1>}
  n28 = TextureCoordinate()
  n29 = AppendVector(
   A=ScalarParameter() {defaultValue=2, group=AnimatedHair, parameterName=AH_EmissiveFresnelDistortionTiling_X}
   B=ScalarParameter() {defaultValue=2, group=AnimatedHair, parameterName=AH_EmissiveFresnelDistortionTiling_Y}
  )
  n30 = VectorParameter() {group=AnimatedHair, parameterName=AH_DistortionPanSpeedAndDirection, defaultValue=<rgba 0 0.25 0 1>}
  n31 = TextureSampleParameter2D(UVs=Panner(Coordinate=Multiply(A=@n28, B=@n29), Speed=ComponentMask(None=@n30.RGB) {g=true, r=true})) {group=DistortionFresnel, parameterName=AH_DistortionNoise, texture=&3xPerlin_M}
  n32 = ComponentMask(None=Subtract(A=Multiply(A=@n31.RGB) {constB=2})) {g=true, r=true}
  n33 = StaticSwitchParameter(True=@n3.RGB, False=VertexInterpolator(VS=@n4).PS) {parameterName=UseInfluenceGradientTexture}
  n34 = Clamp(None=Subtract(A=ComponentMask(None=@n33) {b=true}, B=ScalarParameter() {defaultValue=0.75, group=AnimatedHair, parameterName=AH_GradientOffset2}))
  n35 = ScalarParameter() {defaultValue=0.5, group=AnimatedHair, parameterName=AH_MaskDistortionStrength}
  n36 = Add(A=TextureCoordinate(), B=Multiply(A=@n32, B=Multiply(A=@n34, B=@n35)))
  n37 = TextureSample(UVs=@n36) {
   samplerType=SAMPLERTYPE_Normal
   texture=/Game/Characters/Player/Male/Medium/Bodies/M_MED_Skeleton_Hunter/Textures/T_M_MED_SkeletonHunter_hair_N
  }
  n38 = TextureObjectParameter() {
   group=AnimatedHair
   parameterName=HairFX_Mask
   samplerType=SAMPLERTYPE_Masks
   texture=/Game/Characters/Player/Male/Medium/Bodies/M_MED_Skeleton_Hunter/Textures/T_MED_SkeletonHunter_Hair_FX
  }
  n39 = TextureSample(UVs=@n36, Tex=@n38) {
   samplerType=SAMPLERTYPE_Masks
   texture=&T_MED_SkeletonHunter_Hair_FX
  }
  n40 = Add(
   A=Fresnel(ExponentIn=ScalarParameter() {defaultValue=1, group=AnimatedHair, parameterName=AH_FresnelExponent}, Normal=Transform(None=@n37.RGB))
   B=@n39.B
  )
  n41 = ScalarParameter() {defaultValue=2, group=AnimatedHair, parameterName=AH_FresnelPower}
  n42 = Multiply(A=@n34, B=Clamp(None=Power(Base=@n40, Exp=@n41)))
  Add(
   A=@n26.EmissiveColor
   B=Multiply(A=Multiply(A=@n27.RGB, B=@n42), B=ScalarParameter() {defaultValue=2, group=AnimatedHair, parameterName=AH_GlowStrength})
  )
  n43 = TextureSample() {
   samplerType=SAMPLERTYPE_Masks
   texture=/Game/Characters/Player/Male/Medium/Bodies/M_MED_Skeleton_Hunter/Textures/T_M_MED_SkeletonHunter_hair_M
  }
  n44 = ComponentMask(None=FunctionInput(Preview=@n43.RGB) {bUsePreviewValueAsDefault=true, inputName=GMap}) {r=true}
  Clamp(None=Subtract(A=OneMinus(None=@n44), B=ScalarParameter() {defaultValue=0.05, parameterName=HairAO_Offset}))
  FunctionOutput(None=StaticSwitchParameter(
    True=SetMaterialAttributes!(MaterialAttributes=@n26.MaterialAttributes, WorldPositionOffset=@n21) {attributeSetTypes [#3]}
    False=@n26.MaterialAttributes
   ) {parameterName=UseGlowyAnimatedHair})
  Multiply(A=VertexNormalWS(), B=@n20)
  Subtract(A=Multiply(A=@n18) {constB=2})
  n45 = Multiply(
   A=TextureCoordinate()
   B=AppendVector(
    A=ScalarParameter() {defaultValue=1, group=AnimatedHair, parameterName=AH_EmissiveFresnelNoiseTiling_X}
    B=ScalarParameter() {defaultValue=1, group=AnimatedHair, parameterName=AH_EmissiveFresnelNoiseTiling_Y}
   )
  )
  n46 = Multiply(
   A=@n32
   B=AppendVector(
    A=ScalarParameter() {defaultValue=0.2, group=AnimatedHair, parameterName=AH_X_DistortionStrength}
    B=ScalarParameter() {defaultValue=0.2, group=AnimatedHair, parameterName=AH_Y_DistortionStrength}
   )
  )
  n47 = VectorParameter() {group=AnimatedHair, parameterName=DF_DistortionNoisePanSpeedAndDirection, defaultValue=<rgba 0.5 -0.5 0 1>}
  TextureSampleParameter2D(UVs=Panner(Coordinate=Add(A=@n45, B=@n46), Speed=ComponentMask(None=@n47.RGB) {g=true, r=true})) {group=DistortionFresnel, parameterName=Noise, texture=&3xPerlin_M}
  TextureSample()
  VertexInterpolator(VS=@n8)
  VertexInterpolator(VS=@n14)
 ]}
