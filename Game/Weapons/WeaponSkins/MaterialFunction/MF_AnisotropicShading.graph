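# Material function for weapon skins that optionally adds anisotropic shading to incoming material attributes.
# One texture parameter supplies both inputs: RGB is remapped from a -1 lower bound into Tangent, and alpha scaled by AnisotropyMaxWeight becomes Anisotropy.
# The UseAnisotropicShading static switch outputs either the modified attributes or the untouched input.
# A DebugTimeSine call and a constant are left unconnected.
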MaterialFunction /Game/Weapons/WeaponSkins/MaterialFunction/MF_AnisotropicShading
graph {nodes [
  Constant() {r=0.5}
  n1 = TextureSampleParameter2D() {group=Specular, parameterName=AnisotropigTangentWeight, samplerType=SAMPLERTYPE_LinearColor, texture=/Game/Global/Textures/T_Default_Aniso}
  n2 = LinearInterpolate(A=Constant3Vector() {constant=<rgba -1 -1 -1 0>}, Alpha=@n1.RGB)
  n3 = Multiply(A=@n1.A, B=ScalarParameter() {defaultValue=1, group=Specular, parameterName=AnisotropyMaxWeight, sliderMax=1})
  n4 = FunctionInput(Preview=SetMaterialAttributes!(MaterialAttributes)) {bUsePreviewValueAsDefault=true, inputType=FunctionInput_MaterialAttributes}
  FunctionOutput(None=StaticSwitchParameter(
    True=SetMaterialAttributes!(MaterialAttributes=@n4, Tangent=@n2, Anisotropy=@n3) {attributeSetTypes [#1, #2]}
    False=@n4
   ) {group=Skin FX, parameterName=UseAnisotropicShading})
  MaterialFunctionCall() {materialFunction=&DebugTimeSine}
 ]}
