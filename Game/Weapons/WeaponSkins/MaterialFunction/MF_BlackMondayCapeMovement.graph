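# Material function that adds animated cloth ripples and wind flapping to the Black Monday cape material, modifying Normal and World Position Offset.
# Panning world-aligned fold normals, projected from local or rotated pre-skinned positions, blend in by the mask's red channel.
# Offset sums time-scrolled curl noise scaled by mask blue, a normalized vector toward object origin, and switchable vertex-green-weighted sine waves.
# A second swirl-normal blend is left unconnected.
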MaterialFunction /Game/Weapons/WeaponSkins/MaterialFunction/MF_BlackMondayCapeMovement
graph {nodes [
  n1 = GetMaterialAttributes(None=FunctionInput() {inputName=Material, inputType=FunctionInput_MaterialAttributes}) {attributeGetTypes [#1, #2]}
  n2 = Subtract(A=ObjectPositionWS(), B=WorldPosition().XYZ)
  n3 = FunctionInput(Preview=PreSkinnedPosition()) {bUsePreviewValueAsDefault=true, inputName=PreSkinnedPosition}
  n4 = VectorParameter() {group=BlackMonday, parameterName=DirectionVector, defaultValue=<rgba 1 0 0 1>}
  n5 = LinearInterpolate(
   A=Transform(None=@n2) {transformSourceType=TRANSFORMSOURCE_World, transformType=TRANSFORM_Local}
   B=VertexInterpolator(VS=MaterialFunctionCall!(VectorToRotate=@n3, LookAtVector=@n4.RGB, RestingVector) {materialFunction=/Engine/Functions/Engine_MaterialFunctions01/Vectors/RotateVector}.Result).PS
   Alpha=ScalarParameter() {group=BlackMonday, parameterName=UsePreskinnedUVs}
  )
  n6 = VectorParameter() {group=BlackMonday, parameterName=FoldsPanSpeed, defaultValue=<rgba -0.2 -1.5 0 1>}
  n7 = ComponentMask(None=MaterialFunctionCall!(
    TextureObject=TextureObject() {samplerType=SAMPLERTYPE_Normal, texture=/Game/Gadgets/Assets/VinderTech_GliderChute/Glider_BlackMondayCape/Textures/T_Folds_N}
    TextureSize=VectorParameter() {group=BlackMonday, parameterName=FoldNormalTiling, defaultValue=<rgba 256 128 128 1>}.RGB
    Normal
    WorldSpace
    ProjectionTransitionContrast
    WorldPosition=@n5
    Use High Quality Normals
    Pan_YZ
    Pan_XZ
    Pan_XY=ComponentMask(None=@n6.RGB) {g=true, r=true}
   ) {materialFunction=/Game/Weapons/WeaponSkins/Wraps/MaterialFunctions/WorldAlignedNormal_ComplexPan}.XYZ Texture) {b=true, g=true, r=true}
  n8 = TextureSample() {
   samplerType=SAMPLERTYPE_Masks
   texture=/Game/Gadgets/Assets/VinderTech_GliderChute/Glider_BlackMondayCape/Textures/T_Glider_BlackMonday_Cape_FX
  }
  n9 = FunctionInput(Preview=@n8.RGB) {bUsePreviewValueAsDefault=true, inputName=Mask}
  n10 = ComponentMask(None=@n9) {r=true}
  n11 = LinearInterpolate(
   A=TextureSample() {samplerType=SAMPLERTYPE_Normal, texture=&FlatNormalBlue}.RGB
   B=@n7
   Alpha=Clamp(None=Multiply(A=@n10) {constB=2}, Max=ScalarParameter() {defaultValue=1, group=BlackMonday, parameterName=RippleNormalIntensity})
  )
  n12 = MaterialFunctionCall(Float3=Divide(A=@n3, B=VectorParameter() {group=BlackMonday, parameterName=DisplacementNoiseScale, defaultValue=<rgba 512 256 256 1>}.RGB)) {materialFunction=&BreakOutFloat3Components}
  n13 = Add(A=@n12.G, B=Multiply(A=Time(), B=ScalarParameter() {defaultValue=2, group=BlackMonday, parameterName=WindFlapSpeed}))
  n14 = Multiply(
   A=VectorNoise(World Position=MaterialFunctionCall(X=@n12.R, Y=@n13, Z=@n12.R) {materialFunction=&MakeFloat3}.Result) {noiseFunction=VNF_CurlALU}
   B=Multiply(
    A=ComponentMask(None=@n9) {b=true}
    B=VectorParameter() {group=BlackMonday, parameterName=VertexDIsplacement_Amt, defaultValue=<rgba 0.25 0.25 0.5 1>}.RGB
   )
  )
  n15 = Divide(
   A=ChannelMaskParameter(None=@n3) {maskChannel=Green, parameterName=HighFrequencyWPO_Direction, defaultValue=<rgba 0 1 0 0>}.R
   B=ScalarParameter() {defaultValue=64, parameterName=HighFrequency}
  )
  n16 = Multiply(A=Time(), B=ScalarParameter() {defaultValue=2, parameterName=HighFrequencyWPO_Speed})
  n17 = Multiply(
   A=Sine(None=Add(A=@n15, B=@n16))
   B=Multiply(A=VectorParameter() {parameterName=HighFrequencyWPO_Offset, defaultValue=<rgba 3 2 4 1>}.RGB, B=VertexColor().G)
  )
  n18 = Add(A=Normalize(VectorInput=@n2), B=StaticSwitchParameter(True=Add(A=@n14, B=@n17), False=@n14) {parameterName=AddHighFrequencyWPO})
  FunctionOutput(None=SetMaterialAttributes!(
    MaterialAttributes=@n1.MaterialAttributes
    Normal=MaterialFunctionCall(BaseNormal=@n1.Normal, AdditionalNormal=@n11) {materialFunction=&BlendAngleCorrectedNormals}.Result
    WorldPositionOffset=Add(A=@n1.WorldPositionOffset, B=@n18)
   ) {attributeSetTypes [#1, #2]})
  n19 = TextureObject() {samplerType=SAMPLERTYPE_Normal, texture=/Game/Global/Textures/Default/Noise/T_SwirlNoise_N}
  n20 = VectorParameter() {group=BlackMonday, parameterName=FoldPerlinTiling, defaultValue=<rgba 256 128 128 1>}
  n21 = ComponentMask(None=VectorParameter() {group=BlackMonday, parameterName=FoldsPanSpeed2, defaultValue=<rgba 0.2 -2 0 1>}.RGB) {g=true, r=true}
  MaterialFunctionCall(
   BaseNormal=@n7
   AdditionalNormal=ComponentMask(None=MaterialFunctionCall(TextureObject=@n19, TextureSize=@n20.RGB, WorldPosition=@n5, Pan_XY=@n21) {materialFunction=&WorldAlignedNormal_ComplexPan}.XYZ Texture) {b=true, g=true, r=true}
  ) {materialFunction=&BlendAngleCorrectedNormals}
  OneMinus(None=ComponentMask(None=@n9) {g=true})
  VertexInterpolator()
 ]}
