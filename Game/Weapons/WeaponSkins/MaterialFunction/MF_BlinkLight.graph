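# Weapon skin material function that adds blinking lights to incoming material attributes under SkinFX_BlinkLights parameters.
# Static switches select a simple timed or parameter-driven color blink on base color and emissive, or additive masked emissive blinks.
# FX mask or vertex color channels drive incremental buildup and SphereMask order blinks, colored and randomized through curve atlases, with a texture-mask flash fallback.
# The blink mask can also push vertices along normals through World Position Offset.
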MaterialFunction /Game/Weapons/WeaponSkins/MaterialFunction/MF_BlinkLight
graph {nodes [
  Constant3Vector()
  n1 = Multiply(A=VectorParameter() {group=SkinFX_BlinkLights, parameterName=Simnple Blink Color, defaultValue=<rgba 1 1 1 0>}.RGB, B=NamedRerouteUsage())
  n2 = Frac(None=Multiply(A=Time(), B=NamedRerouteUsage()))
  n3 = StaticSwitchParameter(
   True=ScalarParameter() {desc=Bool Control For turning the blink on/off, group=SkinFX_BlinkLights, parameterName=SimpleBlink_Bool}
   False=Floor(None=Multiply(A=@n2) {constB=2})
  ) {group=SkinFX_BlinkLights, parameterName=Use Bool for Simple Blink}
  n4 = GetMaterialAttributes(None=FunctionInput() {bUsePreviewValueAsDefault=true, inputName=MA_Input, inputType=FunctionInput_MaterialAttributes, sortPriority=1}) {attributeGetTypes [#1, #2, #3]}
  n5 = FunctionInput(Preview=TextureSample() {texture=/Game/Weapons/FORT_Melee/Pickaxe_NeonGraffiti/Textures/T_Pickaxe_NeonGraffiti_FX}.RGBA) {bUsePreviewValueAsDefault=true, inputName=FX_Mask_Vectoir, inputType=FunctionInput_Vector4}
  n6 = VertexColor()
  n7 = AppendVector(A=@n6, B=@n6)
  n8 = LinearInterpolate(A=@n5, B=@n7, Alpha=ScalarParameter() {group=SkinFX_BlinkLights, parameterName=BL_UseVertexColorsForCOlor})
  n9 = ChannelMaskParameter(None=@n8) {group=SkinFX_BlinkLights, maskChannel=Blue, parameterName=BLinkLights_ColorGradient, defaultValue=<rgba 0 0 1 0>}
  n10 = CurveAtlasRowParameter(CurveTime=@n9.R) {
   atlas=&AthenaCurveAtlas01
   curve=/Game/Accessories/FORT_Backpacks/Backpack_F_MED_Butterfly/FX/Curve_Butterfly_BackPack
   defaultValue=102
   group=SkinFX_BlinkLights
   parameterName=BL_LightsColorCurve
  }
  n11 = LinearInterpolate(
   A=LinearInterpolate(A=@n10, B=@n4.EmissiveColor, Alpha=ScalarParameter() {group=SkinFX_BlinkLights, parameterName=UseBaseEmissiveTextureForBlinkColor})
   B=VectorParameter() {group=SkinFX_BlinkLights, parameterName=BL_EmissiveColor, defaultValue=<rgba 1 0 0 1>}.RGB
   Alpha=ScalarParameter() {group=SkinFX_BlinkLights, parameterName=UseEmissiveColorParameter}
  )
  n12 = ChannelMaskParameter(None=@n8) {group=SkinFX_BlinkLights, maskChannel=Green, parameterName=BlinkLights_EmissiveGradientBlendChannel, defaultValue=<rgba 0 1 0 0>}
  n13 = LinearInterpolate(A=@n11, B=Multiply(A=@n11, B=@n12.R), Alpha=ScalarParameter() {group=SkinFX_BlinkLights, parameterName=BL_MultiplyEmissiveGradientChannel})
  n14 = StaticSwitchParameter(True=Add(A=@n5, B=@n7), False=StaticSwitchParameter(True=@n7, False=@n5) {group=SkinFX_BlinkLights, parameterName=UseVertexColors}) {group=SkinFX_BlinkLights, parameterName=UseVertexColorsAndFXMask}
  n15 = ChannelMaskParameter(None=@n14) {group=SkinFX_BlinkLights, maskChannel=Green, parameterName=BlinkLights_IncrementalChannel, defaultValue=<rgba 0 1 0 0>}
  n16 = Time()
  n17 = Multiply(A=@n16, B=ScalarParameter() {defaultValue=1, group=SkinFX_BlinkLights, parameterName=IncrementSpeed})
  n18 = ConstantBiasScale(None=Sine(None=@n17))
  n19 = ScalarParameter() {defaultValue=0.1, group=SkinFX_BlinkLights, parameterName=BlinkBias}
  n20 = LinearInterpolate(
   A=LinearInterpolate(
    A=Frac(None=@n17)
    B=Subtract(A=@n18, B=@n19) {constB=0.1}
    Alpha=ScalarParameter() {group=SkinFX_BlinkLights, parameterName=Blink_BuildAndFade}
   )
   B=ScalarParameter() {group=SkinFX_BlinkLights, parameterName=BlinkControl}
   Alpha=ScalarParameter() {group=SkinFX_BlinkLights, parameterName=UseScalarParameterForBlink}
  )
  n21 = OneMinus(None=@n20)
  n22 = ScalarParameter() {group=SkinFX_BlinkLights, parameterName=BL_FlipIncrementDirection}
  n23 = Ceil(None=Subtract(A=@n15.R, B=LinearInterpolate(A=@n21, B=@n20, Alpha=@n22)))
  n24 = ChannelMaskParameter(None=@n14) {group=SkinFX_BlinkLights, maskChannel=Blue, parameterName=BlinkLights_OrderChannel, defaultValue=<rgba 0 0 1 0>}
  n25 = Multiply(
   A=@n16
   B=NamedRerouteDeclaration(None=ScalarParameter() {defaultValue=1, group=SkinFX_BlinkLights, parameterName=Blink_Speed}) {name=Blink_Speed_Reroute, nodeColor=<rgba 0.5 0 1 1>}
  )
  n26 = Frac(None=@n25)
  n27 = CurveAtlasRowParameter(CurveTime=@n26) {
   atlas=&AthenaCurveAtlas01
   curve=/Game/Athena/LinearColorCurves/Balster/Curve_LaserEffectRamp2
   defaultValue=62
   group=SkinFX_BlinkLights
   parameterName=RandomizedBLinkCurve
  }
  n28 = SphereMask(
   A=@n24.R
   B=LinearInterpolate(
    A=@n26
    B=ComponentMask(None=@n27) {r=true}
    Alpha=ScalarParameter() {defaultValue=1, group=SkinFX_BlinkLights, parameterName=BL_UseRandomizedBlinks}
   )
   Radius=ScalarParameter() {defaultValue=0.05, group=SkinFX_BlinkLights, parameterName=BL_LightValueThreshold}
   Hardness=Constant() {r=1}
  )
  n29 = Multiply(
   A=Ceil(None=@n24.R)
   B=LinearInterpolate(A=@n28, B=OneMinus(None=@n28), Alpha=ScalarParameter() {group=SkinFX_BlinkLights, parameterName=BL_FlipBlink})
  )
  n30 = StaticSwitchParameter(
   True=Clamp(None=Add(A=@n23, B=@n29))
   False=StaticSwitchParameter(True=@n23, False=@n29) {group=SkinFX_BlinkLights, parameterName=UseBuildupBlinks}
  ) {group=SkinFX_BlinkLights, parameterName=UseBuildupAndRandomizedBlinks}
  n31 = NamedRerouteDeclaration(None=ScalarParameter() {defaultValue=1, group=SkinFX_BlinkLights, parameterName=Emissive_BlinkingBrightness}) {name=Emissive_BlinkingBrightness_Reroute, nodeColor=<rgba 0 0.6953 1 1>}
  n32 = TextureObjectParameter() {parameterName=BlinkEmissiveForLights, samplerType=SAMPLERTYPE_Masks, texture=/Engine/EngineMaterials/DefaultDiffuse_TC_Masks}
  n33 = FunctionInput() {bUsePreviewValueAsDefault=true, inputName=FX_MaskInput, inputType=FunctionInput_Scalar, sortPriority=2}
  n34 = Multiply(A=Frac(None=Subtract(A=@n33, B=@n25))) {constB=2}
  n35 = Multiply(A=Floor(None=@n34)) {constA=10, constB=25}
  n36 = Multiply(A=Floor(None=Add(A=@n34) {constB=-0.8})) {constA=100, constB=50}
  n37 = Multiply(
   A=Multiply(
    A=TextureSample(Tex=@n32) {samplerType=SAMPLERTYPE_Masks, texture=&DefaultDiffuse_TC_Masks}.RGB
    B=Add(A=@n35, B=@n36)
   )
   B=@n31
  )
  n38 = OneMinus(None=@n6.G)
  n39 = StaticSwitchParameter(
   True=Multiply(A=Multiply(A=@n13, B=@n30), B=@n31)
   False=StaticSwitchParameter(True=Multiply(A=@n37, B=@n38), False=@n37) {defaultValue=true, group=SkinFX_BlinkLights, parameterName=UseBlinkingVertexColormask}
  ) {group=SkinFX_BlinkLights, parameterName=UseRandomizedBlinks}
  n40 = ScalarParameter() {defaultValue=2, group=SkinFX_BlinkLights, parameterName=BL_VertexDisplacementAmt}
  n41 = Multiply(
   A=Multiply(A=VertexNormalWS(), B=Multiply(A=@n30, B=@n40))
   B=ScalarParameter() {group=SkinFX_BlinkLights, parameterName=BL_EnableVertDisplacement}
  )
  n42 = StaticSwitchParameter(
   True=SetMaterialAttributes!(
    MaterialAttributes=@n4.MaterialAttributes
    EmissiveColor=LinearInterpolate(A=@n4.EmissiveColor, B=@n1, Alpha=@n3)
    BaseColor=LinearInterpolate(A=@n4.BaseColor, B=@n1, Alpha=@n3)
   ) {attributeSetTypes [#1, #3]}
   False=SetMaterialAttributes!(
    MaterialAttributes=@n4.MaterialAttributes
    EmissiveColor=Add(A=@n4.EmissiveColor, B=@n39)
    WorldPositionOffset=Add(A=@n4.WorldPositionOffset, B=@n41)
   ) {attributeSetTypes [#1, #2]}
  ) {
   desc=Use a simple blink that doesn't use any masks and just takes the current emissive and sets them to a specific vector value.
   group=SkinFX_BlinkLights
   parameterName=UseSimpleBlink
  }
  FunctionOutput(None=@n42) {outputName=MA_Output}
 ]}
