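# Material function for weapon skins that recolors the parts of a material's BaseColor matching a key color, creating a chameleon color shift.
# The replacement color is the BaseColor green channel multiplied by the ChameleonColorShift tint.
# A dot product against ChameleonBaseColor, remapped by min/max parameters and clamped, builds the mask.
# The mask, Alpha input and tint alpha (zero by default) scale the blend written back to the material attributes.
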MaterialFunction /Game/Weapons/WeaponSkins/MaterialFunction/MF_ChameleonColorShift
graph {nodes [
  n1 = Constant()
  n2 = GetMaterialAttributes(None=FunctionInput(Preview=SetMaterialAttributes!(MaterialAttributes, BaseColor=@n1) {attributeSetTypes [#1]}) {bUsePreviewValueAsDefault=true, inputName=Material, inputType=FunctionInput_MaterialAttributes}) {attributeGetTypes [#1]}
  n3 = VectorParameter() {parameterName=ChameleonColorShift, defaultValue=<rgba 0.0176 0.3666 0.9375 0>}
  n4 = Multiply(
   A=MaterialFunctionCall(Float3=@n2.BaseColor) {materialFunction=&BreakOutFloat3Components}.G
   B=@n3.RGB
  )
  n5 = VectorParameter() {group=ChameleonColorShift, parameterName=ChameleonBaseColor, defaultValue=<rgba 0.4125 0.6376 0.0612 0>}
  n6 = LinearInterpolate(
   A=ScalarParameter() {group=ChameleonColorShift, parameterName=ChameleonMaskMin}
   B=ScalarParameter() {defaultValue=2, group=ChameleonColorShift, parameterName=ChameleonMaskMax}
   Alpha=DotProduct(A=@n2.BaseColor, B=@n5.RGB)
  )
  n7 = Multiply(A=@n3.A, B=Multiply(A=Clamp(None=@n6), B=FunctionInput() {bUsePreviewValueAsDefault=true, inputName=Alpha, inputType=FunctionInput_Scalar}))
  FunctionOutput(None=SetMaterialAttributes(MaterialAttributes=@n2.MaterialAttributes, BaseColor=LinearInterpolate(A=@n2.BaseColor, B=@n4, Alpha=@n7)) {attributeSetTypes [#1]})
 ]}
