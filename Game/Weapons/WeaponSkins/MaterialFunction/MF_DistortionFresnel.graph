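# Material function for weapon skins that adds an animated, noise-distorted Fresnel glow to a material's emissive.
# Two panning noise lookups blend planar projections of pre-skinned position by normal, the first warping the second.
# The second noise erodes a Fresnel with optional baked normal; an optional distorted UV mask can be added.
# A flat color or curve-atlas gradient tints the result, gated by the Mask input channel.
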MaterialFunction /Game/Weapons/WeaponSkins/MaterialFunction/MF_DistortionFresnel
graph {nodes [
  n1 = Abs(None=VertexInterpolator(VS=PreSkinnedNormal()).PS)
  n2 = ComponentMask(None=@n1) {g=true}
  Abs(None=@n2)
  n3 = Abs(None=VertexNormalWS())
  ComponentMask(None=@n3) {b=true}
  ComponentMask(None=@n3) {r=true}
  n4 = GetMaterialAttributes(None=FunctionInput() {bUsePreviewValueAsDefault=true, inputName=Material, inputType=FunctionInput_MaterialAttributes}) {attributeGetTypes [#1, #2]}
  n5 = Fresnel(
   ExponentIn=ScalarParameter() {defaultValue=2, group=DistortionFresnel, parameterName=DF_FresnelExponent}
   Normal=LinearInterpolate(A=VertexNormalWS(), B=Transform(None=@n4.Normal), Alpha=ScalarParameter() {group=DistortionFresnel, parameterName=DF_bUseBakedNormal})
  )
  n6 = VectorParameter() {group=DistortionFresnel, parameterName=DF_EmissiveFresnelNoiseTiling, defaultValue=<rgba 256 256 256 1>}
  n7 = VertexInterpolator(VS=PreSkinnedPosition())
  n8 = VectorParameter() {group=DistortionFresnel, parameterName=DF_DistortionPanSpeedAndDirection, defaultValue=<rgba -0.25 -0.25 0 1>}
  n9 = Add(
   A=Divide(A=@n7.PS, B=VectorParameter() {group=DistortionFresnel, parameterName=DF_EmissiveDistortionTiling, defaultValue=<rgba 256 256 256 1>}.RGB)
   B=Multiply(A=Time(), B=@n8.RGB)
  )
  n10 = TextureObjectParameter() {parameterName=DistorntionNoise, texture=/Game/Global/Textures/Default/Noise/3xPerlin_M}
  n11 = LinearInterpolate(
   A=TextureSample(UVs=ComponentMask(None=@n9) {g=true, r=true}, Tex=@n10).RGB
   B=TextureSample(UVs=ComponentMask(None=@n9) {b=true, r=true}, Tex=@n10).RGB
   Alpha=@n2
  )
  n12 = TextureSample(UVs=ComponentMask(None=@n9) {b=true, g=true}, Tex=@n10)
  n13 = ComponentMask(None=@n1) {r=true}
  n14 = Subtract(A=Multiply(A=LinearInterpolate(A=@n11, B=@n12.RGB, Alpha=@n13)) {constB=2})
  n15 = VectorParameter() {group=DistortionFresnel, parameterName=DF_DistortionStrength, defaultValue=<rgba 0.2 0.2 0.2 1>}
  n16 = VectorParameter() {group=DistortionFresnel, parameterName=DF_DistortionNoisePanSpeedAndDirection, defaultValue=<rgba -0.1 -0.5 0 1>}
  n17 = Add(A=Add(A=Divide(A=@n7.PS, B=@n6.RGB), B=Multiply(A=@n14, B=@n15.RGB)), B=Multiply(A=Time(), B=@n16.RGB))
  n18 = ComponentMask(None=@n17) {g=true, r=true}
  n19 = TextureObjectParameter() {parameterName=DistorntionNoise, texture=&3xPerlin_M}
  n20 = ComponentMask(None=@n17) {b=true, r=true}
  n21 = LinearInterpolate(
   A=LinearInterpolate(A=TextureSample(UVs=@n18, Tex=@n19).RGB, B=TextureSample(UVs=@n20, Tex=@n19).RGB, Alpha=@n2)
   B=TextureSample(UVs=ComponentMask(None=@n17) {b=true, g=true}, Tex=@n19).RGB
   Alpha=@n13
  )
  n22 = Multiply(
   A=Power(Base=ComponentMask(None=@n21) {r=true}, Exp=ScalarParameter() {defaultValue=2, group=DistortionFresnel, parameterName=NoiseContrast})
   B=ScalarParameter() {defaultValue=1, group=DistortionFresnel, parameterName=DF_DistortionNoiseStrength}
  )
  n23 = Clamp(None=Power(Base=Subtract(A=@n5, B=@n22), Exp=ScalarParameter() {defaultValue=1, group=DistortionFresnel, parameterName=DF_FresnelPower}))
  n24 = ComponentMask(None=@n14) {g=true, r=true}
  n25 = ScalarParameter() {group=DistortionFresnel, parameterName=DF_MaskDistortionStrength}
  n26 = TextureSampleParameter2D(UVs=Add(A=TextureCoordinate(), B=Multiply(A=@n24, B=@n25))) {
   group=DistortionFresnel
   parameterName=DF_DistortionMask
   texture=/Game/Characters/Player/Male/Medium/Bodies/M_MED_Skeleton_Hunter/Textures/T_M_MED_SkeletonHunter_body_E
  }
  n27 = ChannelMaskParameter(None=@n26.RGB) {group=DistortionFresnel, maskChannel=Blue, parameterName=DF_MaskChannel, defaultValue=<rgba 0 0 1 0>}
  n28 = LinearInterpolate(
   A=@n27.R
   B=Add(A=ChannelMaskParameter(None=@n21) {group=DistortionFresnel, parameterName=DF_NoiseChannel}.R, B=@n27.R)
   Alpha=ScalarParameter() {group=DistortionFresnel, parameterName=UseTilingNoise}
  )
  n29 = LinearInterpolate(A=@n23, B=Clamp(None=Add(A=@n23, B=@n28)), Alpha=ScalarParameter() {group=DistortionFresnel, parameterName=DF_AddDistortedMask})
  n30 = Multiply(A=@n29, B=VectorParameter() {group=DistortionFresnel, parameterName=DF_DistortionFresnelColor, defaultValue=<rgba 1 1 1 1>}.RGB)
  n31 = CurveAtlasRowParameter(CurveTime=@n29) {
   atlas=&AthenaCurveAtlas01
   curve=/Game/Athena/LinearColorCurves/skeletonhunterpickaxecolorcurve
   defaultValue=145
   group=DistortionFresnel
   parameterName=DistortionFresnelColorCurve
  }
  n32 = ScalarParameter() {group=DistortionFresnel, parameterName=DF_UseColorCurve}
  n33 = Add(
   A=@n4.EmissiveColor
   B=Multiply(
    A=LinearInterpolate(A=@n30, B=@n31, Alpha=@n32)
    B=ScalarParameter() {defaultValue=1, group=DistortionFresnel, parameterName=DF_DistortionFresnelStrength}
   )
  )
  n34 = ChannelMaskParameter(None=FunctionInput(Preview=Constant() {r=1}) {bUsePreviewValueAsDefault=true, inputName=Mask, inputType=FunctionInput_Vector4}) {group=DistortionFresnel, parameterName=DistortionFresnelMaskChannel}
  FunctionOutput(None=SetMaterialAttributes!(MaterialAttributes=@n4.MaterialAttributes, EmissiveColor=LinearInterpolate(A=@n4.EmissiveColor, B=@n33, Alpha=@n34.R)) {attributeSetTypes [#1]})
  Subtract(A=ObjectPositionWS())
  VertexNormalWS()
  WorldPosition()
 ]}
