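# Material function for weapon skins that blends a fur layer with the character-shader body, behind a static switch that otherwise passes input attributes through.
# Fur uses UV1 textures and subsurface shading, with vertex-color channels building the transition mask.
# An optional baked vertex-normal map on UV2 corrects body and fur normals.
# Blended results run through shared GMap color, balance, emissive and skin functions, then receive an opacity mask from an input texture.
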MaterialFunction /Game/Weapons/WeaponSkins/MaterialFunction/MF_FurNormalAndUVBlend
graph {
 properties {description="https://confluence.it.epicgames.com/pages/viewpage.action?pageId=170066353"}
 nodes [
  Constant()
  n1 = MaterialFunctionCall!(TexCoords=TextureCoordinate()) {materialFunction=/Game/Packages/Fortress_SharedMaterials/Base_Material_Functions/BaseCharacterShaderFunctions/BaseFN_CharacterFunction}
  n2 = MaterialFunctionCall(Float3=LinearInterpolate(A=NamedRerouteUsage(), B=@n1.M, Alpha=NamedRerouteUsage()) {desc=M}) {desc="R = AO\r\nG = GMap\r\nB = Skin", materialFunction=&BreakOutFloat3Components}
  n3 = MaterialFunctionCall!(
   Material=BlendMaterialAttributes(A=NamedRerouteUsage(), B=NamedRerouteUsage(), Alpha=NamedRerouteUsage())
   Diffuse=LinearInterpolate(A=NamedRerouteUsage(), B=@n1.DIffuse, Alpha=NamedRerouteUsage())
   GMap=@n2.G
  ) {materialFunction=/Game/Packages/Fortress_SharedMaterials/Base_Material_Functions/BaseCharacterShaderFunctions/BaseFN_CharacterGMapColor}
  n4 = MaterialFunctionCall!(Material=@n3.Result) {materialFunction=/Game/Packages/Fortress_SharedMaterials/Base_Material_Functions/CharacterBrightnessandScale_Balance}
  n5 = TextureSampleParameter2D(UVs=TextureCoordinate() {coordinateIndex=2}) {
   group=FurNoramlandUVBlend
   parameterName=Fur_Bake_VertexNormal
   samplerSource=SSM_Wrap_WorldGroupSettings
   samplerType=SAMPLERTYPE_Normal
   texture=/Engine/EngineMaterials/DefaultNormal
  }
  n6 = LinearInterpolate(
   A=NamedRerouteUsage()
   B=StaticSwitchParameter(
    True=MaterialFunctionCall(BaseNormal=@n5.RGB, AdditionalNormal=@n1.Normal Map) {materialFunction=&BlendAngleCorrectedNormals}.Result
    False=@n1.Normal Map
   ) {group=FurNoramlandUVBlend, parameterName=Use_VertexNormalBake_forComputedNormalMesh}
   Alpha=NamedRerouteUsage()
  ) {desc=Normal}
  n7 = MaterialFunctionCall!(
   Material=@n4.Result
   Diffuse=@n3.DIffuse
   BalanceDistance=@n4.Result
   Normals=@n6
   AO=@n2.R
   Fresnel=Fresnel(Normal=Transform(None=@n6)) {baseReflectFraction=0, exponent=1}
  ) {materialFunction=/Game/Packages/Fortress_SharedMaterials/Base_Material_Functions/BaseCharacterShaderFunctions/BaseFN_CharacterEmissive}
  n8 = Saturate(None=ScalarParameter() {defaultValue=1, group=FurNoramlandUVBlend, parameterName=OpacityMaskMultiplier})
  n9 = TextureCoordinate() {coordinateIndex=1}
  n10 = TextureSample(UVs=@n9, Tex=FunctionInput() {inputName=FurOpacityMaskTexture, inputType=FunctionInput_Texture2D})
  n11 = ScalarParameter() {
   desc=Set to 1 if the Red Channel in the FX Mask should be counted for Hologram masking in the FE
   group=HologramMasking
   parameterName=UseSkinFXMaskForOpacity
  }
  n12 = ChannelMaskParameter(None=LinearInterpolate(A=@n10.RGB, B=@n10.RGB, Alpha=@n11)) {group=FurNoramlandUVBlend, parameterName=FurOpacityMask}
  n13 = StaticSwitchParameter(
   True=SetMaterialAttributes!(
    MaterialAttributes=MaterialFunctionCall!(Material=@n7.Result, AO=@n2.R, Skin Mask=@n2.B) {materialFunction=/Game/Packages/Fortress_SharedMaterials/Base_Material_Functions/BaseCharacterShaderFunctions/BaseFN_CharacterSkin}.Result
    OpacityMask=Multiply(A=@n8, B=@n12.R)
   ) {attributeSetTypes [#1]}
   False=FunctionInput() {inputType=FunctionInput_MaterialAttributes}
  ) {desc="https://confluence.it.epicgames.com/pages/viewpage.action?pageId=170066353", group=Skin FX, parameterName=UseFurNormalAndUVBlend}
  FunctionOutput(None=@n13)
  LinearInterpolate(A=NamedRerouteUsage(), B=@n1.SpecularMask, Alpha=NamedRerouteUsage())
  LinearInterpolate(A=NamedRerouteUsage(), B=@n1.Roughness, Alpha=NamedRerouteUsage())
  n14 = VertexColor()
  n15 = ChannelMaskParameter(None=@n14) {group=FurNoramlandUVBlend, maskChannel=Green, parameterName=Fur_Blend_Mask_Channel, defaultValue=<rgba 0 1 0 0>}
  n16 = Multiply(A=@n15.R, B=ScalarParameter() {defaultValue=1, group=FurNoramlandUVBlend, parameterName=FurMaskScriptError})
  n17 = LinearInterpolate(
   A=ScalarParameter() {defaultValue=-1.091, group=FurNoramlandUVBlend, parameterName=FurBlend_Mask1}
   B=ScalarParameter() {defaultValue=1, group=FurNoramlandUVBlend, parameterName=FurBlend_Mask2}
   Alpha=ChannelMaskParameter(None=@n14) {group=FurNoramlandUVBlend, parameterName=Fur_Blend_Channel}.R
  )
  NamedRerouteDeclaration(None=Multiply(A=OneMinus(None=@n16), B=Saturate(None=@n17))) {name=FurTransitionMask, nodeColor=<rgba 1 0.3516 0 1>}
  n18 = TextureSampleParameter2D(UVs=@n9) {group=FurNoramlandUVBlend, parameterName=Fur_Diffuse, texture=/Engine/EngineMaterials/DefaultDiffuse}
  NamedRerouteDeclaration(None=@n18.RGB) {name=Fur_Diffuse, nodeColor=<rgba 0 1 0.8359 1>}
  n19 = TextureSampleParameter2D(UVs=@n9) {
   desc="R = SpecMask G = Metallic B = Roughness"
   group=FurNoramlandUVBlend
   parameterName=Fur_SpecularMasks
   samplerType=SAMPLERTYPE_Masks
   texture=&T_Fortnite_Default_S
  }
  n20 = MaterialFunctionCall(Float3=@n19.RGB) {materialFunction=&BreakOutFloat3Components}
  NamedRerouteDeclaration(None=@n20.B) {name=PreFX_RoughnessRawPaint, nodeColor=<rgba 1 0.5156 0 1>}
  n21 = Saturate(None=Power(Base=@n20.G, Exp=ScalarParameter() {defaultValue=0.4, group=Masks, parameterName=MetalMaskPower}))
  n22 = ScalarParameter() {group=Roughness, parameterName=MobileRoughnessMin}
  n23 = ScalarParameter() {defaultValue=1, group=Roughness, parameterName=MobileRoughnessMax}
  n24 = LinearInterpolate(
   A=ScalarParameter() {group=Roughness, parameterName=RoughnessMin}
   B=ScalarParameter() {defaultValue=1, group=Roughness, parameterName=RoughnessMax}
   Alpha=@n20.B
  )
  n25 = QualitySwitch(
   Default=@n24
   Low=StaticSwitchParameter(True=LinearInterpolate(A=@n22, B=@n23, Alpha=@n20.B), False=@n24) {parameterName=OverrideRoughnessOnMobile}
  )
  n26 = Constant()
  n27 = ScalarParameter() {defaultValue=0.5, group=FurNoramlandUVBlend, parameterName=FurSubSurface_Opacity}
  n28 = OneMinus(None=NamedRerouteUsage())
  n29 = TextureSampleParameter2D(UVs=TextureCoordinate() {coordinateIndex=2}) {
   group=FurNoramlandUVBlend
   parameterName=Fur_Bake_VertexNormal
   samplerSource=SSM_Wrap_WorldGroupSettings
   samplerType=SAMPLERTYPE_Normal
   texture=&DefaultNormal
  }
  n30 = Transform(None=@n29.RGB)
  n31 = VertexNormalWS()
  n32 = TextureSampleParameter2D(UVs=@n9) {group=FurNoramlandUVBlend, parameterName=Fur_Normals, samplerType=SAMPLERTYPE_Normal, texture=&DefaultNormal}
  n33 = TwoSidedSign()
  n34 = Transform(None=MaterialFunctionCall(
    BaseNormal=StaticSwitchParameter(True=@n30, False=@n31) {group=FurNoramlandUVBlend, parameterName=Use_VertexNormalBake_forComputedNormalMesh}
    AdditionalNormal=Multiply(A=@n32.RGB, B=@n33)
   ) {materialFunction=&BlendAngleCorrectedNormals}.Result) {transformSourceType=TRANSFORMSOURCE_World, transformType=TRANSFORM_Tangent}
  n35 = Multiply(A=@n18.RGB, B=ScalarParameter() {defaultValue=0.5, group=FurNoramlandUVBlend, parameterName=Fur_SubSurfaceBrightness})
  NamedRerouteDeclaration(None=MakeMaterialAttributes!(
    BaseColor=@n18.RGB
    Metallic=StaticSwitchParameter(True=@n21, False=@n20.G) {group=Masks, parameterName=useMetalMaskPowerAdjustment}
    Specular=@n20.R
    Roughness=@n25
    Anisotropy
    EmissiveColor=@n26
    Opacity=Multiply(A=@n27, B=@n28)
    OpacityMask
    Normal=@n34
    Tangent
    WorldPositionOffset
    SubsurfaceColor=Multiply(A=@n35, B=@n28)
    ClearCoat=@n26
    ClearCoatRoughness=@n26
    AmbientOcclusion
    Refraction
    CustomizedUVs_0
    CustomizedUVs_1
    CustomizedUVs_2
    CustomizedUVs_3
    CustomizedUVs_4
    CustomizedUVs_5
    CustomizedUVs_6
    CustomizedUVs_7
    PixelDepthOffset
    ShadingModel=ShadingModel() {shadingModel=MSM_Subsurface}
    Displacement
   )) {name=Fur, nodeColor=<rgba 0 1 0.2734 1>}
  n36 = GetMaterialAttributes(None=@n1.Result) {attributeGetTypes [#2]}
  n37 = MaterialFunctionCall(BaseNormal=@n5.RGB, AdditionalNormal=@n36.Normal) {materialFunction=&BlendAngleCorrectedNormals}
  n38 = SetMaterialAttributes!(
   MaterialAttributes=StaticSwitchParameter(
    True=SetMaterialAttributes!(MaterialAttributes=@n36.MaterialAttributes, Normal=@n37.Result) {attributeSetTypes [#2]}
    False=@n1.Result
   ) {group=FurNoramlandUVBlend, parameterName=Use_VertexNormalBake_forComputedNormalMesh}
   Opacity=Constant() {r=1}
  ) {attributeSetTypes [#3]}
  NamedRerouteDeclaration(None=@n38) {name=Body, nodeColor=<rgba 0 1 0.3672 1>}
  n39 = TextureSampleParameter2D(UVs=@n9) {
   desc="Masks R=AO G= Fresnel B = Tmission/Skin"
   group=FurNoramlandUVBlend
   parameterName=Fur_M
   samplerType=SAMPLERTYPE_Masks
   texture=/Engine/EngineMaterials/DefaultDiffuse_TC_Masks
  }
  NamedRerouteDeclaration(None=@n39.RGB) {name=Fur_M, nodeColor=<rgba 1 0.2109 0 1>}
  NamedRerouteDeclaration(None=@n25) {name=Fur_Roughness, nodeColor=<rgba 1 0 0.9141 1>}
  NamedRerouteDeclaration(None=@n19.RGB) {name=Fur_SpecularMask, nodeColor=<rgba 0 1 0.0156 1>}
  NamedRerouteDeclaration(None=@n34) {name=Fur_Normal, nodeColor=<rgba 0.9453 1 0 1>}
  Reroute(None=@n38)
  Reroute(None=@n19.RGB)
  Reroute(None=@n25)
 ]
}
