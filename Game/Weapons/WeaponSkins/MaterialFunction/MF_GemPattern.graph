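# Weapon skin material function that layers a tiled, faceted gem pattern with view-dependent color and sparkle onto an input material.
# The pattern normal is blended into the base normal, whose view-angle dot product, scaled by pattern alpha, drives a curve-atlas gem color.
# Per-facet random vectors dotted with camera direction drive an optionally rounded sparkle curve, modulating added emissive between default and sparkle power.
# A selectable mask channel blends the result over the input.
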MaterialFunction /Game/Weapons/WeaponSkins/MaterialFunction/MF_GemPattern
graph {nodes [
  n1 = GetMaterialAttributes(None=FunctionInput() {inputName=Material, inputType=FunctionInput_MaterialAttributes}) {attributeGetTypes [#1, #2]}
  n2 = Multiply(A=TextureCoordinate(), B=ScalarParameter() {defaultValue=1, group=GemPattern, parameterName=GP_PatternTiling})
  n3 = MaterialFunctionCall(
   BaseNormal=@n1.Normal
   AdditionalNormal=TextureSampleParameter2D(UVs=@n2) {parameterName=GP_Pattern_Normal, samplerType=SAMPLERTYPE_Normal, texture=/Game/Weapons/WeaponSkins/Textures/T_DiamondPattern_N}.RGB
  ) {materialFunction=&BlendAngleCorrectedNormals}
  n4 = DotProduct(A=CameraVectorWS(), B=Transform(None=@n3.Result))
  n5 = TextureSampleParameter2D(UVs=@n2) {
   group=GemPattern
   parameterName="GP_PatternRandomVector,Mask(RGB,A)"
   texture=/Game/Weapons/WeaponSkins/Textures/T_DiamondPattern_RandomVector
  }
  n6 = CurveAtlasRowParameter(CurveTime=Multiply(A=Clamp(None=@n4), B=@n5.A)) {
   atlas=&AthenaCurveAtlas01
   curve=/Game/Athena/Prototype/Blueprints/Cattus/FX/Curve_Color2
   defaultValue=113
   group=GemPattern
   parameterName=GP_GemColorCurve
  }
  n7 = Multiply(A=@n6, B=ScalarParameter() {defaultValue=0.1, group=GemPattern, parameterName=GP_DefaultGemPower})
  n8 = Multiply(A=@n6, B=ScalarParameter() {defaultValue=1, group=GemPattern, parameterName=GP_SparklePower})
  n9 = MaterialFunctionCall() {materialFunction=&CameraDirectionVector}
  n10 = Subtract(A=Multiply(A=@n5.RGB) {constB=2})
  n11 = CurveAtlasRowParameter(CurveTime=Abs(None=DotProduct(A=@n9.Result, B=@n10))) {
   atlas=&AthenaCurveAtlas01
   curve=/Game/Characters/Player/Male/Medium/Bodies/M_MED_BattleSuit/FX/BattleSuit_Blue
   defaultValue=82
   group=GemPattern
   parameterName=GP_SparkleControlCurve
  }
  n12 = ChannelMaskParameter(None=@n11) {group=GemPattern, parameterName=SparkleControlCurveChannel}
  n13 = Multiply(
   A=LinearInterpolate(A=@n12.R, B=Round(None=@n12.R), Alpha=ScalarParameter() {group=GemPattern, parameterName=GP_UseRoundedVectorValues})
   B=@n5.A
  )
  n14 = SetMaterialAttributes!(
   MaterialAttributes=@n1.MaterialAttributes
   BaseColor=@n6
   Metallic=ScalarParameter() {defaultValue=1, group=GemPattern, parameterName=GP_Metallic}
   Specular=ScalarParameter() {defaultValue=1, group=GemPattern, parameterName=GP_Specular}
   Roughness=ScalarParameter() {defaultValue=1, group=GemPattern, parameterName=GP_Roughness}
   EmissiveColor=Add(A=@n1.EmissiveColor, B=LinearInterpolate(A=@n7, B=@n8, Alpha=@n13))
   Normal=@n3.Result
  ) {attributeSetTypes [
    #3
    #4
    #5
    #6
    #2
    #1
   ]}
  n15 = FunctionInput(Preview=Constant4Vector() {constant=<rgba 1 0 0 0>}.RGBA) {bUsePreviewValueAsDefault=true, inputName=Mask, inputType=FunctionInput_Vector4}
  FunctionOutput(None=BlendMaterialAttributes(A=@n1.MaterialAttributes, B=@n14, Alpha=ChannelMaskParameter(None=@n15) {group=GemPattern, parameterName=GemPattern_FXMaskChannel}.R))
  Multiply(A=@n8, B=@n13)
 ]}
