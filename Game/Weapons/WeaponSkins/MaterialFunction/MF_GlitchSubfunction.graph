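# Material function for weapon skins that samples an input texture and produces glitched and clean versions of it.
# The texture is read at normal UVs and at UVs distorted by MF_GlitchFunction, then red, green and blue are blended separately by scalar parameters, with alpha taken from the distorted sample.
# A UseAreaMask parameter blends in multiplication by the clean red channel, while a second output passes the unmodified RGBA through.
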MaterialFunction /Game/Weapons/WeaponSkins/MaterialFunction/MF_GlitchSubfunction
graph {nodes [
  n1 = TextureCoordinate()
  n2 = FunctionInput(Preview=TextureObject() {texture=/Game/Characters/Player/Male/Medium/Heads/M_MED_SykoBlue_Head/Textures/T_Syko_Face}) {bUsePreviewValueAsDefault=true, inputType=FunctionInput_Texture2D}
  n3 = TextureSample(UVs=@n1, Tex=@n2)
  n4 = TextureSample(UVs=MaterialFunctionCall!(UV=@n1) {materialFunction=&/MF_GlitchFunction}.Glitched UV, Tex=@n2)
  n5 = LinearInterpolate(A=@n3.R, B=@n4.R, Alpha=ScalarParameter() {group=Glitch, parameterName=Glitch Red})
  n6 = LinearInterpolate(A=@n3.G, B=@n4.G, Alpha=ScalarParameter() {group=Glitch, parameterName=Glitch Green})
  n7 = ScalarParameter() {group=Glitch, parameterName=Glitch Blue}
  n8 = AppendVector(A=AppendVector(A=AppendVector(A=@n5, B=@n6), B=LinearInterpolate(A=@n3.B, B=@n4.B, Alpha=@n7)), B=@n4.A)
  FunctionOutput(None=LinearInterpolate(A=@n8, B=Multiply(A=@n8, B=@n3.R), Alpha=ScalarParameter() {group=Glitch, parameterName=UseAreaMask})) {outputName=Glitched}
  FunctionOutput(None=@n3.RGBA) {outputName=No Glitch}
 ]}
