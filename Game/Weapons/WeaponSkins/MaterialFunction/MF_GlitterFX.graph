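# Weapon-skin material function that layers optional glitter and animated sparkles onto incoming material attributes, selected by mask channels and static switches.
# Glitter blends two tilings of a direction mask by distance into noisy normals, lit by a camera/light half-vector highlight using directional light data from Custom nodes.
# It adds emissive and optionally perturbs normals.
# Animated sparkles add time-driven sine pulses from a tiled mask, optionally colored and faded by inverse Fresnel.
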MaterialFunction /Game/Weapons/WeaponSkins/MaterialFunction/MF_GlitterFX
graph {
 properties {bExposeToLibrary=true}
 nodes [
  CameraVectorWS() {desc=ViewDir}
  n1 = FunctionInput() {inputType=FunctionInput_MaterialAttributes}
  n2 = GetMaterialAttributes(None=@n1) {attributeGetTypes [#1, #2]}
  n3 = StaticSwitchParameter(
   True=MaterialFunctionCall(BaseNormal=@n2.Normal, AdditionalNormal=NamedRerouteUsage()) {materialFunction=&BlendAngleCorrectedNormals}.Result
   False=@n2.Normal
  ) {group=Glitter, parameterName=Use Glitter Normal, sortPriority=1}
  n4 = Add(A=@n2.EmissiveColor, B=NamedRerouteUsage()) {constB=0}
  n5 = FunctionInput() {bUsePreviewValueAsDefault=true, inputName=SkinFX_Mask, inputType=FunctionInput_Vector4, sortPriority=1, previewValue=<1 1 0 1>}
  n6 = NamedRerouteDeclaration(None=@n5) {name=FX Map, nodeColor=<rgba 0 1 0.0625 1>}
  n7 = StaticSwitchParameter(
   True=BlendMaterialAttributes(
    A=@n1
    B=SetMaterialAttributes!(MaterialAttributes=@n2.MaterialAttributes, Normal=@n3, EmissiveColor=@n4) {attributeSetTypes [#1, #2]}
    Alpha=ChannelMaskParameter(None=@n6) {group=Glitter, parameterName=Glitter Channel, sortPriority=0}.R
   )
   False=@n1
  ) {group=Skin FX, parameterName=Use Glitter}
  n8 = GetMaterialAttributes(None=@n7) {attributeGetTypes [#2]}
  n9 = ChannelMaskParameter(None=NamedRerouteUsage()) {group=GlitterAnimated, maskChannel=Green, parameterName=Glitter Animated Channel, sortPriority=0, defaultValue=<rgba 0 1 0 0>}
  n10 = Add(A=@n8.EmissiveColor, B=Multiply(A=@n9.R, B=NamedRerouteUsage()))
  FunctionOutput(None=StaticSwitchParameter(
    True=SetMaterialAttributes!(MaterialAttributes=@n8.MaterialAttributes, EmissiveColor=@n10) {attributeSetTypes [#2]}
    False=@n7
   ) {group=Skin FX, parameterName=Use Glitter Animated})
  n11 = TextureCoordinate()
  n12 = ScalarParameter() {defaultValue=4, group=GlitterAnimated, parameterName=Sparkle Tiling, sliderMax=32}
  LinearInterpolate(
   A=TextureSample(UVs=Multiply(A=@n11, B=@n12)) {samplerType=SAMPLERTYPE_Masks, texture=/Game/Weapons/WeaponSkins/Textures/T_SparkleMask}.RGB
   Alpha=MaterialFunctionCall!(
    Blend Range=ScalarParameter() {defaultValue=50, group=Glitter, parameterName=Blend Distance, sortPriority=4}
    Start Offset=ScalarParameter() {defaultValue=-150, group=Glitter, parameterName=Blend Offset, sortPriority=5}
   ) {materialFunction=/Engine/Functions/Engine_MaterialFunctions03/Particles/Distance_Blend}.Result
  )
  n13 = TextureSampleParameter2D(UVs=Multiply(
    A=TextureCoordinate()
    B=ScalarParameter() {defaultValue=16, group=Glitter, parameterName=Detail Tiling, sliderMax=64, sliderMin=8, sortPriority=2}
   ) {constB=3}) {
   group=Glitter
   parameterName=Glitter Direction Mask
   samplerType=SAMPLERTYPE_Masks
   sortPriority=0
   texture=/Game/Weapons/WeaponSkins/Textures/T_GlitterDirectionMask
  }
  n14 = TextureSampleParameter2D(UVs=Multiply(A=TextureCoordinate(), B=ScalarParameter() {defaultValue=4, group=Glitter, parameterName=Tiling, sliderMax=16, sliderMin=1, sortPriority=3})) {
   group=Glitter
   parameterName=Glitter Direction Mask
   samplerType=SAMPLERTYPE_Masks
   sortPriority=0
   texture=&T_GlitterDirectionMask
  }
  n15 = LinearInterpolate(
   A=@n13.RGBA
   B=@n14.RGBA
   Alpha=MaterialFunctionCall(
    Blend Range=ScalarParameter() {defaultValue=50, group=Glitter, parameterName=Blend Distance, sortPriority=4}
    Start Offset=ScalarParameter() {defaultValue=-150, group=Glitter, parameterName=Blend Offset, sortPriority=5}
   ) {materialFunction=&Distance_Blend}.Result
  )
  NamedRerouteDeclaration(None=ComponentMask(None=@n15) {a=true}) {name=Glitter Alpha, nodeColor=<rgba 0 1 0.5312 1>}
  n16 = NamedRerouteDeclaration(None=ComponentMask(None=@n15) {b=true, g=true, r=true}) {name=Full Random Normal Dir, nodeColor=<rgba 1 0 0.4922 1>}
  n17 = Multiply(
   A=Add(A=Multiply(A=@n16) {constB=2}) {constB=-1}
   B=ScalarParameter() {defaultValue=0.5, group=Glitter, parameterName=Normal Noise Intensity, sliderMax=2, sortPriority=6}
  ) {constB=0.5}
  n18 = Normalize(VectorInput=Add(A=NamedRerouteDeclaration(None=@n17) {name=Random Normal Dir, nodeColor=<rgba 0 0.3501 1 1>}, B=VertexNormalWS()))
  NamedRerouteDeclaration(None=MaterialFunctionCall(
    Normal=Transform(None=@n18) {transformSourceType=TRANSFORMSOURCE_World, transformType=TRANSFORM_Tangent}
    Flatness=ScalarParameter() {group=Glitter, parameterName=Glitter Normal Flattness, sliderMax=1, sortPriority=7}
   ) {materialFunction=&FlattenNormal}.Result) {name=Glitter Normal, nodeColor=<rgba 1 0.1172 0 1>}
  n19 = Custom() {code="return View.DirectionalLightColor;", description=DirectionalLightColor, inputs []}
  n20 = ScalarParameter() {defaultValue=5, group=Glitter, parameterName=Boost Light, sliderMax=16, sortPriority=10}
  n21 = NamedRerouteUsage()
  n22 = Constant3Vector() {constant=<rgba 0 0 -1 0>}
  n23 = Normalize(VectorInput=Add(
    A=Custom() {code="return View.DirectionalLightDirection;", description=DirectionalLightVector, inputs []}
    B=Transform(None=@n22) {desc=Camera Vector Forward, transformSourceType=TRANSFORMSOURCE_Camera}
   )) {desc="L + V  average"}
  n24 = Normalize(VectorInput=Add(A=NamedRerouteUsage(), B=VertexNormalWS())) {desc=GlitterDirection}
  n25 = Multiply(
   A=SmoothStep(
    Min=ScalarParameter() {defaultValue=0.95, group=Glitter, parameterName=Max Glitter, sliderMax=1, sortPriority=8}
    Value=DotProduct(A=@n23, B=@n24)
   )
   B=NamedRerouteUsage()
  )
  n26 = Multiply(
   A=Multiply(
    A=Power(Base=@n21) {constExponent=3}
    B=ScalarParameter() {defaultValue=5, group=Glitter, parameterName=Boost Glitter Color, sliderMax=16, sortPriority=9}
   ) {constB=5}
   B=@n25
  )
  NamedRerouteDeclaration(None=Multiply(
    A=Multiply(A=@n19, B=@n20) {constB=3}
    B=StaticSwitchParameter(True=@n26, False=@n25) {defaultValue=true, group=Glitter, parameterName=Use Random Color, sortPriority=1}
   )) {name=Emissiv Color, nodeColor=<rgba 0 0.2687 1 1>}
  n27 = Constant3Vector() {constant=<rgba 1 0 0.1682 0>}
  n28 = Constant3Vector() {constant=<rgba 0 0.1742 1 0>}
  n29 = Constant3Vector() {constant=<rgba 0 1 0.2352 0>}
  n30 = ScalarParameter() {defaultValue=16, group=GlitterAnimated, parameterName=Sparkle Detail Tiling, sliderMax=64, sliderMin=8, sortPriority=2}
  n31 = Multiply(A=TextureCoordinate(), B=@n30) {constB=3}
  n32 = TextureSampleParameter2D(UVs=@n31) {
   group=GlitterAnimated
   parameterName=Sparkle Mask
   samplerType=SAMPLERTYPE_Masks
   sortPriority=0
   texture=&T_SparkleMask
  }
  n33 = Frac(None=@n31)
  n34 = MaterialFunctionCall(X=@n32.R, Y=@n32.R, Z=Divide(A=Subtract(A=@n31, B=@n33), B=@n30)) {desc=could be removed if no blending is needed, materialFunction=&MakeFloat3}
  n35 = ComponentMask(None=@n34.Result) {g=true}
  n36 = ComponentMask(None=@n34.Result) {b=true}
  n37 = Frac(None=Multiply(A=Add(A=@n35, B=@n36)) {constB=3.333})
  n38 = Add(A=ComponentMask(None=@n34.Result) {b=true}, B=Time())
  n39 = ScalarParameter() {defaultValue=0.2, group=GlitterAnimated, parameterName=Time Scale, sliderMax=2, sortPriority=3}
  n40 = Frac(None=Add(A=ComponentMask(None=@n34.Result) {g=true}, B=Multiply(A=@n38, B=@n39) {constB=0.5}))
  n41 = Multiply(A=Sine(None=MaterialFunctionCall(UV Channel=@n40) {materialFunction=&LinearGradient}.VGradient)) {constB=0.5}
  n42 = Multiply(
   A=ComponentMask(None=@n34.Result) {r=true}
   B=Power(
    Base=Add(A=@n41) {constB=0.5}
    Exp=ScalarParameter() {defaultValue=32, group=GlitterAnimated, parameterName=Fade Curve Exponent, sortPriority=4}
   ) {constExponent=32}
  )
  n43 = Saturate(None=Multiply(A=Power(Base=@n42) {constExponent=4}) {constB=4})
  n44 = StaticSwitchParameter(
   True=Multiply(
    A=MaterialFunctionCall(A=@n27, B=@n28, C=@n29, Alpha=@n37) {materialFunction=&3ColorBlend}.Result
    B=@n43
   )
   False=@n43
  ) {group=GlitterAnimated, parameterName=Use Colored Sparkles, sortPriority=1}
  n45 = ScalarParameter() {defaultValue=10, group=GlitterAnimated, parameterName=Emissive Intensity, sliderMax=20, sortPriority=5}
  n46 = OneMinus(None=Fresnel() {baseReflectFraction=0, exponent=1})
  NamedRerouteDeclaration(None=Multiply(A=Multiply(A=@n44, B=@n45), B=Power(Base=@n46) {constExponent=3})) {name=Glitter Animated Emissive, nodeColor=<rgba 1 0.2874 0 1>}
  ReflectionVectorWS()
 ]
}
