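# Material function for weapon and character skin FX that fakes a glow from inside the mesh on incoming material attributes.
# A pre-skinned position relative to a glow point shifted along the camera vector, plus scaled pixel or vertex normals, is normalized, dotted with the camera vector and powered.
# That falloff lerps In/Out colors that add to emissive and tint base and subsurface color.
# A mask input blends the result; a static switch bypasses it.
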MaterialFunction /Game/Weapons/WeaponSkins/MaterialFunction/MF_InteriorGlow
graph {nodes [
  n1 = PreSkinnedPosition()
  n2 = VectorParameter() {group=InteriorEmissiveGlow, parameterName=InteriorGlowLocation, defaultValue=<rgba 0 0 150 1>}
  n3 = CameraVectorWS()
  n4 = Multiply(A=@n3, B=ScalarParameter() {defaultValue=-0.7177, group=InteriorEmissiveGlow, parameterName=InteriorDistanceBehindCharacter})
  n5 = PixelNormalWS()
  n6 = VertexNormalWS()
  n7 = Add(
   A=Subtract(A=VertexInterpolator(VS=@n1).PS, B=Add(A=@n2.RGB, B=@n4))
   B=Multiply(
    A=StaticSwitchParameter(True=@n5, False=@n6) {group=InteriorEmissiveGlow, parameterName=InteriorGlowUsePixelNormals}
    B=ScalarParameter() {defaultValue=211.6945, group=InteriorEmissiveGlow, parameterName=InteriorGlowNormalContribution}
   )
  )
  n8 = Power(
   Base=DotProduct(A=Normalize(VectorInput=@n7), B=@n3)
   Exp=ScalarParameter() {defaultValue=4.028, group=InteriorEmissiveGlow, parameterName=FalloffInteriorglow}
  )
  n9 = Saturate(None=@n8)
  n10 = LinearInterpolate(
   A=VectorParameter() {group=InteriorEmissiveGlow, parameterName=InteriorGlowSubSurface_Out, defaultValue=<rgba 1 1 1 1>}.RGB
   B=VectorParameter() {group=InteriorEmissiveGlow, parameterName=InteriorGlowSubSurface_In, defaultValue=<rgba 1 1 1 1>}.RGB
   Alpha=@n9
  )
  Add(B=@n10)
  n11 = FunctionInput() {inputType=FunctionInput_MaterialAttributes}
  n12 = GetMaterialAttributes(None=@n11) {attributeGetTypes [#1, #2, #3]}
  n13 = VectorParameter() {group=InteriorEmissiveGlow, parameterName=InteriorGlowEmissive_Out, defaultValue=<rgba 0 0 0 1>}
  n14 = VectorParameter() {group=InteriorEmissiveGlow, parameterName=InteriorGlowEmissive_In, defaultValue=<rgba 0 0 0 1>}
  n15 = VectorParameter() {group=InteriorEmissiveGlow, parameterName=InteriorGlowBaseColor_Out, defaultValue=<rgba 1 1 1 1>}
  n16 = VectorParameter() {group=InteriorEmissiveGlow, parameterName=InteriorGlowBaseColor_In, defaultValue=<rgba 1 1 1 1>}
  n17 = SetMaterialAttributes!(
   MaterialAttributes=@n12.MaterialAttributes
   EmissiveColor=Add(A=@n12.EmissiveColor, B=LinearInterpolate(A=@n13.RGB, B=@n14.RGB, Alpha=@n8))
   BaseColor=Multiply(A=@n12.BaseColor, B=LinearInterpolate(A=@n15.RGB, B=@n16.RGB, Alpha=@n9))
   SubsurfaceColor=Multiply(A=@n12.SubsurfaceColor, B=@n10)
  ) {attributeSetTypes [#1, #2, #3]}
  n18 = FunctionInput() {bUsePreviewValueAsDefault=true, inputName=Mask, inputType=FunctionInput_Scalar, sortPriority=3, previewValue=<1 0 0 1>}
  FunctionOutput(None=StaticSwitchParameter(True=BlendMaterialAttributes(A=@n11, B=@n17, Alpha=@n18), False=@n11) {group=Skin FX, parameterName=UseInteriorEmissiveGlow})
 ]}
