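# MF_MinWidth is a weapon-skin material function that outputs a camera-facing offset vector to widen thin strip geometry, apparently for World Position Offset.
# The direction crosses the camera vector with the TangentBasis V axis, with a smoothstepped normal-to-camera dot product flipping its sign.
# Centered U spreads the two sides apart.
# Width sums WorldWidth and PixelWidth, the latter scaled by depth, FOV and viewport size to hold on-screen thickness.
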MaterialFunction /Game/Weapons/WeaponSkins/MaterialFunction/MF_MinWidth
graph {nodes [
  n1 = Subtract(A=CameraPositionWS(), B=WorldPosition().XYZ)
  n2 = Normalize(VectorInput=CrossProduct(A=@n1, B=MaterialFunctionCall() {materialFunction=&TangentBasis}.Y V))
  n3 = DotProduct(A=VertexNormalWS(), B=@n1)
  n4 = Constant() {r=0.01}
  n5 = Multiply(A=Subtract(A=MaterialFunctionCall(Alpha=@n3, Max=@n4) {materialFunction=&SmoothStep}.Result) {constB=0.5}) {constB=2}
  n6 = Subtract(A=ComponentMask(None=TextureCoordinate()) {r=true}, B=Constant() {r=0.5}) {constB=0.5}
  n7 = Multiply(
   A=Custom!(None) {code="return GetTanHalfFieldOfView();", description=TanHalfFieldOfView, outputType=CMOT_Float1}
   B=ComponentMask(None=PixelDepth().R) {g=true}
  )
  n8 = ComponentMask(None=ViewSize()) {g=true}
  n9 = Constant() {r=1}
  n10 = Add(
   A=Multiply(A=Divide(A=@n7, B=@n8), B=FunctionInput(Preview=@n9) {bUsePreviewValueAsDefault=true, inputName=PixelWidth}) {constB=1.414}
   B=FunctionInput(Preview=Constant()) {bUsePreviewValueAsDefault=true, inputName=WorldWidth}
  )
  FunctionOutput(None=Multiply(A=Multiply(A=@n2, B=@n5), B=Multiply(A=@n6, B=@n10)))
 ]}
