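# Weapon-skin material function that layers an animated, distorted pattern onto incoming material attributes.
# A panning noise texture warps the main texture's UVs, which pan only in X (Y speed is zeroed) and take optional bump-offset parallax.
# The pattern's red channel picks a multiplier that tints BaseColor and drives added, Fresnel-shaped emissive, both limited by the Mask input.
# A second distortion branch is left unconnected.
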MaterialFunction /Game/Weapons/WeaponSkins/MaterialFunction/MF_PanAndDistort
graph {nodes [
  n1 = TextureCoordinate()
  n2 = LinearInterpolate(
   A=@n1
   B=BumpOffset(
    Coordinate=@n1
    Height=ScalarParameter() {group=PanAndDistort, parameterName=Height}
    HeightRatioInput=ScalarParameter() {defaultValue=0.1, group=PanAndDistort, parameterName=ParallaxHeightRatio}
   )
   Alpha=ScalarParameter() {group=PanAndDistort, parameterName=useBumpOffset}
  )
  n3 = Multiply(A=@n2, B=ScalarParameter() {defaultValue=2, group=PanAndDistort, parameterName=PAD_MainTexture-Tile})
  n4 = ScalarParameter() {defaultValue=-0.02, group=PanAndDistort, parameterName=PAD_MainTexture-SpeedX}
  n5 = ScalarParameter() {group=PanAndDistort, parameterName=PAD_MainTexture-SpeedY}
  n6 = Multiply(
   A=Multiply(A=MaterialFunctionCall(X=@n4, Y=@n5) {materialFunction=&MakeFloat2}.Result, B=Time())
   B=Constant2Vector() {r=1}
  )
  n7 = MaterialFunctionCall(
   X=ScalarParameter() {defaultValue=4, group=PanAndDistort, parameterName=PAD_Distort-TileX}
   Y=ScalarParameter() {defaultValue=3, group=PanAndDistort, parameterName=PAD_Distort-TileY}
  ) {materialFunction=&MakeFloat2}
  n8 = MaterialFunctionCall(
   X=ScalarParameter() {defaultValue=0.03, group=PanAndDistort, parameterName=PAD_Distort-SpeedX}
   Y=ScalarParameter() {defaultValue=-0.015, group=PanAndDistort, parameterName=PAD_Distort-SpeedY}
  ) {materialFunction=&MakeFloat2}
  n9 = Time()
  n10 = TextureSampleParameter2D(UVs=Add(A=Multiply(A=@n2, B=@n7.Result), B=Multiply(A=@n8.Result, B=@n9))) {
   group=PanAndDistort
   parameterName=PAD_DistortUVTexture
   samplerSource=SSM_Wrap_WorldGroupSettings
   samplerType=SAMPLERTYPE_LinearGrayscale
   texture=/Game/Weapons/WeaponSkins/Wraps/Textures/LowResBlurredNoise_Wrap
  }
  n11 = ScalarParameter() {defaultValue=0.4, group=PanAndDistort, parameterName=PAD_Distort-Amount}
  n12 = TextureSampleParameter2D(UVs=Add(A=Add(A=@n3, B=@n6), B=Multiply(A=@n10.R, B=@n11))) {
   group=PanAndDistort
   parameterName=PAD_MainTexture
   samplerSource=SSM_Wrap_WorldGroupSettings
   samplerType=SAMPLERTYPE_LinearColor
   texture=/Game/Weapons/WeaponSkins/Wraps/Textures/LineSwirl/T_Wrap_LineSwirl_M2
  }
  n13 = LinearInterpolate(
   A=ScalarParameter() {defaultValue=1, group=PanAndDistort, parameterName=PAD_MultiplyA}
   B=ScalarParameter() {defaultValue=0.25, group=PanAndDistort, parameterName=PAD_MultiplyB}
   Alpha=@n12.R
  )
  n14 = GetMaterialAttributes(None=FunctionInput() {inputName=Material, inputType=FunctionInput_MaterialAttributes}) {attributeGetTypes [#1, #2, #3]}
  Add(A=@n14.BaseColor, B=@n13)
  n15 = Multiply(A=@n14.BaseColor, B=@n13)
  n16 = FunctionInput(Preview=Constant() {r=1}) {bUsePreviewValueAsDefault=true, inputName=Mask, inputType=FunctionInput_Scalar, sortPriority=-19}
  n17 = LinearInterpolate(
   A=VectorParameter() {group=PanAndDistort, parameterName=PanDistortEmissiveColor, defaultValue=<rgba 0 0 0 1>}.RGB
   B=@n14.BaseColor
   Alpha=ScalarParameter() {group=PanAndDistort, parameterName=PAD_UseBaseColorAsEmissive}
  )
  n18 = Multiply(A=@n13, B=@n17)
  n19 = Fresnel(ExponentIn=ScalarParameter() {defaultValue=2, group=PanAndDistort, parameterName=DIstortFresnelExponent}, Normal=Transform(None=@n14.Normal))
  n20 = OneMinus(None=@n19)
  n21 = ScalarParameter() {group=PanAndDistort, parameterName=InvertFresnel}
  n22 = Multiply(A=@n17, B=Power(Base=LinearInterpolate(A=@n19, B=@n20, Alpha=@n21)))
  n23 = Multiply(
   A=@n16
   B=Multiply(A=Multiply(A=@n18, B=@n22), B=ScalarParameter() {defaultValue=10, group=PanAndDistort, parameterName=PanEmissiveStrength})
  )
  FunctionOutput(None=SetMaterialAttributes!(
    MaterialAttributes=@n14.MaterialAttributes
    BaseColor=LinearInterpolate(A=@n14.BaseColor, B=@n15, Alpha=@n16)
    EmissiveColor=Add(A=@n14.EmissiveColor, B=@n23)
   ) {attributeSetTypes [#1, #3]})
  n24 = TextureCoordinate()
  n25 = MaterialFunctionCall(
   X=ScalarParameter() {defaultValue=5, group=PanAndDistort, parameterName=PAD_Distort2-TileX}
   Y=ScalarParameter() {defaultValue=5, group=PanAndDistort, parameterName=PAD_Distort2-TileY}
  ) {materialFunction=&MakeFloat2}
  n26 = MaterialFunctionCall(
   X=ScalarParameter() {defaultValue=0.03, group=PanAndDistort, parameterName=PAD_Distort2-SpeedX}
   Y=ScalarParameter() {defaultValue=-0.03, group=PanAndDistort, parameterName=PAD_Distort2-SpeedY}
  ) {materialFunction=&MakeFloat2}
  n27 = Time()
  n28 = TextureSampleParameter2D(UVs=Add(A=Multiply(A=@n24, B=@n25.Result), B=Multiply(A=@n26.Result, B=@n27))) {
   group=PanAndDistort
   parameterName=PAD_DistortUVTexture
   samplerSource=SSM_Wrap_WorldGroupSettings
   samplerType=SAMPLERTYPE_LinearGrayscale
   texture=&LowResBlurredNoise_Wrap
  }
  Multiply(A=@n28.RGB, B=ScalarParameter() {defaultValue=0.25, group=PanAndDistort, parameterName=PAD_Distort2-Amount})
  TextureCoordinate()
 ]}
