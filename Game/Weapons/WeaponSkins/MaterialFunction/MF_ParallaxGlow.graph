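# Weapon-skin material function that adds a fake-depth parallax glow to incoming material attributes within a masked area.
# It multiplies two BumpOffset mask samples at separate depths, the edge one at a chosen mip, into a gradient blending two colors.
# The masked glow fades by view-angle Fresnel and adds to Emissive; a switch, off by default, similarly darkens BaseColor.
# Hand-built parallax UV math and a curve-atlas lookup are left unconnected.
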MaterialFunction /Game/Weapons/WeaponSkins/MaterialFunction/MF_ParallaxGlow
graph {nodes [
  n1 = TextureCoordinate()
  n2 = VertexInterpolator(VS=PreSkinnedPosition())
  n3 = CameraPositionWS()
  n4 = Normalize(VectorInput=Subtract(A=@n2.PS, B=TransformPosition(None=@n3) {transformSourceType=TRANSFORMPOSSOURCE_World}))
  n5 = ScalarParameter() {defaultValue=0.1, group=ParallaxGlow, parameterName=PG_Depth_Edges}
  n6 = MaterialFunctionCall(Float3=Multiply(A=@n4, B=@n5)) {materialFunction=&BreakOutFloat3Components}
  n7 = Multiply(A=@n6.R) {constB=-1}
  n8 = ComponentMask(None=@n2.PS) {r=true}
  n9 = Clamp(None=Round(None=@n8))
  Add(
   A=@n1
   B=MaterialFunctionCall(X=LinearInterpolate(A=@n6.R, B=@n7, Alpha=@n9), Y=@n6.B) {materialFunction=&MakeFloat2}.Result
  )
  n10 = TextureCoordinate()
  n11 = VertexInterpolator(VS=PreSkinnedPosition())
  n12 = CameraPositionWS()
  n13 = Normalize(VectorInput=Subtract(A=@n11.PS, B=TransformPosition(None=@n12) {transformSourceType=TRANSFORMPOSSOURCE_World}))
  n14 = ScalarParameter() {defaultValue=0.1, group=ParallaxGlow, parameterName=PG_BCDepth}
  n15 = TextureObjectParameter() {
   group=ParallaxGlow
   parameterName=PG_EmissiveMask
   samplerType=SAMPLERTYPE_Masks
   texture=/Game/Gadgets/Assets/VinderTech_GliderChute/Glider_Galileo_Rocket/Textures/T_Galileo_Rocket_Glider_Back_FX
  }
  n16 = TextureSample!(UVs=@n10, Tex=@n15, Level, Apply View MipBias) {
   mipValueMode=TMVM_MipLevel
   samplerType=SAMPLERTYPE_Masks
   texture=&T_Galileo_Rocket_Glider_Back_FX
  }
  n17 = MaterialFunctionCall(Float3=Multiply(A=@n13, B=LinearInterpolate(B=@n14, Alpha=@n16.G)) {constB=12}) {materialFunction=&BreakOutFloat3Components}
  n18 = Multiply(A=@n17.R) {constB=-1}
  n19 = ComponentMask(None=@n11.PS) {r=true}
  n20 = Clamp(None=Round(None=@n19))
  Add(
   A=@n10
   B=MaterialFunctionCall(X=LinearInterpolate(A=@n17.R, B=@n18, Alpha=@n20), Y=@n17.B) {materialFunction=&MakeFloat2}.Result
  )
  n21 = ScalarParameter() {defaultValue=0.1, group=ParallaxGlow, parameterName=PG_Depth}
  n22 = TextureObjectParameter() {
   group=ParallaxGlow
   parameterName=PG_EmissiveMask
   samplerType=SAMPLERTYPE_Masks
   texture=&T_Galileo_Rocket_Glider_Back_FX
  }
  n23 = TextureSample(UVs=@n1, Tex=@n22) {
   mipValueMode=TMVM_MipLevel
   samplerType=SAMPLERTYPE_Masks
   texture=&T_Galileo_Rocket_Glider_Back_FX
  }
  n24 = MaterialFunctionCall(Float3=Multiply(A=@n4, B=LinearInterpolate(B=@n21, Alpha=@n23.G)) {constB=12}) {materialFunction=&BreakOutFloat3Components}
  n25 = Multiply(A=@n24.R) {constB=-1}
  Add(
   A=@n1
   B=MaterialFunctionCall(X=LinearInterpolate(A=@n24.R, B=@n25, Alpha=@n9), Y=@n24.B) {materialFunction=&MakeFloat2}.Result
  )
  n26 = ScalarParameter() {defaultValue=0.1, group=ParallaxGlow, parameterName=PG_BCDepth_Edges}
  n27 = MaterialFunctionCall(Float3=Multiply(A=@n13, B=@n26)) {materialFunction=&BreakOutFloat3Components}
  n28 = Multiply(A=@n27.R) {constB=-1}
  Add(
   A=@n10
   B=MaterialFunctionCall(X=LinearInterpolate(A=@n27.R, B=@n28, Alpha=@n20), Y=@n27.B) {materialFunction=&MakeFloat2}.Result
  )
  Ceil(None=@n8)
  Ceil(None=@n19)
  n29 = Constant()
  n30 = TextureSample(UVs=BumpOffset(Coordinate=@n1, Height=@n29, HeightRatioInput=@n21), Tex=@n22) {
   mipValueMode=TMVM_MipLevel
   samplerType=SAMPLERTYPE_Masks
   texture=&T_Galileo_Rocket_Glider_Back_FX
  }
  n31 = ChannelMaskParameter(None=@n30.RGBA) {group=ParallaxGlow, parameterName=PG_GradientChannel}
  CurveAtlasRowParameter(CurveTime=@n31.R) {
   atlas=&AthenaCurveAtlas01
   curve=/Game/Athena/LinearColorCurves/Curve_Galileo_Hologram
   defaultValue=171
   group=ParallaxGlow
   parameterName=PG_ParallaxGlowColorCurve
  }
  n32 = GetMaterialAttributes(None=FunctionInput() {inputName=Materials, inputType=FunctionInput_MaterialAttributes}) {attributeGetTypes [#1, #2, #3]}
  n33 = TextureSample(UVs=@n1, Tex=@n22) {
   mipValueMode=TMVM_MipLevel
   samplerType=SAMPLERTYPE_Masks
   texture=&T_Galileo_Rocket_Glider_Back_FX
  }
  n34 = VectorParameter() {group=ParallaxGlow, parameterName=PG_Color_02, defaultValue=<rgba 0 0 0 1>}
  n35 = VectorParameter() {group=ParallaxGlow, parameterName=PG_Color_01, defaultValue=<rgba 0 0.5816 1 1>}
  n36 = TextureSample(
   UVs=BumpOffset(Coordinate=@n1, Height=@n29, HeightRatioInput=@n5)
   Tex=@n22
   Level=ScalarParameter() {defaultValue=4, group=ParallaxGlow, parameterName=MipLevel}
  ) {
   mipValueMode=TMVM_MipLevel
   samplerType=SAMPLERTYPE_Masks
   texture=&T_Galileo_Rocket_Glider_Back_FX
  }
  n37 = Multiply(A=@n31.R, B=ChannelMaskParameter(None=@n36.RGBA) {group=ParallaxGlow, parameterName=PG_GradientChannel}.R)
  n38 = Multiply(
   A=Multiply(
    A=ChannelMaskParameter(None=@n33.RGBA) {group=ParallaxGlow, parameterName=PG_MaskChannel}.R
    B=LinearInterpolate(A=@n34.RGB, B=@n35.RGB, Alpha=@n37)
   )
   B=ScalarParameter() {defaultValue=1, group=ParallaxGlow, parameterName=PG_GlowStrength}
  )
  n39 = CameraVectorWS()
  n40 = VertexNormalWS()
  n41 = Clamp(None=Clamp(None=DotProduct(A=@n39, B=@n40)), Min=ScalarParameter() {group=ParallaxGlow, parameterName=PG_MaxFresnelEdgeDarkness})
  n42 = ChannelMaskParameter(None=FunctionInput() {bUsePreviewValueAsDefault=true, inputName=Mask, inputType=FunctionInput_Vector4}) {group=ParallaxGlow, parameterName=PG_AreaMaskChannel}
  n43 = LinearInterpolate(A=@n32.EmissiveColor, B=Add(A=@n32.EmissiveColor, B=Multiply(A=@n38, B=@n41)), Alpha=@n42.R)
  n44 = TextureSample(UVs=@n10, Tex=@n15) {
   mipValueMode=TMVM_MipLevel
   samplerType=SAMPLERTYPE_Masks
   texture=&T_Galileo_Rocket_Glider_Back_FX
  }
  n45 = Multiply(A=@n32.BaseColor, B=ScalarParameter() {defaultValue=0.25, group=ParallaxGlow, parameterName=PG_BaseColorDarken})
  n46 = Constant()
  n47 = TextureSample(UVs=BumpOffset(Coordinate=@n10, Height=@n46, HeightRatioInput=@n14), Tex=@n15) {
   mipValueMode=TMVM_MipLevel
   samplerType=SAMPLERTYPE_Masks
   texture=&T_Galileo_Rocket_Glider_Back_FX
  }
  n48 = TextureSample(
   UVs=BumpOffset(Coordinate=@n10, Height=@n46, HeightRatioInput=@n26)
   Tex=@n15
   Level=ScalarParameter() {defaultValue=4, group=ParallaxGlow, parameterName=BaseColor_MipLevel}
  ) {
   mipValueMode=TMVM_MipLevel
   samplerType=SAMPLERTYPE_Masks
   texture=&T_Galileo_Rocket_Glider_Back_FX
  }
  n49 = Multiply(
   A=ChannelMaskParameter(None=@n47.RGBA) {group=ParallaxGlow, parameterName=PG_GradientChannel}.R
   B=ChannelMaskParameter(None=@n48.RGBA) {group=ParallaxGlow, parameterName=PG_GradientChannel}.R
  )
  n50 = LinearInterpolate(
   A=@n32.BaseColor
   B=Multiply(
    A=ChannelMaskParameter(None=@n44.RGBA) {group=ParallaxGlow, parameterName=PG_MaskChannel}.R
    B=LinearInterpolate(A=@n32.BaseColor, B=@n45, Alpha=@n49)
   )
   Alpha=@n42.R
  )
  FunctionOutput(None=SetMaterialAttributes!(
    MaterialAttributes=@n32.MaterialAttributes
    EmissiveColor=StaticSwitchParameter(True=@n43, False=@n32.EmissiveColor) {defaultValue=true, group=ParallaxGlow, parameterName=PG_UseEmissiveColorParallax}
    BaseColor=StaticSwitchParameter(True=@n50, False=@n32.BaseColor) {group=ParallaxGlow, parameterName=PG_UseBaseColorParallax}
   ) {attributeSetTypes [#1, #3]})
  ScalarParameter() {defaultValue=0.1, group=ParallaxGlow, parameterName=PG_Depth_EdgesHeightRatio}
 ]}
