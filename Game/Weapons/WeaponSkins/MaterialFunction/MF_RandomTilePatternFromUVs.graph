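# Material function for weapon skins that generates two layers of random tile patterns from separate UV inputs, each with its own seed, threshold and edge width.
# Both layers are packed into one four-channel vector, whose floored cells feed a dot-sine-frac hash for per-tile random values.
# Floored random-plus-threshold selects tiles, multiplied by a square edge mask from centered fractional UVs.
# Outputs each layer's mask and raw random value.
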MaterialFunction /Game/Weapons/WeaponSkins/MaterialFunction/MF_RandomTilePatternFromUVs
graph {nodes [
  n1 = Constant() {r=0.5}
  n2 = TextureCoordinate()
  n3 = AppendVector(
   A=FunctionInput(Preview=Multiply(A=@n2) {constB=10}) {bUsePreviewValueAsDefault=true, inputName=UVs_Layer1, inputType=FunctionInput_Vector2}
   B=FunctionInput(Preview=Multiply(A=@n2) {constB=25}) {bUsePreviewValueAsDefault=true, inputName=UVs_Layer2, inputType=FunctionInput_Vector2, sortPriority=1}
  )
  n4 = Floor(None=@n3)
  n5 = AppendVector(
   A=DotProduct(A=ComponentMask(None=@n4) {g=true, r=true}, B=Constant2Vector() {g=78.1651, r=12.9265})
   B=DotProduct(A=ComponentMask(None=@n4) {a=true, b=true}, B=Constant2Vector() {g=21.1655, r=5.1541})
  )
  n6 = FunctionInput(Preview=Constant() {r=7562.4819}) {bUsePreviewValueAsDefault=true, inputName=Seed_Layer1, inputType=FunctionInput_Scalar, sortPriority=2}
  n7 = FunctionInput(Preview=Constant() {r=2241.5586}) {bUsePreviewValueAsDefault=true, inputName=Seed_Layer2, inputType=FunctionInput_Scalar, sortPriority=3}
  n8 = Frac(None=Multiply(A=Sine(None=@n5), B=AppendVector(A=@n6, B=@n7)))
  n9 = ComponentMask(None=@n8) {r=true}
  n10 = Add(
   A=FunctionInput(Preview=@n1) {bUsePreviewValueAsDefault=true, inputName=Threshold_Layer1, inputType=FunctionInput_Scalar, sortPriority=4}
   B=@n9
  )
  n11 = Multiply(A=Subtract(A=Frac(None=@n3)) {constB=0.5}) {constB=2}
  n12 = FunctionInput() {bUsePreviewValueAsDefault=true, inputName=EdgeMaskWidth_Layer1, inputType=FunctionInput_Scalar, sortPriority=10, previewValue=<1 0 0 1>}
  n13 = AppendVector(A=@n12, B=@n12)
  n14 = FunctionInput() {bUsePreviewValueAsDefault=true, inputName=EdgeMaskWidth_Layer2, inputType=FunctionInput_Scalar, sortPriority=11, previewValue=<1 0 0 1>}
  n15 = AppendVector(A=@n14, B=@n14)
  n16 = Ceil(None=Subtract(A=Abs(None=@n11), B=AppendVector(A=@n13, B=@n15)) {constB=0.95})
  n17 = OneMinus(None=Add(A=ComponentMask(None=@n16) {r=true}, B=ComponentMask(None=@n16) {g=true}))
  FunctionOutput(None=Multiply(A=Floor(None=@n10), B=Saturate(None=@n17))) {outputName=Layer1_Mask}
  n18 = ComponentMask(None=@n8) {g=true}
  n19 = Add(
   A=@n18
   B=FunctionInput(Preview=@n1) {bUsePreviewValueAsDefault=true, inputName=Threshold_Layer2, inputType=FunctionInput_Scalar, sortPriority=5}
  )
  n20 = OneMinus(None=Add(A=ComponentMask(None=@n16) {b=true}, B=ComponentMask(None=@n16) {a=true}))
  FunctionOutput(None=Multiply(A=Floor(None=@n19), B=Saturate(None=@n20))) {outputName=Layer2_Mask, sortPriority=1}
  FunctionOutput(None=@n9) {outputName=Layer1_RandomValue, sortPriority=3}
  FunctionOutput(None=@n18) {outputName=Layer2_RandomValue, sortPriority=4}
 ]}
