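# Weapon skin material function that adds four layers of animated, randomly tiled patterns to Base Color.
# Each layer's UVs are tiled, sine-swayed, panned and optionally parallax-offset in the vertex shader.
# MF_RandomTilePatternFromUVs supplies per-tile masks and random values that vary scale, rotation speed and opacity of each layer's texture.
# Layers are weighted by shared mask-texture channels, summed, and optionally blended by the Alpha FX mask input.
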MaterialFunction /Game/Weapons/WeaponSkins/MaterialFunction/MF_RandomTileAnimatedPattern
graph {nodes [
  n1 = FunctionInput() {inputName=Material, inputType=FunctionInput_MaterialAttributes}
  n2 = GetMaterialAttributes(None=@n1) {attributeGetTypes [#1]}
  n3 = TextureSampleParameter2D() {
   group=RandomTileAnimatedPattern
   parameterName=Layers_MaskTexture
   samplerType=SAMPLERTYPE_Masks
   texture=&T_WhiteMask
  }
  n4 = StaticComponentMaskParameter(None=@n3.RGBA) {defaultR=true, group=RandomTileAnimatedPattern, parameterName=Layer1_MaskChannel}
  n5 = ScalarParameter() {defaultValue=0.5, group=RandomTileAnimatedPattern, parameterName=Layer1_RandomMinOpacity}
  n6 = Time()
  n7 = Frac(None=Multiply(A=@n6, B=ScalarParameter() {defaultValue=0.2, group=RandomTileAnimatedPattern, parameterName=Layer1_U_SineSpeed}) {constB=0.2})
  n8 = CameraPositionWS()
  n9 = Normalize(VectorInput=Subtract(A=PreSkinnedPosition(), B=TransformPosition(None=@n8) {transformSourceType=TRANSFORMPOSSOURCE_World}))
  n10 = ComponentMask(None=Multiply(A=@n9, B=ScalarParameter() {defaultValue=-0.125, group=RandomTileAnimatedPattern, parameterName=Layer1_ParallaxScale}) {constB=-0.125}) {g=true, r=true}
  n11 = TextureCoordinate()
  n12 = StaticBoolParameter() {group=RandomTileAnimatedPattern, parameterName=UseParallax}
  n13 = Multiply(
   A=StaticSwitch(True=Add(A=@n10, B=@n11), False=@n11, Value=@n12)
   B=ScalarParameter() {defaultValue=12, group=RandomTileAnimatedPattern, parameterName=Layer1_Tiling}
  ) {constB=10}
  n14 = Add(
   A=Multiply(A=Sine(None=@n7), B=ScalarParameter() {defaultValue=0.05, group=RandomTileAnimatedPattern, parameterName=Layer1_U_SineScale}) {constB=0.05}
   B=ComponentMask(None=@n13) {r=true}
  )
  n15 = ComponentMask(None=@n13) {g=true}
  n16 = AppendVector(
   A=ScalarParameter() {group=RandomTileAnimatedPattern, parameterName=Layer1_U_Speed}
   B=ScalarParameter() {defaultValue=-0.1, group=RandomTileAnimatedPattern, parameterName=Layer1_V_Speed}
  )
  n17 = ScalarParameter() {defaultValue=1, group=RandomTileAnimatedPattern, parameterName=Layers_GlobalPanningSpeedScale}
  n18 = VertexInterpolator(VS=Panner(Coordinate=AppendVector(A=@n14, B=@n15), Time=@n6, Speed=Multiply(A=@n16, B=@n17)))
  n19 = Frac(None=Multiply(A=@n6, B=ScalarParameter() {defaultValue=0.15, group=RandomTileAnimatedPattern, parameterName=Layer2_U_SineSpeed}) {constB=0.15})
  n20 = ComponentMask(None=Multiply(A=@n9, B=ScalarParameter() {defaultValue=-0.1, group=RandomTileAnimatedPattern, parameterName=Layer2_ParallaxScale}) {constB=-0.1}) {g=true, r=true}
  n21 = Multiply(
   A=StaticSwitch(True=Add(A=@n20, B=@n11), False=@n11, Value=@n12)
   B=ScalarParameter() {defaultValue=12, group=RandomTileAnimatedPattern, parameterName=Layer2_Tiling}
  ) {constB=12}
  n22 = Add(
   A=Multiply(A=Sine(None=@n19), B=ScalarParameter() {defaultValue=-0.05, group=RandomTileAnimatedPattern, parameterName=Layer2_U_SineScale}) {constB=-0.05}
   B=ComponentMask(None=@n21) {r=true}
  )
  n23 = ComponentMask(None=@n21) {g=true}
  n24 = AppendVector(
   A=ScalarParameter() {group=RandomTileAnimatedPattern, parameterName=Layer2_U_Speed}
   B=ScalarParameter() {defaultValue=-0.25, group=RandomTileAnimatedPattern, parameterName=Layer2_V_Speed}
  )
  n25 = VertexInterpolator(VS=Panner(Coordinate=AppendVector(A=@n22, B=@n23), Time=@n6, Speed=Multiply(A=@n24, B=@n17)))
  n26 = ScalarParameter() {defaultValue=0.25, group=RandomTileAnimatedPattern, parameterName=Layer1_Threshold}
  n27 = ScalarParameter() {defaultValue=0.2867, group=RandomTileAnimatedPattern, parameterName=Layer2_Threshold}
  n28 = MaterialFunctionCall!(
   UVs_Layer1=@n18.PS
   UVs_Layer2=@n25.PS
   Seed_Layer1=ScalarParameter() {defaultValue=0.514, group=RandomTileAnimatedPattern, parameterName=Layer1_Seed}
   Seed_Layer2=ScalarParameter() {defaultValue=8.4151, group=RandomTileAnimatedPattern, parameterName=Layer2_Seed}
   Threshold_Layer1=@n26
   Threshold_Layer2=@n27
   EdgeMaskWidth_Layer1=ScalarParameter() {defaultValue=0.95, group=RandomTileAnimatedPattern, parameterName=Layer1_EdgeMaskWidth}
   EdgeMaskWidth_Layer2=ScalarParameter() {defaultValue=0.95, group=RandomTileAnimatedPattern, parameterName=Layer2_EdgeMaskWidth}
  ) {materialFunction=&/MF_RandomTilePatternFromUVs}
  n29 = Add(A=@n28.Layer1_RandomValue, B=@n26)
  n30 = Divide(A=Saturate(None=Subtract(A=@n29)), B=@n26)
  n31 = Multiply(A=Subtract(A=@n30) {constB=0.5}) {constB=2}
  n32 = Add(A=Multiply(A=@n31, B=ScalarParameter() {defaultValue=0.25, group=RandomTileAnimatedPattern, parameterName=Layer1_RandomScaleFactor}) {constB=0.25})
  n33 = Multiply(A=Frac(None=@n18.PS), B=@n32)
  n34 = Subtract(A=Multiply(A=@n32) {constB=0.5}) {constB=0.5}
  n35 = ScalarParameter() {defaultValue=1, group=RandomTileAnimatedPattern, parameterName=Layer1_SelfRotationSpeed}
  n36 = ScalarParameter() {defaultValue=0.5, group=RandomTileAnimatedPattern, parameterName=Layer1_RandomMinRotationSpeed, sliderMax=1}
  n37 = Add(B=OneMinus(None=@n36)) {constA=-1}
  n38 = Add(A=Sign(None=@n31))
  n39 = Time()
  n40 = Multiply(A=LinearInterpolate(A=Min(A=@n31, B=@n37), B=Max(A=@n31, B=@n36), Alpha=Multiply(A=@n38) {constB=0.5}), B=@n39)
  n41 = Constant() {r=-1}
  n42 = TextureSampleParameter2D!(UVs=Rotator(Coordinate=Subtract(A=@n33, B=@n34), Time=Multiply(A=@n35, B=@n40)), Bias=@n41, Apply View MipBias) {
   group=RandomTileAnimatedPattern
   mipValueMode=TMVM_MipBias
   parameterName=Layer1
   samplerType=SAMPLERTYPE_Masks
   texture=&T_WhiteMask
  }
  n43 = Multiply(A=Multiply(A=Max(A=@n30, B=@n5) {constB=0.5}, B=@n42.R), B=@n28.Layer1_Mask)
  n44 = StaticComponentMaskParameter(None=@n3.RGBA) {defaultG=true, group=RandomTileAnimatedPattern, parameterName=Layer2_MaskChannel}
  n45 = Add(A=@n28.Layer2_RandomValue, B=@n27)
  n46 = Divide(A=Saturate(None=Subtract(A=@n45)), B=@n27)
  n47 = Multiply(A=Subtract(A=@n46) {constB=0.5}) {constB=2}
  n48 = Add(A=Multiply(A=@n47, B=ScalarParameter() {defaultValue=0.33, group=RandomTileAnimatedPattern, parameterName=Layer2_RandomScaleFactor}) {constB=0.33})
  n49 = Multiply(A=Frac(None=@n25.PS), B=@n48)
  n50 = Subtract(A=Multiply(A=@n48) {constB=0.5}) {constB=0.5}
  n51 = ScalarParameter() {defaultValue=0.5, group=RandomTileAnimatedPattern, parameterName=Layer2_RandomMinRotationSpeed, sliderMax=1}
  n52 = Add(B=OneMinus(None=@n51)) {constA=-1}
  n53 = Add(A=Sign(None=@n47))
  n54 = Multiply(A=LinearInterpolate(A=Min(A=@n47, B=@n52), B=Max(A=@n47, B=@n51), Alpha=Multiply(A=@n53) {constB=0.5}), B=@n39)
  n55 = ScalarParameter() {defaultValue=1, group=RandomTileAnimatedPattern, parameterName=Layer2_SelfRotationSpeed}
  n56 = TextureSampleParameter2D(UVs=Rotator(Coordinate=Subtract(A=@n49, B=@n50), Time=Multiply(A=@n54, B=@n55)) {speed=-0.15}, Bias=@n41) {
   group=RandomTileAnimatedPattern
   mipValueMode=TMVM_MipBias
   parameterName=Layer2
   samplerType=SAMPLERTYPE_Masks
   texture=&T_WhiteMask
  }
  n57 = ScalarParameter() {defaultValue=0.5, group=RandomTileAnimatedPattern, parameterName=Layer2_RandomMinOpacity}
  n58 = Multiply(A=Multiply(A=@n56.G, B=Max(A=@n46, B=@n57) {constB=0.5}), B=@n28.Layer2_Mask)
  n59 = StaticComponentMaskParameter(None=@n3.RGBA) {defaultB=true, group=RandomTileAnimatedPattern, parameterName=Layer3_MaskChannel}
  n60 = ScalarParameter() {defaultValue=0.5, group=RandomTileAnimatedPattern, parameterName=Layer3_RandomMinOpacity}
  n61 = Frac(None=Multiply(A=@n6, B=ScalarParameter() {defaultValue=0.2, group=RandomTileAnimatedPattern, parameterName=Layer3_U_SineSpeed}) {constB=0.2})
  n62 = ComponentMask(None=Multiply(A=@n9, B=ScalarParameter() {defaultValue=-0.1, group=RandomTileAnimatedPattern, parameterName=Layer3_ParallaxScale}) {constB=-0.1}) {g=true, r=true}
  n63 = Multiply(
   A=StaticSwitch(True=Add(A=@n62, B=@n11), False=@n11, Value=@n12)
   B=ScalarParameter() {defaultValue=16, group=RandomTileAnimatedPattern, parameterName=Layer3_Tiling}
  ) {constB=16}
  n64 = Add(
   A=Multiply(A=Sine(None=@n61), B=ScalarParameter() {defaultValue=0.05, group=RandomTileAnimatedPattern, parameterName=Layer3_U_SineScale}) {constB=0.05}
   B=ComponentMask(None=@n63) {r=true}
  )
  n65 = ComponentMask(None=@n63) {g=true}
  n66 = AppendVector(
   A=ScalarParameter() {group=RandomTileAnimatedPattern, parameterName=Layer3_U_Speed}
   B=ScalarParameter() {defaultValue=-0.25, group=RandomTileAnimatedPattern, parameterName=Layer3_V_Speed}
  )
  n67 = VertexInterpolator(VS=Panner(Coordinate=AppendVector(A=@n64, B=@n65), Time=@n6, Speed=Multiply(A=@n66, B=@n17)))
  n68 = Frac(None=Multiply(A=@n6, B=ScalarParameter() {defaultValue=0.2, group=RandomTileAnimatedPattern, parameterName=Layer4_U_SineSpeed}) {constB=0.2})
  n69 = ComponentMask(None=Multiply(A=@n9, B=ScalarParameter() {defaultValue=-0.05, group=RandomTileAnimatedPattern, parameterName=Layer4_ParallaxScale}) {constB=-0.05}) {g=true, r=true}
  n70 = Multiply(
   A=StaticSwitch(True=Add(A=@n69, B=@n11), False=@n11, Value=@n12)
   B=ScalarParameter() {defaultValue=20, group=RandomTileAnimatedPattern, parameterName=Layer4_Tiling}
  ) {constB=21}
  n71 = Add(
   A=Multiply(A=Sine(None=@n68), B=ScalarParameter() {defaultValue=-0.05, group=RandomTileAnimatedPattern, parameterName=Layer4_U_SineScale}) {constB=0.05}
   B=ComponentMask(None=@n70) {r=true}
  )
  n72 = ComponentMask(None=@n70) {g=true}
  n73 = AppendVector(
   A=ScalarParameter() {group=RandomTileAnimatedPattern, parameterName=Layer4_U_Speed}
   B=ScalarParameter() {defaultValue=-0.5, group=RandomTileAnimatedPattern, parameterName=Layer4_V_Speed}
  )
  n74 = VertexInterpolator(VS=Panner(Coordinate=AppendVector(A=@n71, B=@n72), Speed=Multiply(A=@n73, B=@n17)))
  n75 = ScalarParameter() {defaultValue=0.5, group=RandomTileAnimatedPattern, parameterName=Layer3_Threshold}
  n76 = ScalarParameter() {defaultValue=0.1, group=RandomTileAnimatedPattern, parameterName=Layer4_Threshold}
  n77 = MaterialFunctionCall(
   UVs_Layer1=@n67.PS
   UVs_Layer2=@n74.PS
   Seed_Layer1=ScalarParameter() {defaultValue=1.351, group=RandomTileAnimatedPattern, parameterName=Layer3_Seed}
   Seed_Layer2=ScalarParameter() {defaultValue=3.144, group=RandomTileAnimatedPattern, parameterName=Layer4_Seed}
   Threshold_Layer1=@n75
   Threshold_Layer2=@n76
   EdgeMaskWidth_Layer1=ScalarParameter() {defaultValue=0.95, group=RandomTileAnimatedPattern, parameterName=Layer3_EdgeMaskWidth}
   EdgeMaskWidth_Layer2=ScalarParameter() {defaultValue=0.95, group=RandomTileAnimatedPattern, parameterName=Layer4_EdgeMaskWidth}
  ) {materialFunction=&MF_RandomTilePatternFromUVs}
  n78 = Add(A=@n77.Layer1_RandomValue, B=@n75)
  n79 = Divide(A=Saturate(None=Subtract(A=@n78)), B=@n75)
  n80 = Multiply(A=Subtract(A=@n79) {constB=0.5}) {constB=2}
  n81 = Add(A=Multiply(A=@n80, B=ScalarParameter() {defaultValue=0.4, group=RandomTileAnimatedPattern, parameterName=Layer3_RandomScaleFactor}) {constB=0.4})
  n82 = Multiply(A=Frac(None=@n67.PS), B=@n81)
  n83 = Subtract(A=Multiply(A=@n81) {constB=0.5}) {constB=0.5}
  n84 = ScalarParameter() {defaultValue=1, group=RandomTileAnimatedPattern, parameterName=Layer3_SelfRotationSpeed}
  n85 = ScalarParameter() {defaultValue=0.5, group=RandomTileAnimatedPattern, parameterName=Layer3_RandomMinRotationSpeed}
  n86 = Add(B=OneMinus(None=@n85)) {constA=-1}
  n87 = Add(A=Sign(None=@n80))
  n88 = Time()
  n89 = Multiply(A=LinearInterpolate(A=Min(A=@n80, B=@n86), B=Max(A=@n80, B=@n85), Alpha=Multiply(A=@n87) {constB=0.5}), B=@n88)
  n90 = TextureSampleParameter2D(UVs=Rotator(Coordinate=Subtract(A=@n82, B=@n83), Time=Multiply(A=@n84, B=@n89)) {speed=0.33}, Bias=@n41) {
   group=RandomTileAnimatedPattern
   mipValueMode=TMVM_MipBias
   parameterName=Layer3
   samplerType=SAMPLERTYPE_Masks
   texture=&T_WhiteMask
  }
  n91 = Multiply(A=Multiply(A=Max(A=@n79, B=@n60) {constB=0.5}, B=@n90.B), B=@n77.Layer1_Mask)
  n92 = StaticComponentMaskParameter(None=@n3.RGBA) {defaultA=true, group=RandomTileAnimatedPattern, parameterName=Layer4_MaskChannel}
  n93 = Add(A=@n77.Layer2_RandomValue, B=@n76)
  n94 = Divide(A=Saturate(None=Subtract(A=@n93)), B=@n76)
  n95 = Multiply(A=Subtract(A=@n94) {constB=0.5}) {constB=2}
  n96 = Add(A=Multiply(A=@n95, B=ScalarParameter() {defaultValue=0.5, group=RandomTileAnimatedPattern, parameterName=Layer4_RandomScaleFactor}) {constB=0.5})
  n97 = Multiply(A=Frac(None=@n74.PS), B=@n96)
  n98 = Subtract(A=Multiply(A=@n96) {constB=0.5}) {constB=0.5}
  n99 = ScalarParameter() {defaultValue=0.5, group=RandomTileAnimatedPattern, parameterName=Layer4_RandomMinRotationSpeed}
  n100 = Add(B=OneMinus(None=@n99)) {constA=-1}
  n101 = Add(A=Sign(None=@n95))
  n102 = Multiply(A=LinearInterpolate(A=Min(A=@n95, B=@n100), B=Max(A=@n95, B=@n99), Alpha=Multiply(A=@n101) {constB=0.5}), B=@n88)
  n103 = ScalarParameter() {defaultValue=1, group=RandomTileAnimatedPattern, parameterName=Layer4_SelfRotationSpeed}
  n104 = TextureSampleParameter2D(UVs=Rotator(Coordinate=Subtract(A=@n97, B=@n98), Time=Multiply(A=@n102, B=@n103)) {speed=0.33}, Bias=@n41) {
   group=RandomTileAnimatedPattern
   mipValueMode=TMVM_MipBias
   parameterName=Layer4
   samplerType=SAMPLERTYPE_Masks
   texture=&T_WhiteMask
  }
  n105 = ScalarParameter() {defaultValue=0.5, group=RandomTileAnimatedPattern, parameterName=Layer4_RandomMinOpacity}
  n106 = Multiply(A=Multiply(A=@n104.A, B=Max(A=@n94, B=@n105) {constB=0.5}), B=@n77.Layer2_Mask)
  n107 = Add(A=Add(A=Multiply(A=@n4, B=@n43), B=Multiply(A=@n44, B=@n58)), B=Add(A=Multiply(A=@n59, B=@n91), B=Multiply(A=@n92, B=@n106)))
  n108 = SetMaterialAttributes!(MaterialAttributes=@n2.MaterialAttributes, BaseColor=Saturate(None=Add(A=@n2.BaseColor, B=@n107))) {attributeSetTypes [#1]}
  n109 = FunctionInput() {bUsePreviewValueAsDefault=true, description=FXMaskChannel, inputName=Alpha, inputType=FunctionInput_Scalar, previewValue=<1 1 1 1>}
  FunctionOutput(None=StaticSwitch(
    True=BlendMaterialAttributes(A=@n1, B=@n108, Alpha=@n109)
    False=@n108
    Value=StaticBoolParameter() {group=RandomTileAnimatedPattern, parameterName=Use FXmask for RandomTileAnimatedPattern}
   ))
 ]}
