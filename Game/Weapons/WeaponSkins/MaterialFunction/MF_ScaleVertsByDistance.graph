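# Material function for weapon skins that pushes vertices outward along their normals as the camera gets farther from the actor.
# Actor-to-camera distance is inverse-lerped between start and end distance parameters, with a static switch to saturate the result.
# That factor multiplies the world vertex normal directly, with no separate strength parameter, so a saturated offset peaks at one unit.
# The offset adds to the incoming material attributes' World Position Offset.
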MaterialFunction /Game/Weapons/WeaponSkins/MaterialFunction/MF_ScaleVertsByDistance
graph {nodes [
  n1 = GetMaterialAttributes(None=FunctionInput() {inputName=MA_In, inputType=FunctionInput_MaterialAttributes}) {attributeGetTypes [#1]}
  n2 = InverseLinearInterpolate(
   A=ScalarParameter() {group=ScaleVertsByDistance, parameterName=StartScaleDistance}
   B=ScalarParameter() {defaultValue=800, group=ScaleVertsByDistance, parameterName=EndScaleDistance}
   Value=Distance(A=ActorPositionWS(), B=CameraPositionWS())
  )
  n3 = Multiply(
   A=VertexNormalWS()
   B=StaticSwitchParameter(True=Saturate(None=@n2), False=@n2) {group=ScaleVertsByDistance, parameterName=SaturateDistanceValues}
  )
  FunctionOutput(None=SetMaterialAttributes!(MaterialAttributes=@n1.MaterialAttributes, WorldPositionOffset=Add(A=@n1.WorldPositionOffset, B=@n3)) {attributeSetTypes [#1]})
  ObjectPositionWS()
 ]}
