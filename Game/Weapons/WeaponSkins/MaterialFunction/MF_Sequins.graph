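# Material function for weapon skins that layers a tiled sequin look over input material attributes within a channel-selected mask.
# A per-sequin offset pattern and Fresnel falloff drive color from two tints or a curve-atlas LUT.
# Tiling textures supply roughness and normal detail, with options to reuse base values.
# Optional emissive sparkle multiplies scrolling noise and caustics, and a static switch can reuse the TechArt FX mask to save a texture sample.
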MaterialFunction /Game/Weapons/WeaponSkins/MaterialFunction/MF_Sequins
graph {
 properties {bExposeToLibrary=true}
 nodes [
  n1 = FunctionInput(Preview=SetMaterialAttributes!(MaterialAttributes)) {bUsePreviewValueAsDefault=true, inputName=inMat, inputType=FunctionInput_MaterialAttributes}
  n2 = GetMaterialAttributes(None=@n1) {attributeGetTypes [
    #1
    #2
    #3
    #4
    #5
   ]}
  n3 = GetMaterialAttributes(None=@n1) {attributeGetTypes [
    #3
    #4
    #1
    #2
    #5
   ]}
  n4 = TextureObject()
  n5 = StaticSwitchParameter(
   True=FunctionInput(Preview=Constant3Vector()) {bUsePreviewValueAsDefault=true, inputName=TechArtMask}
   False=TextureSample(Tex=FunctionInput(Preview=@n4) {bUsePreviewValueAsDefault=true, inputName=Mask, inputType=FunctionInput_Texture2D}).RGBA
  ) {
   desc=Enable this to use the default TechArt FX Mask instead of paying another Texture Sample cost for this effect.
   group=MF_Sequin
   parameterName=UseInputMaskSampleV3
  }
  n6 = ChannelMaskParameter(None=@n5) {group=MF_Sequin, parameterName=SequinMaskChannel}
  n7 = ScalarParameter() {defaultValue=1.5, group=MF_Sequin, parameterName=SequinBrightness}
  n8 = Fresnel(ExponentIn=ScalarParameter() {defaultValue=1, group=MF_Sequin, parameterName=SequinFresnel}) {exponent=1}
  n9 = Multiply(A=@n8, B=@n8)
  n10 = OneMinus(None=@n9)
  n11 = Multiply(A=NamedRerouteUsage(), B=LinearInterpolate(Alpha=Multiply(A=@n10, B=@n9)) {constA=0.7})
  n12 = ScalarParameter() {defaultValue=0.9, group=MF_Sequin, parameterName=SequinMetalness}
  n13 = NamedRerouteUsage()
  n14 = ScalarParameter() {group=MF_Sequin, parameterName=UseBaseMetalness}
  n15 = NamedRerouteUsage()
  n16 = TextureSampleParameter2D(UVs=NamedRerouteUsage()) {
   group=MF_Sequin
   parameterName=Sequin_TilingRoughness
   samplerType=SAMPLERTYPE_LinearGrayscale
   texture=/Game/Global/Textures/Default/Tileable_Patterns/T_SequinTile_roughness
  }
  n17 = LinearInterpolate(
   A=ScalarParameter() {group=MF_Sequin, parameterName=Sequin_MinRoughness}
   B=ScalarParameter() {defaultValue=1, group=MF_Sequin, parameterName=Sequin_MaxRoughness}
   Alpha=@n16.G
  )
  n18 = NamedRerouteUsage()
  n19 = TextureSampleParameter2D(UVs=NamedRerouteUsage()) {
   group=MF_Sequin
   parameterName=Sequin_TilingNormal
   samplerType=SAMPLERTYPE_Normal
   texture=/Game/Global/Textures/Default/Tileable_Patterns/T_SequinTile_N
  }
  n20 = Multiply(A=ScalarParameter() {defaultValue=0.8, group=MF_Sequin, parameterName=SequinNormalIntensity}, B=Constant3Vector() {constant=<rgba 1 1 0 0>})
  n21 = Add(A=@n18, B=Multiply(A=@n19.RGB, B=@n20)) {constB=0}
  n22 = NamedRerouteUsage()
  n23 = ScalarParameter() {defaultValue=1, group=MF_Sequin, parameterName=SequinEmissiveIntensity}
  n24 = Time()
  n25 = ScalarParameter() {defaultValue=0.01, group=MF_Sequin, parameterName=SequinSparkleSpeed}
  n26 = TextureSample(UVs=Add(A=NamedRerouteUsage(), B=Multiply(A=@n24, B=@n25) {constB=0.2})) {samplerSource=SSM_Wrap_WorldGroupSettings, samplerType=SAMPLERTYPE_Normal, texture=/Game/Global/Textures/Default/Noise/64x64Noise}
  n27 = TextureCoordinate()
  n28 = ScalarParameter() {defaultValue=2, group=MF_Sequin, parameterName=SequinDiamondTile}
  n29 = TextureSample(UVs=Panner(Coordinate=Multiply(A=@n27, B=@n28)) {speedX=0.05, speedY=0.05}) {texture=/Game/Effects/Fort_Effects/Textures/Tile/T_ShieldCaustics_Tile_Chromatic}
  n30 = Multiply(A=Power(Base=@n26.R) {constExponent=4}, B=Multiply(A=@n29.R, B=@n29.R))
  n31 = VectorParameter() {parameterName=SparkleColor, defaultValue=<rgba 0.5742 0.228 1 1>}
  n32 = Multiply(
   A=NamedRerouteUsage()
   B=Multiply(A=Multiply(A=@n30, B=@n31.RGB), B=ScalarParameter() {defaultValue=10, group=MF_Sequin, parameterName=SparkleBrightness})
  )
  n33 = StaticSwitchParameter(True=Add(A=Multiply(A=@n22, B=@n23), B=@n32), False=Add(A=NamedRerouteUsage(), B=@n32)) {desc=Replace emissive channel for sequin color, group=MF_Sequin, parameterName=EnableEmissiveOverride}
  n34 = SetMaterialAttributes!(
   MaterialAttributes=@n2.MaterialAttributes
   BaseColor=LinearInterpolate(A=@n3.BaseColor, B=Multiply(A=@n7, B=@n11), Alpha=@n6.R)
   Metallic=LinearInterpolate(A=@n3.Metallic, B=LinearInterpolate(A=@n12, B=@n13, Alpha=@n14), Alpha=@n6.R)
   Roughness=LinearInterpolate(A=@n3.Roughness, B=StaticSwitchParameter(True=@n15, False=@n17) {group=MF_Sequin, parameterName=UseBaseRoughness}, Alpha=@n6.R)
   Normal=LinearInterpolate(A=@n3.Normal, B=StaticSwitchParameter(True=@n18, False=@n21) {group=MF_Sequin, parameterName=UseBaseNormal}, Alpha=@n6.R)
   EmissiveColor=StaticSwitchParameter(True=LinearInterpolate(A=@n3.EmissiveColor, B=@n33, Alpha=@n6.R), False=@n3.EmissiveColor) {group=MF_Sequin, parameterName=EnableSequinSparkle}
  ) {attributeSetTypes [
    #3
    #4
    #1
    #2
    #5
   ]}
  FunctionOutput(None=@n34)
  Multiply(A=@n18, B=@n20)
  NamedRerouteDeclaration(None=Multiply(A=TextureCoordinate(), B=ScalarParameter() {defaultValue=32, group=MF_Sequin, parameterName=SequinTile})) {name=SequinUVs, nodeColor=<rgba 0 1 0.0859 1>}
  n35 = TextureSampleParameter2D(UVs=NamedRerouteUsage()) {
   group=MF_Sequin
   parameterName=Sequin_TilingOffsetPattern
   samplerType=SAMPLERTYPE_LinearGrayscale
   texture=/Game/Global/Textures/Default/Tileable_Patterns/T_SequinTile
  }
  NamedRerouteDeclaration(None=@n35.R) {name=SequinOffset, nodeColor=<rgba 1 0 0.6562 1>}
  n36 = Multiply(
   A=LinearInterpolate(
    A=ScalarParameter() {defaultValue=0.5, group=MF_Sequin, parameterName=SequinColorOffsetMin}
    B=ScalarParameter() {defaultValue=1.5, group=MF_Sequin, parameterName=SequinColorOffsetMax}
    Alpha=NamedRerouteUsage()
   ) {constA=0.5, constB=1.5}
   B=@n10
  )
  n37 = CurveAtlasRowParameter(CurveTime=@n36) {
   atlas=/Game/Characters/Player/Female/Medium/Bodies/F_MED_SkeleProbe/Skins/Candy/Textures/CA_SkeleProbe_Candy
   curve=/Game/Characters/Player/Female/Medium/Bodies/F_MED_SkeleProbe/Skins/Candy/Textures/C_SkeleProbe_Candy
   group=MF_Sequin
   parameterName=CurveLUTColor
  }
  n38 = VectorParameter() {group=MF_Sequin, parameterName=SequinFalloffColor01, defaultValue=<rgba 0.8842 0.6282 1 1>}
  n39 = VectorParameter() {group=MF_Sequin, parameterName=SequinFalloffColor02, defaultValue=<rgba 0.6663 0.575 1 1>}
  n40 = LinearInterpolate(A=@n38.RGB, B=@n39.RGB, Alpha=@n36)
  n41 = NamedRerouteUsage()
  n42 = LinearInterpolate(A=Multiply(A=@n41, B=@n38.RGB), B=Multiply(A=@n41, B=@n39.RGB), Alpha=@n36)
  n43 = ScalarParameter() {defaultValue=1, group=MF_Sequin, parameterName=UseBaseColor}
  NamedRerouteDeclaration(None=StaticSwitchParameter(True=@n37, False=LinearInterpolate(A=@n40, B=@n42, Alpha=@n43)) {group=MF_Sequin, parameterName=UseLUTTextureAsColor, sortPriority=16}) {name=SequinColor, nodeColor=<rgba 0.8984 0 1 1>}
  NamedRerouteDeclaration(None=@n2.Roughness) {name=BaseRoughness, nodeColor=<rgba 0.6172 1 0 1>}
  NamedRerouteDeclaration(None=@n2.Normal) {name=BaseNormals, nodeColor=<rgba 0 1 0.1562 1>}
  NamedRerouteDeclaration(None=@n2.BaseColor) {name=BaseColor, nodeColor=<rgba 0 1 0.7188 1>}
  NamedRerouteDeclaration(None=@n2.Metallic) {name=BaseMetalness, nodeColor=<rgba 0.6406 0 1 1>}
  NamedRerouteDeclaration(None=@n2.EmissiveColor) {name=BaseEmissiveness, nodeColor=<rgba 0 0.7422 1 1>}
 ]
}
