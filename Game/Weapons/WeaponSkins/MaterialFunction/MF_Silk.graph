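# Material function for weapon skins that gives a silk look to incoming material attributes, masked by an Alpha input.
# A Fresnel term blends brightness-scaled base color toward an edge tint and interpolates metallic between two parameters.
# The same term scales roughness at edges by SilkEdgeAniso, which is only a multiplier despite its name.
# The result also sets shading model and CustomData0, then blends over the original, keeping its vertex attributes.
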MaterialFunction /Game/Weapons/WeaponSkins/MaterialFunction/MF_Silk
graph {nodes [
  n1 = GetMaterialAttributes(None=FunctionInput() {inputName=Material, inputType=FunctionInput_MaterialAttributes}) {attributeGetTypes [
    #1
    #2
    #3
    #4
   ]}
  n2 = ScalarParameter() {defaultValue=1, group=Silk, parameterName=SilkBaseColorBrightness}
  n3 = Fresnel(ExponentIn=ScalarParameter() {defaultValue=2, group=Silk, parameterName=Silk Fresnel}) {baseReflectFraction=0}
  n4 = ScalarParameter() {defaultValue=0.5, group=Silk, parameterName=SilkEdgeAniso}
  n5 = SetMaterialAttributes!(
   MaterialAttributes=@n1.MaterialAttributes
   BaseColor=LinearInterpolate(
    A=Multiply(A=@n1.BaseColor, B=@n2)
    B=VectorParameter() {group=Silk, parameterName=SilkEdgeTint, defaultValue=<rgba 0.1094 0.1094 0.1094 1>}.RGB
    Alpha=@n3
   )
   Metallic=LinearInterpolate(
    A=ScalarParameter() {defaultValue=1, group=Silk, parameterName=SIlkFresnelMax}
    B=ScalarParameter() {group=Silk, parameterName=SIlkFresnelMin}
    Alpha=@n3
   )
   Specular=@n1.MaterialAttributes
   Roughness=LinearInterpolate(A=@n1.Roughness, B=Multiply(A=@n1.Roughness, B=@n4), Alpha=@n3)
   ShadingModel=ShadingModel()
   SubsurfaceColor=@n1.MaterialAttributes
   CustomData0=Constant() {r=1}
  ) {attributeSetTypes [
    #1
    #2
    #3
    #4
    #5
    #6
    #7
   ]}
  FunctionOutput(None=BlendMaterialAttributes(A=@n1.MaterialAttributes, B=@n5, Alpha=FunctionInput() {inputName=Alpha, inputType=FunctionInput_Scalar}) {vertexAttributeBlendType=UseA})
 ]}
