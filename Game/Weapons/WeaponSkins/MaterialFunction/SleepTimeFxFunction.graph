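# SleepTimeFxFunction is a material function for weapon skins that builds a stylized Fresnel-based "SleepyTime" color effect intended for emissive.
# An inner contrast-adjusted Fresnel blends a dark and a main inside color, and a second, contrast-shaped Fresnel overlays an outline color.
# The blended result passes through HueShift to the output.
# Colors, outer Fresnel settings and hue are function inputs with SleepyTime parameters as preview defaults.
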MaterialFunction /Game/Weapons/WeaponSkins/MaterialFunction/SleepTimeFxFunction
graph {nodes [
  n1 = ScalarParameter() {group=SleepyTime, parameterName=Hue}
  n2 = ScalarParameter() {defaultValue=0.9091, group=SleepyTime, parameterName=ExponenInner}
  n3 = ScalarParameter() {defaultValue=0.0491, group=SleepyTime, parameterName=BaseRefInner}
  n4 = LinearInterpolate(
   A=FunctionInput(Preview=VectorParameter() {group=SleepyTime, parameterName=InsideDark, defaultValue=<rgba 0.4097 0.1339 0.0433 1>}.RGB) {bUsePreviewValueAsDefault=true, inputName=InsideDarkColor}
   B=FunctionInput(Preview=VectorParameter() {group=SleepyTime, parameterName=insideMain, defaultValue=<rgba 1 0.7732 0.0645 1>}.RGB) {bUsePreviewValueAsDefault=true, inputName=InsideMaincolor}
   Alpha=MaterialFunctionCall(
    In=Fresnel(ExponentIn=@n2, BaseReflectFractionIn=@n3) {exponent=2}
    Contrast=ScalarParameter() {defaultValue=-0.019, group=SleepyTime, parameterName=inerlianeshape}
   ) {materialFunction=&CheapContrast}.Result
  )
  n5 = FunctionInput(Preview=VectorParameter() {group=SleepyTime, parameterName=outline color, defaultValue=<rgba 0.7656 0.6165 0.1979 1>}.RGB) {bUsePreviewValueAsDefault=true, inputName=outline color}
  n6 = ScalarParameter() {defaultValue=0.7286, group=SleepyTime, parameterName=Exponen}
  n7 = ScalarParameter() {defaultValue=0.1733, group=SleepyTime, parameterName=BaseRef}
  n8 = MaterialFunctionCall(
   In=Fresnel(
    ExponentIn=FunctionInput(Preview=@n6) {bUsePreviewValueAsDefault=true, inputName=Exponen, inputType=FunctionInput_Scalar}
    BaseReflectFractionIn=FunctionInput(Preview=@n7) {bUsePreviewValueAsDefault=true, inputName=BaseRef, inputType=FunctionInput_Scalar}
   ) {exponent=2}
   Contrast=ScalarParameter() {defaultValue=11.3985, group=SleepyTime, parameterName=outlineShape}
  ) {materialFunction=&CheapContrast}
  FunctionOutput(None=MaterialFunctionCall(
    Hue Shift Percentage=FunctionInput(Preview=@n1) {bUsePreviewValueAsDefault=true, inputName=Hue, inputType=FunctionInput_Scalar}
    Texture=LinearInterpolate(A=@n4, B=@n5, Alpha=@n8.Result)
   ) {materialFunction=&HueShift}.Result) {description=ToEmissive}
 ]}
