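# Material function that layers the animated Sleepy Time skin effect onto weapon material attributes.
# A ribbon mask selects hues and colors for SleepTimeFxFunction and panning speeds for the streak texture driving opacity, with an optional UV1 gradient cutoff.
# A switchable TechArt overlay derives four U-axis color bands procedurally, each panning separately.
# The emissive, black-base result blends over the input using Mask channels, or fully when UseHueshift is one.
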MaterialFunction /Game/Weapons/WeaponSkins/MaterialFunction/MF_SleepyTime
graph {nodes [
  Constant() {r=1}
  n1 = ScalarParameter() {group=SleepyTime, parameterName=UseEdgeMasking}
  n2 = Multiply(
   A=LinearInterpolate(A=TextureCoordinate(), B=TextureCoordinate() {coordinateIndex=1}, Alpha=@n1)
   B=ScalarParameter() {defaultValue=1, group=SleepyTime, parameterName=RibbonTiling}
  )
  n3 = TextureSample(UVs=@n2) {samplerType=SAMPLERTYPE_Masks, texture=/Game/Characters/Player/Male/Medium/Bodies/M_MED_SleepyTime/Textures/T_MED_SleepTime_Body_Mask01}
  n4 = ScalarParameter() {defaultValue=1, group=SleepyTime, parameterName=UseHueshift}
  n5 = LinearInterpolate(A=Constant() {r=1}, B=ScalarParameter() {defaultValue=0.4889, group=SleepyTime, parameterName=Blue}, Alpha=Multiply(A=@n3.G, B=@n4))
  n6 = ScalarParameter() {defaultValue=0.7175, group=SleepyTime, parameterName=purple}
  n7 = Multiply(A=@n3.B, B=@n4)
  n8 = VectorParameter() {group=SleepyTime, parameterName=insideMain, defaultValue=<rgba 1 0.7732 0.0645 1>}
  n9 = VectorParameter() {group=SleepyTime, parameterName=magenta ColorA, defaultValue=<rgba 1 0 0.1111 1>}
  n10 = LinearInterpolate(A=@n9.RGB, B=@n8.RGB, Alpha=@n3.R)
  n11 = VectorParameter() {group=SleepyTime, parameterName=InsideDark, defaultValue=<rgba 0.4097 0.1339 0.0433 1>}
  n12 = LinearInterpolate(A=@n9.RGB, B=@n11.RGB, Alpha=@n3.R)
  n13 = VectorParameter() {group=SleepyTime, parameterName=outline color, defaultValue=<rgba 0.7656 0.6165 0.1979 1>}
  n14 = LinearInterpolate(A=@n9.RGB, B=@n13.RGB, Alpha=@n3.R)
  n15 = Constant() {r=1}
  n16 = Multiply(
   A=MaterialFunctionCall!(
    Exponen=ScalarParameter() {defaultValue=0.9324, group=SleepyTime, parameterName=Exponen}
    Hue=LinearInterpolate(A=@n5, B=@n6, Alpha=@n7)
    InsideMaincolor=LinearInterpolate(A=@n8.RGB, B=@n10, Alpha=@n4)
    InsideDarkColor=LinearInterpolate(A=@n11.RGB, B=@n12, Alpha=@n4)
    outline color=LinearInterpolate(A=@n13.RGB, B=@n14, Alpha=@n4)
    BaseRef=ScalarParameter() {defaultValue=0.0643, group=SleepyTime, parameterName=BaseRef}
   ) {materialFunction=&/SleepTimeFxFunction}.Result
   B=Clamp(None=@n15, Min=ScalarParameter() {group=SleepyTime, parameterName=FaceContrastValue})
  )
  n17 = ScalarParameter() {defaultValue=0.7286, group=SleepyTime, parameterName=Exponen}
  n18 = ScalarParameter() {defaultValue=1, group=SleepyTime, parameterName=TechArtUTiling}
  n19 = AppendVector(A=@n18, B=ScalarParameter() {defaultValue=1, group=SleepyTime, parameterName=TechArtVTiling})
  n20 = Multiply(A=TextureCoordinate(), B=@n19)
  n21 = Multiply(A=ComponentMask(None=Frac(None=@n20)) {r=true}) {constB=4}
  n22 = Floor(None=@n21)
  n23 = Multiply(A=OneMinus(None=Clamp(None=@n22)), B=Constant3Vector() {constant=<rgba 1 0 0 0>})
  n24 = Abs(None=Subtract(A=Floor(None=@n21)))
  n25 = Multiply(A=OneMinus(None=Clamp(None=@n24)), B=Constant3Vector() {constant=<rgba 0 0 1 0>})
  n26 = Subtract(A=Floor(None=@n21)) {constB=3}
  n27 = OneMinus(None=Clamp(None=Abs(None=@n26)))
  n28 = Constant3Vector() {constant=<rgba 0 1 0 0>}
  n29 = MaterialFunctionCall(Float3=Add(A=Add(A=@n23, B=@n25), B=Multiply(A=@n27, B=@n28))) {materialFunction=&BreakOutFloat3Components}
  n30 = LinearInterpolate(
   A=LinearInterpolate(A=Constant() {r=1}, B=ScalarParameter() {defaultValue=0.4889, group=SleepyTime, parameterName=Blue}, Alpha=@n29.G)
   B=ScalarParameter() {defaultValue=0.7175, group=SleepyTime, parameterName=purple}
   Alpha=@n29.B
  )
  n31 = VectorParameter() {group=SleepyTime, parameterName=magenta ColorA, defaultValue=<rgba 1 0 0.1111 1>}
  n32 = LinearInterpolate(A=@n31.RGB, B=VectorParameter() {group=SleepyTime, parameterName=insideMain, defaultValue=<rgba 1 0.7732 0.0645 1>}.RGB, Alpha=@n29.R)
  n33 = LinearInterpolate(A=@n31.RGB, B=VectorParameter() {group=SleepyTime, parameterName=InsideDark, defaultValue=<rgba 0.4097 0.1339 0.0433 1>}.RGB, Alpha=@n29.R)
  n34 = LinearInterpolate(
   A=@n31.RGB
   B=VectorParameter() {group=SleepyTime, parameterName=outline color, defaultValue=<rgba 0.7656 0.6165 0.1979 1>}.RGB
   Alpha=@n29.R
  )
  n35 = ScalarParameter() {defaultValue=0.1733, group=SleepyTime, parameterName=BaseRef}
  n36 = TextureCoordinate()
  n37 = Multiply(A=@n36, B=@n19)
  n38 = VectorParameter() {group=SleepyTime, parameterName=MagentaPanSpeed, defaultValue=<rgba 0 0.25 0 1>}
  n39 = VectorParameter() {group=SleepyTime, parameterName=YelloPanSpeed, defaultValue=<rgba 0 0.1 0 1>}
  n40 = LinearInterpolate(
   A=Panner(Coordinate=@n37, Speed=ComponentMask(None=@n38.RGB) {g=true, r=true}) {speedY=0.25}
   B=Panner(Coordinate=@n37, Speed=ComponentMask(None=@n39.RGB) {g=true, r=true}) {speedY=0.1}
   Alpha=@n29.R
  )
  n41 = Panner(
   Coordinate=@n37
   Speed=ComponentMask(None=VectorParameter() {group=SleepyTime, parameterName=BluePanSpeed, defaultValue=<rgba 0 0.5 0 1>}.RGB) {g=true, r=true}
  ) {speedY=0.5}
  n42 = ComponentMask(None=VectorParameter() {group=SleepyTime, parameterName=PurplePanSpeed, defaultValue=<rgba 0 0.9 0 1>}.RGB) {g=true, r=true}
  n43 = TextureSample(UVs=LinearInterpolate(A=LinearInterpolate(A=@n40, B=@n41, Alpha=@n29.B), B=Panner(Coordinate=@n37, Speed=@n42) {speedY=0.9}, Alpha=@n29.G)) {samplerType=SAMPLERTYPE_Masks, texture=/Game/Characters/Player/Male/Medium/Bodies/M_MED_SleepyTime/Textures/T_MED_SleepTime_Body_Mask02}
  n44 = TextureSample() {samplerType=SAMPLERTYPE_Masks, texture=/Game/Weapons/FORT_Melee/Pickaxe_Sleepy_Time/Textures/T_Pickaxe_Sleepy_Time_FX}
  n45 = FunctionInput(Preview=AppendVector(A=@n44.RGB, B=@n44.RGB)) {bUsePreviewValueAsDefault=true, inputName=Mask, inputType=FunctionInput_Vector4}
  n46 = Multiply(A=Multiply(A=@n43.R, B=ComponentMask(None=@n45) {r=true}), B=OneMinus(None=ComponentMask(None=@n45) {b=true}))
  n47 = ScalarParameter() {group=SleepyTime, parameterName=OverlayTechArt}
  n48 = StaticSwitchParameter(True=@n46, False=@n47) {parameterName=OverlayTechArt}
  n49 = LinearInterpolate(
   A=@n16
   B=Multiply(
    A=MaterialFunctionCall(Exponen=@n17, Hue=@n30, InsideMaincolor=@n32, InsideDarkColor=@n33, outline color=@n34, BaseRef=@n35) {materialFunction=&SleepTimeFxFunction}.Result
    B=@n48
   )
   Alpha=@n48
  )
  n50 = GetMaterialAttributes(None=FunctionInput() {inputName=Material, inputType=FunctionInput_MaterialAttributes}) {attributeGetTypes [#1]}
  n51 = Multiply(A=@n49, B=Clamp(None=Multiply(A=@n50.BaseColor) {constB=10}))
  n52 = LinearInterpolate(A=Panner(Coordinate=@n2) {speedY=0.25}, B=Panner(Coordinate=@n2) {speedY=0.1}, Alpha=@n3.R)
  n53 = Panner(Coordinate=@n2) {speedY=0.5}
  n54 = TextureSample(UVs=LinearInterpolate(A=LinearInterpolate(A=@n52, B=@n53, Alpha=@n3.B), B=Panner(Coordinate=@n2) {speedY=0.9}, Alpha=@n3.G)) {samplerType=SAMPLERTYPE_Masks, texture=&T_MED_SleepTime_Body_Mask02}
  n55 = TextureCoordinate() {coordinateIndex=1}
  n56 = Frac(None=Multiply(A=ComponentMask(None=@n55) {g=true}) {constB=2})
  n57 = Add(A=@n56)
  n58 = Multiply(A=ScalarParameter() {defaultValue=0.85, group=SleepyTime, parameterName=GradientTransition}) {constB=2}
  n59 = ScalarParameter() {defaultValue=0.1, group=SleepyTime, parameterName=Falloff}
  n60 = Ceil(None=Subtract(A=Subtract(A=@n57, B=@n58), B=Multiply(A=@n54.R, B=@n59)))
  n61 = SetMaterialAttributes!(
   MaterialAttributes=@n50.MaterialAttributes
   EmissiveColor=StaticSwitchParameter(True=@n51, False=@n49) {parameterName=IsSleepyTimeHead}
   OpacityMask=LinearInterpolate(A=@n54.R, B=Clamp(None=@n60), Alpha=@n1)
   Metallic=ScalarParameter() {group=SleepyTime, parameterName=SleepyTimeMetallic}
   Roughness=ScalarParameter() {defaultValue=0.5, group=SleepyTime, parameterName=SleepyTimeRoughness}
   BaseColor=Constant3Vector()
   Specular=ScalarParameter() {group=SleepyTime, parameterName=SleepyTimeSpecular}
  ) {attributeSetTypes [
    #2
    #3
    #4
    #5
    #1
    #6
   ]}
  n62 = ComponentMask(None=@n45) {b=true}
  n63 = Add(A=ChannelMaskParameter(None=@n45) {group=SleepyTime, parameterName=YellowAreaChannel}.R, B=@n62)
  n64 = Constant() {r=1}
  FunctionOutput(None=BlendMaterialAttributes(A=@n50.MaterialAttributes, B=@n61, Alpha=LinearInterpolate(A=@n63, B=@n64, Alpha=@n4)))
  n65 = OneMinus(None=ComponentMask(None=@n45) {a=true})
  n66 = ScalarParameter() {defaultValue=1, group=SleepyTime, parameterName=isNotGlider}
  LinearInterpolate(A=@n65, B=@n15, Alpha=@n66)
  LinearInterpolate(A=@n49, B=@n51, Alpha=ScalarParameter() {group=SleepyTime, parameterName=IsSleepyTimeHead})
  n67 = Multiply(A=@n9.RGB, B=@n62)
  LinearInterpolate(A=@n16, B=@n67, Alpha=Multiply(A=@n62, B=@n66))
  n68 = MaterialFunctionCall(Float2=@n36) {materialFunction=&BreakOutFloat2Components}
  MaterialFunctionCall(X=Multiply(A=@n68.R, B=@n18), Y=@n68.G) {materialFunction=&MakeFloat2}
  Multiply(A=@n46, B=@n47)
  OneMinus(None=@n56)
  StaticSwitchParameter(True=@n15, False=@n65) {defaultValue=true, parameterName=IsNotGlider}
  StaticSwitchParameter(True=@n16, False=@n67) {defaultValue=true, parameterName=IsNotGlider}
  n69 = Constant()
  StaticSwitchParameter(True=@n3.G, False=@n69) {parameterName=UseHueShift}
  StaticSwitchParameter(True=@n3.B, False=@n69) {parameterName=UseHueShift}
  StaticSwitchParameter(True=@n64, False=@n63) {parameterName=UseHueShift}
  TextureSample(UVs=@n36) {
   samplerType=SAMPLERTYPE_LinearColor
   texture=&T_MED_SleepTime_Body_Mask01
  }
 ]}
