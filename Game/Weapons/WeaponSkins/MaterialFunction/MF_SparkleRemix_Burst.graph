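# Material function for the SparkleRemix weapon skin that blends an animated, hue-cycling emissive burst into the input material by an Alpha input.
# It samples a rotating burst texture on UV1, using red for a sine-animated SphereMask band, blue plus time for hue shift, and alpha as mask.
# Emote_EnvTime from the Fortnite parameter collection, or the isAlwaysDancing override, drives a dancing state that widens, speeds up and brightens the burst.
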MaterialFunction /Game/Weapons/WeaponSkins/MaterialFunction/MF_SparkleRemix_Burst
graph {nodes [
  n1 = FunctionInput() {inputName=Material, inputType=FunctionInput_MaterialAttributes}
  n2 = CollectionParameter() {collection=&FortniteMaterialParameters, parameterName=Emote_EnvTime}
  n3 = Constant() {r=0.001}
  n4 = Constant() {r=1}
  n5 = Constant()
  n6 = LinearInterpolate(A=Saturate(None=If(A=@n2, B=@n3, "A > B"=@n4, "A < B"=@n5)), Alpha=ScalarParameter() {group=Burst, parameterName=isAlwaysDancing})
  n7 = LinearInterpolate(A=Constant() {r=0.2}, B=Constant() {r=0.9}, Alpha=@n6)
  n8 = ScalarParameter() {defaultValue=0.33, group=Burst, parameterName=Time Scale}
  n9 = Multiply(A=Time(), B=LinearInterpolate(A=@n8, B=Multiply(A=@n8) {constB=2}, Alpha=@n6))
  n10 = Multiply(A=TextureCoordinate() {coordinateIndex=1}, B=ScalarParameter() {defaultValue=0.7, group=Burst, parameterName=Scale})
  n11 = AppendVector(
   A=ScalarParameter() {defaultValue=-0.0286, group=Burst, parameterName=X Offset}
   B=ScalarParameter() {defaultValue=0.1172, group=Burst, parameterName=Y Offset}
  )
  n12 = TextureSample(UVs=Rotator(Coordinate=Add(A=@n10, B=@n11), Time=@n9) {speed=1}) {
   samplerSource=SSM_Clamp_WorldGroupSettings
   samplerType=SAMPLERTYPE_LinearColor
   texture=/Game/Characters/Player/Female/Medium/Bodies/F_MED_SparkleRemix/Textures/Burst/Burst_M
  }
  n13 = SphereMask(
   A=LinearInterpolate(
    A=Add(A=@n7) {constB=0.001}
    B=Constant() {r=1.1}
    Alpha=MaterialFunctionCall(Value=@n9) {materialFunction=&LinearSine}.Linear Sine
   ) {constA=0.1, constB=0.98}
   B=@n12.R
   Radius=@n7
   Hardness=Constant() {r=1}
  )
  n14 = MaterialFunctionCall(Hue Shift Percentage=Add(A=@n12.B, B=@n9), Texture=VectorParameter() {parameterName=HUE, defaultValue=<rgba 0 1 0 1>}.RGB) {materialFunction=&HueShift}
  n15 = ScalarParameter() {defaultValue=30, group=Burst, parameterName=Global Atomic Brightness}
  n16 = Multiply(
   A=Multiply(A=Multiply(A=@n13, B=@n14.Result), B=@n12.A)
   B=LinearInterpolate(A=@n15, B=Multiply(A=@n15) {constB=6.5}, Alpha=@n6) {constB=2.5}
  )
  FunctionOutput(None=BlendMaterialAttributes(
    A=@n1
    B=SetMaterialAttributes!(MaterialAttributes=@n1, EmissiveColor=@n16) {attributeSetTypes [#1]}
    Alpha=FunctionInput() {inputName=Alpha, inputType=FunctionInput_Scalar}
   ))
 ]}
