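# Weapon skin material function that layers a spray paint texture over incoming material attributes, revealed by a transition parameter.
# A gradient, by default pre-skinned position along a selectable axis scaled by local bounds, can be flipped and is offset by the transition and tiled noise.
# Ceil and clamp turn it into a hard-edged mask, multiplied by paint alpha, that blends base color, specular, roughness and metallic.
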MaterialFunction /Game/Weapons/WeaponSkins/MaterialFunction/MF_SprayPaintFunction
graph {nodes [
  Frac(None=Multiply(A=Time(), B=ScalarParameter() {defaultValue=0.5, group=SprayPaint, parameterName=DebugTransitionSpeed}))
  n1 = GetMaterialAttributes(None=FunctionInput() {inputType=FunctionInput_MaterialAttributes}) {attributeGetTypes [
    #1
    #2
    #3
    #4
   ]}
  n2 = TextureCoordinate()
  n3 = AppendVector(A=ScalarParameter() {defaultValue=1.45, parameterName=X_Scale, sliderMin=1}, B=ScalarParameter() {defaultValue=1.45, parameterName=Y_Scale})
  n4 = ScalarParameter() {defaultValue=-0.4619, parameterName=X_Offset}
  n5 = ScalarParameter() {defaultValue=-0.0286, parameterName=Y_Offset}
  n6 = TextureSampleParameter2D(UVs=Add(A=Multiply(A=@n2, B=@n3), B=AppendVector(A=@n4, B=@n5))) {
   group=SprayPaint
   parameterName=SprayPaint_Texture
   texture=/Game/Gadgets/Assets/VinderTech_GliderChute/Glider_GraffitiRemix/Textures/T_Glider_Graffiti_Remix_paint_only_D
  }
  n7 = ChannelMaskParameter(None=VertexInterpolator(VS=PreSkinnedPosition()).PS) {parameterName=SprayPaintMaskDirection}
  n8 = ChannelMaskParameter(None=PreSkinnedLocalBounds().Max) {parameterName=SprayPaintMaskDirection}
  n9 = Multiply(A=@n8.R) {constB=0.5}
  n10 = FunctionInput(Preview=Divide(A=Add(A=@n7.R, B=@n9), B=@n8.R)) {bUsePreviewValueAsDefault=true, inputName=TransitionGradient, inputType=FunctionInput_Scalar}
  n11 = LinearInterpolate(A=OneMinus(None=@n10), B=@n10, Alpha=ScalarParameter() {group=SprayPaint, parameterName=FlipSprayDirection})
  n12 = OneMinus(None=ScalarParameter() {group=SprayPaint, parameterName=SprayPaintTransition})
  n13 = TextureSampleParameter2D(UVs=Multiply(A=TextureCoordinate(), B=ScalarParameter() {defaultValue=4, group=SprayPaint, parameterName=TransitionNoiseTiling})) {
   group=SprayPaint
   parameterName=TransitionPattern
   samplerType=SAMPLERTYPE_LinearColor
   texture=&HighContrastNoise
  }
  n14 = Subtract(
   A=Subtract(A=Add(A=@n11), B=Multiply(A=@n12) {constB=2})
   B=Multiply(
    A=ChannelMaskParameter(None=@n13.RGB) {group=SprayPaint, parameterName=TransitionMaskChanel}.R
    B=ScalarParameter() {defaultValue=0.5, group=SprayPaint, parameterName=TransitionSpread}
   )
  )
  n15 = Multiply(A=Clamp(None=Ceil(None=@n14)), B=@n6.A)
  n16 = SetMaterialAttributes!(
   MaterialAttributes=@n1.MaterialAttributes
   BaseColor=LinearInterpolate(A=@n1.BaseColor, B=@n6.RGB, Alpha=@n15) {constB=0}
   Specular=LinearInterpolate(A=@n1.Specular, B=ScalarParameter() {defaultValue=0.5, group=SprayPaint, parameterName=Spraypaint_Spec}, Alpha=@n15)
   Roughness=LinearInterpolate(A=@n1.Roughness, B=ScalarParameter() {defaultValue=0.33, group=SprayPaint, parameterName=Spraypaint_Roughness}, Alpha=@n15)
   Metallic=LinearInterpolate(A=@n1.Metallic, B=ScalarParameter() {group=SprayPaint, parameterName=Spraypaint_Metallness}, Alpha=@n15)
  ) {attributeSetTypes [
    #1
    #2
    #3
    #4
   ]}
  FunctionOutput(None=@n16)
 ]}
