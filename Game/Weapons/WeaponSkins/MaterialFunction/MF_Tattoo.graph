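# MF_Tattoo is a weapon skin material function that overlays a tattoo on the base color and optionally emissive of incoming material attributes.
# A static switch takes color and mask from a UV1 tattoo texture, or from a color parameter and the FXMask input.
# The channel-selected mask and opacity drive the base color blend.
# An optional emissive tint masked by the tattoo blends between multiplying and adding to existing emissive, unaffected by opacity.
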MaterialFunction /Game/Weapons/WeaponSkins/MaterialFunction/MF_Tattoo
graph {nodes [
  n1 = GetMaterialAttributes(None=FunctionInput() {inputName=MaterialAttributes, inputType=FunctionInput_MaterialAttributes}) {attributeGetTypes [#1, #2]}
  n2 = TextureCoordinate() {coordinateIndex=1}
  n3 = TextureSampleParameter2D(UVs=@n2) {group=MF_Tattoo, parameterName=Tattoo_Texture}
  n4 = StaticSwitchParameter(True=@n3.RGB, False=VectorParameter() {group=MF_Tattoo, parameterName=TattooColor, defaultValue=<rgba 0 0 0 1>}.RGB) {group=MF_Tattoo, parameterName=UseTattooTexture}
  n5 = FunctionInput(Preview=TextureObject()) {bUsePreviewValueAsDefault=true, inputName=FXMask, inputType=FunctionInput_Texture2D}
  n6 = ChannelMaskParameter(None=StaticSwitchParameter(True=@n3.RGBA, False=TextureSample(UVs=@n2, Tex=@n5).RGBA) {group=MF_Tattoo, parameterName=UseTattooTexture}) {group=MF_Tattoo, parameterName=TattooMaskChannel}
  n7 = LinearInterpolate(A=@n1.BaseColor, B=@n4, Alpha=@n6.R)
  n8 = ScalarParameter() {defaultValue=1, group=MF_Tattoo, parameterName=TattooOpacity}
  n9 = VectorParameter() {group=MF_Tattoo, parameterName=Tattoo_EmissiveColor, defaultValue=<rgba 0 0 0 1>}
  n10 = ScalarParameter() {group=MF_Tattoo, parameterName=UseTattooBaseColorforEmissive}
  n11 = Multiply(
   A=Multiply(A=@n6.R, B=LinearInterpolate(A=@n9.RGB, B=@n4, Alpha=@n10))
   B=ScalarParameter() {group=MF_Tattoo, parameterName=Tattoo_EmissiveStrength}
  )
  n12 = LinearInterpolate(
   A=Multiply(A=@n1.EmissiveColor, B=LinearInterpolate(B=@n11, Alpha=@n6.R) {constA=1})
   B=Add(A=@n1.EmissiveColor, B=@n11)
   Alpha=ScalarParameter() {group=MF_Tattoo, parameterName=AddorMultiplyTattooEmisive}
  )
  FunctionOutput(None=SetMaterialAttributes!(
    MaterialAttributes=@n1.MaterialAttributes
    BaseColor=LinearInterpolate(A=@n1.BaseColor, B=@n7, Alpha=@n8)
    EmissiveColor=StaticSwitchParameter(True=@n12, False=@n1.EmissiveColor) {group=MF_Tattoo, parameterName=bEmissiveTattoo}
   ) {attributeSetTypes [#1, #2]})
 ]}
