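# Material function for Viper glitch weapon skins that adds an animated, curve-colored glow to incoming material attributes.
# Mask red sets the effect area, green the flippable gradient, blue the pan direction.
# Panning soft noise distorts a second noise pattern, which erodes the gradient and drives curve-atlas colors for body and tip glows.
# The blended glow adds to Emissive and replaces masked BaseColor; Roughness and Specular blends are left unconnected.
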MaterialFunction /Game/Weapons/WeaponSkins/MaterialFunction/MF_ViperGlitchWeapon
graph {nodes [
  n1 = FunctionInput(Preview=TextureSample() {texture=/Game/Weapons/FORT_Melee/Pickaxe_Viper/Textures/T_Pickaxe_Viper_FX}.RGB) {inputName=Mask}
  n2 = ComponentMask(None=@n1) {r=true}
  n3 = ComponentMask(None=@n1) {g=true}
  n4 = ScalarParameter() {defaultValue=1, group=ViperGlitch, parameterName=FlipGradientDirection, sortPriority=1}
  n5 = Multiply(A=@n2, B=LinearInterpolate(A=@n3, B=OneMinus(None=@n3), Alpha=@n4))
  n6 = Multiply(
   A=TextureCoordinate()
   B=AppendVector(
    A=ScalarParameter() {defaultValue=4, group=ViperGlitch, parameterName=Pattern_U_Tiling}
    B=ScalarParameter() {defaultValue=4, group=ViperGlitch, parameterName=Pattern_V_Tiling}
   )
  )
  n7 = Multiply(
   A=TextureCoordinate()
   B=AppendVector(
    A=ScalarParameter() {defaultValue=4, group=ViperGlitch, parameterName=Noise_U_Tiling}
    B=ScalarParameter() {defaultValue=4, group=ViperGlitch, parameterName=Noise_V_Tiling}
   )
  )
  n8 = VectorParameter() {group=ViperGlitch, parameterName=NoisePanSpeedandDirection, defaultValue=<rgba -1 0 0 1>}
  n9 = ComponentMask(None=LinearInterpolate(A=Multiply(A=@n8.RGB) {constB=-1}, B=@n8.RGB, Alpha=ComponentMask(None=@n1) {b=true})) {g=true, r=true}
  n10 = ComponentMask(None=TextureSample(UVs=Panner(Coordinate=@n7, Speed=@n9) {speedX=0.1, speedY=-0.1}) {texture=/Game/Characters/Player/Male/Medium/Bodies/M_MED_Viper/Textures/Glitch/T_SoftNoise}.RGB) {g=true, r=true}
  n11 = Multiply(A=Subtract(A=Multiply(A=@n10) {constB=2}), B=ScalarParameter() {defaultValue=0.4, group=ViperGlitch, parameterName=Distortion_Strength})
  n12 = VectorParameter() {group=ViperGlitch, parameterName=TexturePanSpeedandDirection, defaultValue=<rgba 1 0 0 1>}
  n13 = LinearInterpolate(A=Multiply(A=@n12.RGB) {constB=-1}, B=@n12.RGB, Alpha=ComponentMask(None=@n1) {b=true})
  n14 = TextureSample(UVs=Panner(Coordinate=Add(A=@n6, B=@n11), Speed=ComponentMask(None=@n13) {g=true, r=true}) {speedX=-0.1, speedY=0.1}) {texture=&T_SoftNoise}
  n15 = Multiply(A=@n14.R, B=ScalarParameter() {defaultValue=0.1, group=ViperGlitch, parameterName=TipNoiseStrength})
  n16 = CurveAtlasRowParameter(CurveTime=Clamp(None=Subtract(A=@n5, B=@n15))) {
   atlas=&AthenaCurveAtlas01
   curve=/Game/Athena/LinearColorCurves/vipertipglowcolor
   defaultValue=138
   group=ViperGlitch
   parameterName=TipGlowCurve
  }
  n17 = Multiply(A=@n16, B=ScalarParameter() {defaultValue=2, group=ViperGlitch, parameterName=TipGlowStrength})
  n18 = OneMinus(None=@n3)
  n19 = Clamp(None=Subtract(A=LinearInterpolate(A=@n3, B=@n18, Alpha=@n4), B=ScalarParameter() {group=ViperGlitch, parameterName=GradientFalloff}))
  n20 = Multiply(A=@n14.R, B=ScalarParameter() {defaultValue=0.5, group=ViperGlitch, parameterName=NoiseDistortionStrength})
  n21 = CurveAtlasRowParameter(CurveTime=@n14.R) {
   atlas=&AthenaCurveAtlas01
   curve=/Game/Athena/LinearColorCurves/viperpickaxeColorCurve
   defaultValue=137
   group=ViperGlitch
   parameterName=ColorCurve
  }
  n22 = Multiply(A=Clamp(None=Subtract(A=@n19, B=@n20)), B=@n21)
  n23 = ScalarParameter() {defaultValue=5, group=ViperGlitch, parameterName=EmissiveStrength}
  n24 = Multiply(A=@n2, B=Power(Base=Multiply(A=@n22, B=@n23)))
  Add(A=@n17, B=@n24)
  n25 = GetMaterialAttributes(None=FunctionInput() {inputName=Material, inputType=FunctionInput_MaterialAttributes}) {attributeGetTypes [
    #1
    #2
    #3
    #4
    #5
   ]}
  n26 = Multiply(A=Desaturation(None=@n16, Fraction=Constant() {r=1})) {constB=10}
  n27 = LinearInterpolate(A=@n24, B=@n17, Alpha=Clamp(None=Power(Base=@n26)))
  FunctionOutput(None=SetMaterialAttributes!(
    MaterialAttributes=@n25.MaterialAttributes
    EmissiveColor=Add(A=@n25.EmissiveColor, B=@n27)
    BaseColor=LinearInterpolate(A=@n25.BaseColor, B=@n27, Alpha=@n2)
   ) {attributeSetTypes [#1, #3]})
  LinearInterpolate(A=@n25.Roughness, B=Constant(), Alpha=@n2)
  LinearInterpolate(A=@n25.Specular, B=Constant() {r=1}, Alpha=@n2)
 ]}
