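# Marshmello weapon-skin material function that overlays audio-reactive emissive light-show patterns on the input material attributes.
# Event audio spectrum, optionally height-blended, plus a fake-audio fade drives hue shifts, brightness and panning.
# Random or forced selection picks among wavy pulse, stepped scrolling stripe and two-band EQ patterns using sphere masks, LED dots and Fresnel.
# The result darkens base color, sets roughness and blends through the Alpha mask only while signal is present.
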MaterialFunction /Game/Weapons/WeaponSkins/MaterialFunction/Marshmello
graph {nodes [
  n1 = VertexInterpolator(VS=PreSkinnedPosition())
  n2 = Constant() {r=128}
  n3 = CollectionParameter() {collection=&FortniteMaterialParameters, parameterName=EventAudio_Spectrum}
  n4 = MaterialFunctionCall(Float4=Power(Base=@n3)) {materialFunction=&BreakOutFloat4Components}
  n5 = PreSkinnedLocalBounds()
  n6 = PreSkinnedPosition()
  n7 = ShaderStageSwitch(PixelShader=VertexInterpolator(VS=@n6).PS, VertexShader=@n6)
  n8 = MaterialFunctionCall!(1=@n4.R, 2=@n4.R, 3=@n4.R, 4=@n4.R, A=ComponentMask(None=InverseLinearInterpolate(A=@n5.Min, B=@n5.Min, Value=@n7)) {b=true}) {materialFunction=/Engine/Functions/Engine_MaterialFunctions02/Lerp_Multiple_Float}
  n9 = ConstantBiasScale(None=Sine(None=Time() {bOverride_Period=true, period=8}) {period=8}) {bias=0.5, scale=1}
  n10 = Constant() {r=2}
  n11 = LinearInterpolate(
   A=ScalarParameter() {group=Marshemello, parameterName=FakeAudioData}
   B=Clamp(None=Power(Base=@n9, Exp=@n10)) {maxDefault=0.5}
   Alpha=ScalarParameter() {group=Marshemello, parameterName=FakeAudioData_Fade}
  )
  n12 = LinearInterpolate(
   A=Max(
    A=CollectionParameter() {collection=&FortniteMaterialParameters, parameterName=Emote_EnvTime}
    B=StaticSwitchParameter(True=@n8.Lerp 4 Inputs, False=@n4.G) {group=Marshemello, parameterName=UseHeightAudioBlend}
   )
   Alpha=@n11
  )
  n13 = Round(None=Frac(None=Divide(A=@n12, B=@n2) {constB=2.1111}))
  n14 = LinearInterpolate(
   A=ScalarParameter() {defaultValue=100, group=Marshemello, parameterName=Pattern 1 Panner Spread 1}
   B=ScalarParameter() {defaultValue=40, group=Marshemello, parameterName=Pattern 1 Panner Spread 2}
   Alpha=@n13
  )
  n15 = Divide(A=@n12, B=Constant() {r=32}) {constB=2.1111}
  n16 = Multiply(
   A=Time() {bOverride_Period=true, period=1}
   B=LinearInterpolate(
    A=VectorParameter() {group=Marshemello, parameterName=Panning Speed, defaultValue=<rgba 0 0.5 0 1>}.RGB
    B=VectorParameter() {group=Marshemello, parameterName=Panning Speed2, defaultValue=<rgba 0 0 -0.5 1>}.RGB
    Alpha=Frac(None=@n15)
   )
  )
  n17 = Add(A=ComponentMask(None=Divide(A=@n1.PS, B=@n14)) {b=true, g=true}, B=ComponentMask(None=Frac(None=@n16)) {b=true, g=true})
  n18 = Add(A=TextureCoordinate(), B=@n17)
  n19 = ComponentMask(None=@n18) {g=true}
  n20 = Constant() {r=0.5}
  n21 = ComponentMask(None=@n18) {r=true}
  n22 = Add(A=@n19, B=Divide(A=Sine(None=@n21)) {constB=-10})
  n23 = Power(Base=AppendVector(A=Distance(A=@n19, B=@n20), B=Distance(A=@n22, B=@n20))) {constExponent=1.25}
  ComponentMask(None=@n23) {g=true}
  ComponentMask(None=@n23) {r=true}
  n24 = FunctionInput() {inputName=Material, inputType=FunctionInput_MaterialAttributes}
  n25 = Constant() {r=8}
  n26 = Multiply(A=@n12, B=@n25) {constB=8}
  n27 = Divide(A=Divide(A=Ceil(None=@n26), B=@n25) {constB=8}) {constB=2.1111}
  n28 = LinearInterpolate(
   A=StaticSwitchParameter(True=@n8.Lerp 4 Inputs, False=@n4.R) {group=Marshemello, parameterName=UseHeightAudioBlend}
   Alpha=Saturate(None=@n11)
  )
  n29 = Power(Base=@n28, Exp=ScalarParameter() {defaultValue=2, group=Marshemello, parameterName=Pattern 1 Pulse Power})
  n30 = Ceil(None=Multiply(A=@n29) {constB=8})
  n31 = Add(
   A=Divide(A=FunctionInput() {inputName=Offset, inputType=FunctionInput_Scalar}) {constB=1.5}
   B=Divide(A=Divide(A=@n30) {constB=8}, B=Constant() {r=2}) {constB=20}
  )
  n32 = TextureSampleParameter2D(UVs=@n17) {
   group=Marshemello
   parameterName=Wavy
   samplerType=SAMPLERTYPE_Masks
   texture=/Game/Accessories/Hats/M_MED_Marshmellow_DJ_Hat/Textures/Mello_M
  }
  n33 = VectorParameter() {group=Marshemello, parameterName=StartingHue, defaultValue=<rgba 0 1 0.3489 1>}
  n34 = MaterialFunctionCall(Hue Shift Percentage=Add(A=Add(A=@n27, B=@n31), B=LinearInterpolate(A=@n32.R, B=@n32.R, Alpha=@n13)), Texture=@n33.RGB) {materialFunction=&HueShift}
  n35 = Multiply(A=@n29, B=ScalarParameter() {defaultValue=4000, group=Marshemello, parameterName=Pattern 1 Brightness}) {constB=2}
  n36 = Multiply(A=NamedRerouteUsage(), B=ScalarParameter() {defaultValue=2, parameterName=LED tiling})
  n37 = VectorParameter() {parameterName=DotTilingAdjustment, defaultValue=<rgba 1 1 0 1>}
  n38 = AppendVector(A=@n37.R, B=@n37.R)
  n39 = LinearInterpolate(
   A=Constant() {r=1}
   B=TextureSample(UVs=Multiply(A=@n36, B=@n38)) {texture=/Game/Accessories/Hats/M_MED_Marshmellow_DJ_Hat/Textures/Mello_Dots}.G
   Alpha=ScalarParameter() {defaultValue=1, group=Marshemello, parameterName=Overlay LED Dots}
  )
  n40 = Saturate(None=Fresnel(ExponentIn=ScalarParameter() {defaultValue=0.0025, group=Marshemello, parameterName=Emissive Fresnel}) {baseReflectFraction=0})
  n41 = LinearInterpolate(A=Multiply(A=Multiply(A=@n34.Result, B=@n35), B=@n39), Alpha=@n40) {constB=0}
  n42 = MaterialFunctionCall(Hue Shift Percentage=@n27, Texture=@n33.RGB) {materialFunction=&HueShift}
  n43 = Multiply(A=NamedRerouteUsage(), B=ScalarParameter() {defaultValue=6, parameterName=StripeTiling})
  n44 = Time() {period=1}
  n45 = Multiply(
   A=@n44
   B=AppendVector(
    A=ScalarParameter() {defaultValue=2, group=Marshemello, parameterName=Pattern 3 Panner X}
    B=ScalarParameter() {group=Marshemello, parameterName=Pattern 3 Panner Y}
   )
  )
  n46 = Constant() {r=6}
  n47 = Divide(A=Round(None=Multiply(A=@n45, B=@n46) {constA=12, constB=12}), B=@n46) {constB=12}
  n48 = TextureSample(UVs=Multiply(A=@n38, B=Add(A=@n43, B=@n47))) {samplerType=SAMPLERTYPE_Masks, texture=&Mello_M}
  n49 = LinearInterpolate(
   A=Multiply(A=Multiply(A=@n42.Result, B=@n48.A), B=ScalarParameter() {defaultValue=4000, group=Marshemello, parameterName=Pattern 3 Brightness})
   Alpha=@n40
  ) {constB=0}
  n50 = LinearInterpolate(
   A=ScalarParameter() {group=Marshemello, parameterName=Emote_Rand}
   B=ScalarParameter() {group=Marshemello, parameterName=Pattern Select}
   Alpha=ScalarParameter() {group=Marshemello, parameterName=Force Pattern Select}
  )
  n51 = SphereMask(A=@n50, B=Constant() {r=100}, Radius=Constant() {r=20}, Hardness=Constant() {r=1})
  n52 = Constant() {r=3}
  n53 = CollectionParameter() {collection=&FortniteMaterialParameters, parameterName=EventAudio_Amplitude}
  n54 = Power(Base=@n53)
  n55 = Constant() {r=1}
  n56 = Subtract(A=FunctionInput() {inputName=EQ High, inputType=FunctionInput_Scalar}, B=@n55)
  n57 = Multiply(
   A=MaterialFunctionCall(Hue Shift Percentage=Divide(A=@n44) {constB=43.21}, Texture=@n33.RGB) {materialFunction=&HueShift}.Result
   B=Add(A=Multiply(A=@n54, B=@n52), B=Saturate(None=@n56))
  )
  n58 = Subtract(A=FunctionInput() {inputName=EQ Low, inputType=FunctionInput_Scalar}, B=@n55)
  n59 = Multiply(A=@n28, B=Constant() {r=1.2})
  n60 = Constant() {r=1.6}
  n61 = Multiply(
   A=MaterialFunctionCall(Hue Shift Percentage=@n44, Texture=@n33.RGB) {materialFunction=&HueShift}.Result
   B=Add(A=Saturate(None=@n58), B=Power(Base=@n59, Exp=@n60))
  )
  n62 = Multiply(A=ScalarParameter() {defaultValue=4000, group=Marshemello, parameterName=Pattern 4 Brightness}, B=Multiply(A=@n39, B=Add(A=@n57, B=@n61)))
  n63 = Constant() {r=33}
  n64 = Constant() {r=10}
  n65 = Constant() {r=1}
  n66 = Max(A=LinearInterpolate(
    A=LinearInterpolate(A=@n41, B=@n49, Alpha=@n51)
    B=LinearInterpolate(A=@n62, Alpha=@n40) {constB=0}
    Alpha=SphereMask(A=@n50, B=@n63, Radius=@n64, Hardness=@n65)
   ) {constB=0}) {constB=0}
  n67 = LinearInterpolate(A=BreakMaterialAttributes!(Attr=@n24).BaseColor, B=Constant(), Alpha=Saturate(None=@n66))
  n68 = Constant() {r=0.3}
  n69 = Saturate(None=If(A=@n28, B=Constant() {r=0.001}, "A > B"=Constant() {r=1}, "A < B"=Constant()))
  n70 = FunctionInput(Preview=TextureSample() {samplerType=SAMPLERTYPE_Masks, texture=/Game/Accessories/Hats/M_MED_Marshmellow_DJ_Hat/Textures/T_Marshmello_FX_Marshmellow_DJ1_FX}.R) {inputName=Alpha, inputType=FunctionInput_Scalar}
  FunctionOutput(None=BlendMaterialAttributes(
    A=@n24
    B=SetMaterialAttributes!(MaterialAttributes=@n24, BaseColor=@n67, EmissiveColor=@n66, Roughness=@n68) {attributeSetTypes [#1, #2, #3]}
    Alpha=Multiply(A=@n69, B=@n70)
   ))
  n71 = PreSkinnedNormal()
  n72 = Abs(None=VertexInterpolator(VS=ComponentMask(None=@n71) {b=true}).PS)
  n73 = VertexInterpolator(VS=PreSkinnedPosition())
  n74 = Normalize(VectorInput=Multiply(A=ComponentMask(None=@n73.PS) {g=true, r=true}, B=Constant2Vector() {g=1, r=-1}))
  n75 = Divide(A=Arctangent2Fast(Y=ComponentMask(None=@n74) {g=true}, X=ComponentMask(None=@n74) {r=true})) {constB=3.14}
  n76 = ComponentMask(None=@n73.PS) {b=true}
  n77 = VectorParameter() {group=Marshemello, parameterName=Cylindrecal Mapping Size, defaultValue=<rgba 20 20 20 0>}
  n78 = VectorParameter() {group=Marshemello, parameterName=Cylindrecal Mapping Offset}
  n79 = OneMinus(None=Add(A=Divide(A=@n76, B=@n77.B) {constB=100}, B=@n78.B))
  n80 = Divide(A=ComponentMask(None=@n73.PS) {g=true, r=true}, B=ComponentMask(None=@n77.RGB) {g=true, r=true})
  n81 = ComponentMask(None=@n78.RGB) {g=true, r=true}
  NamedRerouteDeclaration(None=LinearInterpolate(A=AppendVector(A=@n75, B=@n79), B=Add(A=@n80, B=@n81), Alpha=Round(None=@n72))) {name=Cylindrical Mapping, nodeColor=<rgba 0 0 0 1>}
  NamedRerouteDeclaration(None=StaticSwitchParameter(True=TextureCoordinate(), False=TextureCoordinate() {coordinateIndex=1}) {group=Marshemello, parameterName=UseUV0}) {name=UV, nodeColor=<rgba 0.5703 1 0 1>}
 ]}
