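# Material function for weapon skins that adds an automatic headlight glow to emissive at night.
# An inverted Fresnel falloff shapes the headlight color, brightest facing the camera, and adds it to the incoming emissive.
# An offset global time of day, switchable to a constant, drives a remapped cosine sharpened by a high power into a near on/off gate.
# That gate times the Alpha input blends the glowing attributes over the originals.
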MaterialFunction /Game/Weapons/WeaponSkins/MaterialFunction/NightTime_AutomaticHeadlights
graph {nodes [
  n1 = FunctionInput() {inputType=FunctionInput_MaterialAttributes}
  n2 = GetMaterialAttributes(None=@n1) {attributeGetTypes [#1, #2]}
  n3 = Power(
   Base=Fresnel(ExponentIn=ScalarParameter() {defaultValue=0.1, group=Headlights, parameterName=HeadlightFalloff}) {baseReflectFraction=0, exponent=0.04}
   Exp=ScalarParameter() {defaultValue=2, parameterName=Power}
  )
  n4 = Add(
   A=@n2.EmissiveColor
   B=LinearInterpolate(A=VectorParameter() {group=Headlights, parameterName=Headlight Color, defaultValue=<rgba 7 6.6166 5.861 1>}.RGB, Alpha=Saturate(None=@n3)) {constB=0}
  )
  n5 = Add(
   A=StaticSwitchParameter(
    True=CollectionParameter() {collection=&FortniteMaterialParameters, parameterName=TimeOfDay_0-1}
    False=Constant()
   ) {defaultValue=true, group=Headlights, parameterName=UseTimeofDay}
   B=ScalarParameter() {defaultValue=0.48, parameterName=Tod_Offset}
  )
  n6 = LinearInterpolate(
   A=ScalarParameter() {defaultValue=-1, parameterName=Bottom}
   B=ScalarParameter() {defaultValue=1.7, parameterName=Top}
   Alpha=ConstantBiasScale(None=Cosine(None=@n5))
  )
  n7 = OneMinus(None=Clamp(None=Power(Base=@n6) {constExponent=20}))
  n8 = FunctionInput() {inputName=Alpha, inputType=FunctionInput_Scalar}
  FunctionOutput(None=BlendMaterialAttributes(
    A=@n1
    B=SetMaterialAttributes!(MaterialAttributes=@n2.MaterialAttributes, EmissiveColor=@n4) {attributeSetTypes [#2]}
    Alpha=Multiply(A=@n8, B=@n7)
   ) {vertexAttributeBlendType=UseA})
  Transform(None=@n2.Normal)
 ]}
