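# Material function for weapon skins that generates an animated scanline mask from tiled, vertically panning textures.
# A scrolling vertical gradient has a second offset texture subtracted from it, with independent tiling and speed and an opposite default direction.
# A quality switch drops the subtraction on Low and Medium, using only the gradient.
# A thickness bias is then added and the result clamped with an adjustable minimum.
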MaterialFunction /Game/Weapons/WeaponSkins/MaterialFunction/ScanLines
graph {
 properties {bExposeToLibrary=true}
 nodes [
  n1 = FunctionInput(Preview=TextureCoordinate()) {bUsePreviewValueAsDefault=true, inputName=UVs, inputType=FunctionInput_Vector2, sortPriority=6}
  n2 = FunctionInput(Preview=ScalarParameter() {defaultValue=40, group=Scanlines, parameterName=ScanLine_Tiling}) {bUsePreviewValueAsDefault=true, inputName=ScanLine_Tiling, inputType=FunctionInput_Scalar}
  n3 = Constant()
  n4 = FunctionInput(Preview=ScalarParameter() {defaultValue=0.5, group=Scanlines, parameterName=Scanline_PannerSpeed}) {bUsePreviewValueAsDefault=true, inputName=Scanline_PannerSpeed, inputType=FunctionInput_Scalar}
  n5 = TextureSample(UVs=Panner(Coordinate=Multiply(A=@n1, B=@n2), Speed=AppendVector(A=@n3, B=@n4))) {samplerType=SAMPLERTYPE_LinearColor, texture=/Game/Global/Textures/Default/Masks_and_Gradients/T_VertGradient}
  n6 = Multiply(
   A=@n1
   B=FunctionInput(Preview=ScalarParameter() {defaultValue=2, group=Scanlines, parameterName=ScanLine_Tiling_1}) {bUsePreviewValueAsDefault=true, inputName=ScanLine_Tiling_1, inputType=FunctionInput_Scalar}
  )
  n7 = AppendVector(
   A=Constant()
   B=FunctionInput(Preview=ScalarParameter() {defaultValue=-0.5, group=Scanlines, parameterName=Scanline_PannerSpeed_1}) {bUsePreviewValueAsDefault=true, inputName=Scanline_PannerSpeed_1, inputType=FunctionInput_Scalar}
  )
  n8 = Subtract(
   A=@n5.R
   B=TextureSample(UVs=Panner(Coordinate=@n6, Speed=@n7)) {texture=/Game/Global/Textures/Default/Masks_and_Gradients/Scanlines_Offset}.R
  )
  n9 = QualitySwitch(Default=@n8, Low=@n5.R, High=@n8, Medium=@n5.R)
  n10 = ScalarParameter() {group=Scanlines, parameterName=Scanline Thickness Bias}
  FunctionOutput(None=Clamp(None=Add(A=@n9, B=@n10), Min=ScalarParameter() {group=Scanlines, parameterName=Scanline Min}))
 ]
}
