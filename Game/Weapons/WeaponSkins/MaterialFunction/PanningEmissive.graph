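# Weapon-skin material function that layers animated, Fresnel-weighted emissive onto input material attributes, blended by an Alpha mask.
# A gradient pans triplanar over pre-skinned local position or through UVs with optional bump offset, optionally multiplied by two radial spiral masks using an fwidth seam fix.
# Static switches select gradient-map color, flat value, hue shift and scanlines.
# It also outputs the raw gradient and unmasked emissive.
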MaterialFunction /Game/Weapons/WeaponSkins/MaterialFunction/PanningEmissive
graph {nodes [
  n1 = FunctionInput(Preview=VectorParameter() {group=PanningEmissive, parameterName=EmissiveColor, defaultValue=<rgba 1 1 1 1>}.RGB) {bUsePreviewValueAsDefault=true, inputName=EmissiveColorOverride}
  n2 = FunctionInput() {inputName=Material, inputType=FunctionInput_MaterialAttributes}
  n3 = GetMaterialAttributes(None=@n2) {attributeGetTypes [
    #1
    #2
    #3
    #4
   ]}
  n4 = Add(A=@n3.EmissiveColor, B=@n1)
  n5 = StaticBoolParameter() {group=PanningEmissive, parameterName=Use 'Emissive' Texture as Panning Emissive Color, sortPriority=15}
  n6 = Saturate(None=Fresnel(ExponentIn=ScalarParameter() {defaultValue=4, group=PanningEmissive, parameterName=PanningEmissive_Fresnel Exponent}) {baseReflectFraction=0})
  n7 = LinearInterpolate(A=OneMinus(None=@n6), B=@n6, Alpha=ScalarParameter() {group=PanningEmissive, parameterName=Panning Emissive_Invert Fresnel})
  n8 = ScalarParameter() {defaultValue=2, group=PanningEmissive, parameterName=Emissive_BrightnessMax}
  n9 = Multiply(
   A=FunctionInput() {bUsePreviewValueAsDefault=true, inputName=UVOffset, inputType=FunctionInput_Scalar}
   B=AppendVector(A=ScalarParameter() {group=PanningEmissive, parameterName=UVOffsetU}, B=ScalarParameter() {group=PanningEmissive, parameterName=UVOffsetV})
  )
  n10 = PreSkinnedPosition()
  n11 = Divide(
   A=FunctionInput(Preview=VertexInterpolator(VS=@n10).PS) {bUsePreviewValueAsDefault=true, inputName=Local Position}
   B=ScalarParameter() {defaultValue=40, group=PanningEmissive, parameterName=Triplanner_ConstantScaleDivisor}
  )
  n12 = TextureSampleParameter2D() {
   group=PanningEmissive
   parameterName=UVScaleMap
   samplerSource=SSM_Wrap_WorldGroupSettings
   samplerType=SAMPLERTYPE_Masks
   texture=&T_WhiteMask
  }
  n13 = Add(A=@n12.RGB, B=OneMinus(None=ScalarParameter() {defaultValue=1, group=PanningEmissive, parameterName=UVScaleMapIntensity}))
  n14 = ScalarParameter() {defaultValue=1, group=PanningEmissive, parameterName=Emissive Tiling}
  n15 = StaticSwitchParameter(True=Multiply(A=Saturate(None=@n13), B=@n14), False=@n14) {
   desc="This allows modification of the panning EmissiveGradient texture direction using a RG \"UV\" texture.\r\nIf \"WorldAlignedGradient\" is true, it uses RGB."
   group=PanningEmissive
   parameterName=UseUVScaleMap
  }
  n16 = Frac(None=Multiply(A=Time(), B=VectorParameter() {group=PanningEmissive, parameterName=TriPlanner Speed, defaultValue=<rgba 0 0.2 0.2 1>}.RGB))
  n17 = ComponentMask(None=Add(A=Multiply(A=@n15, B=@n11), B=@n16)) {b=true, g=true}
  n18 = TextureObjectParameter() {
   group=PanningEmissive
   parameterName=EmissiveGradient
   samplerType=SAMPLERTYPE_Masks
   texture=/Game/Global/Textures/Default/Masks_and_Gradients/T_VertGradient_MASK
  }
  n19 = Add(A=Multiply(A=@n15, B=@n11), B=@n16)
  n20 = TextureSample(UVs=Add(A=@n9, B=ComponentMask(None=@n19) {b=true, r=true}), Tex=@n18) {samplerType=SAMPLERTYPE_Masks, texture=/Game/Weapons/WeaponSkins/Textures/T_Swirly_Noise_highContrast}
  n21 = FunctionInput(Preview=VertexInterpolator(VS=PreSkinnedNormal()).PS) {bUsePreviewValueAsDefault=true, inputName=Local Normal}
  n22 = ScalarParameter() {defaultValue=1, group=PanningEmissive, parameterName=TriPlanner_AxisBlendStrength}
  n23 = MaterialFunctionCall(In=Abs(None=ComponentMask(None=@n21) {g=true}), Contrast=@n22) {materialFunction=&CheapContrast}
  n24 = LinearInterpolate(
   A=TextureSample(UVs=Add(A=@n9, B=@n17), Tex=@n18) {samplerType=SAMPLERTYPE_Masks, texture=&T_Swirly_Noise_highContrast}.RGB
   B=@n20.RGB
   Alpha=Saturate(None=ComponentMask(None=@n23.Result) {g=true})
  )
  n25 = Add(A=Multiply(A=@n15, B=@n11), B=@n16)
  n26 = TextureSample(UVs=Add(A=@n9, B=ComponentMask(None=@n25) {g=true, r=true}), Tex=@n18) {samplerType=SAMPLERTYPE_Masks, texture=&T_Swirly_Noise_highContrast}
  n27 = Abs(None=ComponentMask(None=@n21) {b=true})
  n28 = Saturate(None=ComponentMask(None=MaterialFunctionCall(In=@n27, Contrast=@n22) {materialFunction=&CheapContrast}.Result) {b=true})
  n29 = FunctionInput() {bUsePreviewValueAsDefault=true, inputName=isVehicle, inputType=FunctionInput_Scalar}
  n30 = LinearInterpolate(
   A=TextureCoordinate()
   B=TextureCoordinate() {coordinateIndex=1}
   Alpha=Round(None=ScalarParameter() {group=PanningEmissive, parameterName=UseSecondUVChannelForPanning})
  )
  n31 = ComponentMask(None=@n15) {
   desc="Kendall added this node on 5/27/2020.\r\nPreviously, this just piped the 3-vector value from\r\nthe texture sample straight into the UV's, which\r\ncaused the material to fail compile"
   g=true
   r=true
  }
  n32 = Panner(
   Coordinate=Add(A=@n9, B=Multiply(A=@n30, B=@n31))
   Speed=AppendVector(
    A=ScalarParameter() {group=PanningEmissive, parameterName=GradientSpeed_X}
    B=ScalarParameter() {defaultValue=0.2, group=PanningEmissive, parameterName=GradientSpeed_Y}
   )
  )
  n33 = BumpOffset(Coordinate=@n32, Height=ScalarParameter() {group=PanningEmissive, parameterName=Bump Offset Height})
  n34 = StaticSwitchParameter(
   True=LinearInterpolate(
    A=LinearInterpolate(A=@n24, B=@n26.RGB, Alpha=@n28)
    B=LinearInterpolate(A=@n20.RGB, B=@n26.RGB, Alpha=@n29)
    Alpha=ScalarParameter() {group=PanningEmissive, parameterName=Triplanner_Z axis Only}
   )
   False=TextureSample(UVs=StaticSwitchParameter(True=@n33, False=@n32) {group=PanningEmissive, parameterName=use Bump Offset on UV Panning}, Tex=@n18) {samplerType=SAMPLERTYPE_Masks, texture=&T_Swirly_Noise_highContrast}.RGB
  ) {defaultValue=true, group=PanningEmissive, parameterName=World Aligned Gradient}
  n35 = MaterialFunctionCall(Float2=TextureCoordinate()) {materialFunction=&BreakOutFloat2Components}
  n36 = ScalarParameter() {group=PanningEmissive, parameterName=Spiral_U_Offset}
  n37 = ScalarParameter() {group=PanningEmissive, parameterName=Spiral_U_Offset}
  n38 = Multiply(
   A=AppendVector(A=Subtract(A=@n35.R, B=@n36), B=Subtract(A=@n35.G, B=@n37))
   B=ScalarParameter() {defaultValue=1, group=PanningEmissive, parameterName=Spiral_UV_Scale}
  )
  n39 = ComponentMask(None=VectorParameter() {group=PanningEmissive, parameterName=Spiral_UV_Center, defaultValue=<rgba 0.5 0.5 0 1>}.RGB) {g=true, r=true}
  n40 = MaterialFunctionCall!(Vector or UVs=VertexInterpolator(VS=Subtract(A=@n38, B=@n39) {constB=0.5}).PS) {materialFunction=/Game/Athena/SafeZone/S11/MF_VectorToRadialValueFast}
  n41 = ComponentMask(None=@n40.Result) {r=true}
  MaterialExpressionFrac_5 = Frac(None=@n41)
  MaterialExpressionSubtract_3 = Subtract(A=Frac(None=Add(A=@n41) {constB=0.5})) {constB=0.5}
  n42 = Custom!(u0=@MaterialExpressionFrac_5, u1=@MaterialExpressionSubtract_3) {
   code="return ((fwidth(u0) < fwidth(u1)-0.001)? u0:u1);"
   description=FixPerPixelCylidricalUV
   outputType=CMOT_Float1
   inputs [
    {inputName=u0, input {expression=&self:MaterialExpressionFrac_5}}
    {inputName=u1, input {expression=&self:MaterialExpressionSubtract_3}}
   ]
  }
  n43 = ComponentMask(None=@n42) {r=true}
  n44 = Time()
  n45 = ScalarParameter() {defaultValue=1, group=PanningEmissive, parameterName=Spiral Mask1 Speed}
  n46 = Add(A=@n43, B=Frac(None=Multiply(A=@n44, B=@n45)))
  n47 = ComponentMask(None=@n40.Result) {g=true}
  n48 = Multiply(A=@n47, B=ScalarParameter() {defaultValue=4, group=PanningEmissive, parameterName=Spiral Mask1 VTiling}) {constB=4}
  n49 = TextureSampleParameter2D(UVs=AppendVector(A=@n46, B=Add(A=@n46, B=@n48))) {
   group=PanningEmissive
   parameterName=Spiral Mask1
   samplerType=SAMPLERTYPE_Masks
   texture=&T_VertGradient_MASK
  }
  n50 = ScalarParameter() {defaultValue=-1, group=PanningEmissive, parameterName=Spiral Mask2 Speed}
  n51 = Add(A=@n43, B=Frac(None=Multiply(A=@n44, B=@n50) {constB=-1}))
  n52 = Multiply(A=@n47, B=ScalarParameter() {defaultValue=2, group=PanningEmissive, parameterName=Spiral Mask2 VTiling}) {constB=2}
  n53 = TextureSampleParameter2D(UVs=AppendVector(A=@n51, B=Add(A=@n52, B=@n51))) {
   group=PanningEmissive
   parameterName=Spiral Mask2
   samplerType=SAMPLERTYPE_Masks
   texture=&T_VertGradient_MASK
  }
  n54 = StaticSwitchParameter(True=Multiply(A=@n34, B=Multiply(A=@n49.R, B=@n53.R)), False=@n34) {group=PanningEmissive, parameterName=Spiral Panning Emissive}
  n55 = MaterialFunctionCall(Float3=@n54) {materialFunction=&BreakOutFloat3Components}
  n56 = LinearInterpolate(A=ScalarParameter() {defaultValue=0.1, group=PanningEmissive, parameterName=Emissive_BrightnessMIn}, B=@n8, Alpha=@n55.R)
  n57 = Saturate(None=@n56)
  n58 = FunctionInput() {bUsePreviewValueAsDefault=true, inputName=Gmap, inputType=FunctionInput_Scalar, previewValue=<1 0 0 1>}
  n59 = Multiply(
   A=Power(Base=@n58, Exp=ScalarParameter() {defaultValue=2, group=PanningEmissive, parameterName=UV Gradient_Gmap Power})
   B=ScalarParameter() {defaultValue=1, group=PanningEmissive, parameterName=UV Gradient_Gmap Intensity}
  )
  n60 = TextureSampleParameter2D(UVs=LinearInterpolate(A=@n57, B=Multiply(A=@n57, B=@n59), Alpha=ScalarParameter() {parameterName=UV Gradient_ Use Gmap})) {
   group=PanningEmissive
   parameterName=UV Gradient Map
   samplerSource=SSM_Wrap_WorldGroupSettings
   texture=/Game/Global/Textures/Default/Masks_and_Gradients/WeaponGradients/T_BlackBodyGradient
  }
  n61 = ScalarParameter() {defaultValue=1, group=PanningEmissive, parameterName=UV Gradient_Brightness}
  n62 = Multiply(A=@n8, B=@n54)
  n63 = ScalarParameter() {group=PanningEmissive, parameterName=Use Colored Gradient Map}
  n64 = StaticSwitchParameter(
   True=ScalarParameter() {defaultValue=1, group=PanningEmissive, parameterName="Flat Emissive  Mask Value"}
   False=StaticSwitchParameter(True=Multiply(A=@n60.RGB, B=@n61), False=LinearInterpolate(A=@n56, B=@n62, Alpha=@n63)) {group=PanningEmissive, parameterName=use UV gradient Color}
  ) {group=PanningEmissive, parameterName="Use Flat value for Emissive Texture (Optimization!)"}
  n65 = Multiply(A=@n7, B=@n64)
  n66 = Multiply(
   A=StaticSwitch(
    True=StaticSwitchParameter(True=@n4, False=@n3.EmissiveColor) {group=PanningEmissive, parameterName=Add 'Emissive' Texture to Panning Emissive Color, sortPriority=15}
    False=@n1
    Value=@n5
   )
   B=LinearInterpolate(
    A=Multiply(A=@n7, B=@n64)
    B=Multiply(A=@n3.BaseColor, B=@n65)
    Alpha=ScalarParameter() {group=PanningEmissive, parameterName=Panning Emissive_ Use Base Color as Emissive Color}
   )
  )
  n67 = Add(A=@n3.EmissiveColor, B=@n66)
  n68 = Multiply(A=Time(), B=ScalarParameter() {defaultValue=0.05, group=PanningEmissive, parameterName=Panning Emissive HueShiftSpeed})
  n69 = StaticSwitch(
   True=LinearInterpolate(A=@n67, B=@n66, Alpha=ScalarParameter() {group=PanningEmissive, parameterName=Remove Default Emissive From Panning Mask})
   False=StaticSwitchParameter(
    True=MaterialFunctionCall(Hue Shift Percentage=@n68, Texture=@n67) {materialFunction=&HueShift}.Result
    False=@n67
   ) {group=PanningEmissive, parameterName=RainbowHueShift}
   Value=@n5
  )
  n70 = StaticSwitchParameter(
   True=Multiply(
    A=@n69
    B=MaterialFunctionCall!(ScanLine_Tiling_1, Scanline_PannerSpeed_1, ScanLine_Tiling, Scanline_PannerSpeed, UVs) {materialFunction=&/ScanLines}.Result
   )
   False=@n69
  ) {group=PanningEmissive, parameterName=UseScanLines}
  n71 = ScalarParameter() {defaultValue=1, group=PanningEmissive, parameterName=PanningEmissive_Roughness}
  n72 = SetMaterialAttributes!(
   MaterialAttributes=@n2
   EmissiveColor=StaticSwitchParameter(True=Add(A=@n3.EmissiveColor, B=@n70), False=@n70) {group=PanningEmissive, parameterName=Draw Emissive Above Panning Emissive}
   Roughness=Saturate(None=Multiply(A=@n71, B=@n3.Roughness))
   BaseColor=LinearInterpolate(
    A=Constant()
    B=@n3.BaseColor
    Alpha=ScalarParameter() {group=PanningEmissive, parameterName="Panning Emissive_Use Diffuse Map as Base Color (Otherwise Black)"}
   )
   WorldPositionOffset=@n3.WorldPositionOffset
   Opacity=Constant()
  ) {attributeSetTypes [
    #3
    #2
    #1
    #4
    #5
   ]}
  n73 = FunctionInput() {inputName=Alpha, inputType=FunctionInput_Scalar}
  n74 = Multiply(A=@n73, B=@n73)
  n75 = ScalarParameter() {group=PanningEmissive, parameterName=SquareMaskPower}
  FunctionOutput(None=BlendMaterialAttributes(A=@n2, B=@n72, Alpha=LinearInterpolate(A=@n73, B=@n74, Alpha=@n75)))
  FunctionOutput(None=@n55.R) {outputName=PanningGradient}
  FunctionOutput(None=@n65) {outputName=Unmasked Emissive}
 ]}
