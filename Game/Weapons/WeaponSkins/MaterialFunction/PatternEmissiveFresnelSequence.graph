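# Material function for weapon skins that overlays a tiled, panning emissive pattern on incoming material attributes.
# The mask texture's red channel times the Alpha input blends toward a variant with black base color and CoreColor added to emissive.
# An optional brightness input from a panning emissive effect boosts CoreColor, gated by an interaction parameter.
# A static switch can bypass the pattern on mobile; despite the name, no Fresnel is computed here.
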MaterialFunction /Game/Weapons/WeaponSkins/MaterialFunction/PatternEmissiveFresnelSequence
graph {nodes [
  n1 = GetMaterialAttributes(None=FunctionInput() {inputName=Material, inputType=FunctionInput_MaterialAttributes}) {attributeGetTypes [#1, #2, #3]}
  n2 = LinearInterpolate(
   B=ScalarParameter() {defaultValue=1, group=PatternEmissive, parameterName=PanningEmissiveInfluence_Max}
   Alpha=FunctionInput() {bUsePreviewValueAsDefault=true, inputName=PanningEmissiveBrightnessAdjustment, inputType=FunctionInput_Scalar}
  )
  n3 = ScalarParameter() {group=PatternEmissive, parameterName=InteractWithPanningEmissiveFunction}
  n4 = Multiply(
   A=VectorParameter() {group=PatternEmissive, parameterName=CoreColor, defaultValue=<rgba 0 0 0 1>}.RGB
   B=Add(A=Constant() {r=1}, B=LinearInterpolate(B=@n2, Alpha=@n3))
  )
  n5 = TextureCoordinate()
  n6 = ScalarParameter() {defaultValue=4, group=PatternEmissive, parameterName=EmissivePattern_Tiling}
  n7 = ScalarParameter() {group=PatternEmissive, parameterName=PatternPanning_X}
  n8 = ScalarParameter() {group=PatternEmissive, parameterName=PatternPanning_Y}
  n9 = TextureSampleParameter2D(UVs=Panner(Coordinate=Multiply(A=@n5, B=@n6), Speed=AppendVector(A=@n7, B=@n8))) {
   group=PatternEmissive
   parameterName=PatternEmissiveTexture
   samplerType=SAMPLERTYPE_Masks
   texture=/Game/Characters/Player/Female/Medium/Bodies/F_MED_Neoncat/Textures/Triangles_M
  }
  n10 = BlendMaterialAttributes(
   A=@n1.MaterialAttributes
   B=SetMaterialAttributes!(
    MaterialAttributes=@n1.MaterialAttributes
    Normal=@n1.MaterialAttributes
    BaseColor=Constant()
    EmissiveColor=Add(A=@n1.EmissiveColor, B=@n4)
   ) {attributeSetTypes [#1, #2, #3]}
   Alpha=Multiply(A=@n9.R, B=FunctionInput() {bUsePreviewValueAsDefault=true, inputName=Alpha, inputType=FunctionInput_Scalar})
  )
  FunctionOutput(None=StaticSwitchParameter(True=ShadingPathSwitch(Default=@n10, Deferred=@n10, Mobile=@n1.MaterialAttributes) {desc=Migrated from FeatureLevelSwitch}, False=@n10) {group=PatternEmissive, parameterName=disable Pattern emissive on Mobile})
 ]}
