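# Material function for weapon skins that composites the player's banner icon and background shape onto the incoming material attributes.
# Up to two placements are mirrored, offset, rotated and scaled in UV space, clipped to 0-1, then merged with Max.
# Icon cutout and invert masks drive banner colors, shaded by a Gmap or Diffuse green, plus non-metallic specular and roughness.
# The result blends in through the banner mask, by default unioned with Diffuse alpha.
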MaterialFunction /Game/Weapons/WeaponSkins/MaterialFunction/PlayerBanner
graph {nodes [
  n1 = GetMaterialAttributes(None=FunctionInput() {inputName=Materials, inputType=FunctionInput_MaterialAttributes}) {attributeGetTypes [
    #1
    #2
    #3
    #4
   ]}
  n2 = Multiply(
   A=AppendVector(
    A=ScalarParameter() {defaultValue=1, group=PlayerBanner, parameterName=Banner2_MirrorX, sliderMax=1, sliderMin=-1}
    B=ScalarParameter() {defaultValue=1, group=PlayerBanner, parameterName=Banner2_MirrorY, sliderMax=1, sliderMin=-1}
   )
   B=TextureCoordinate()
  )
  n3 = AppendVector(
   A=ScalarParameter() {defaultValue=0.25, group=PlayerBanner, parameterName=Banner2_X}
   B=ScalarParameter() {group=PlayerBanner, parameterName=Banner2_Y}
  )
  n4 = MaterialFunctionCall(
   UVs=MaterialFunctionCall(UVs=Add(A=@n2, B=@n3), "Rotation Angle (0-1)"=ScalarParameter() {group=PlayerBanner, parameterName=Banner2_Rotation}) {materialFunction=&CustomRotator}.Rotated Values
   Texture Scale=ScalarParameter() {defaultValue=0.15, group=PlayerBanner, parameterName=Banner2_Scale}
  ) {materialFunction=&ScaleUVsByCenter}
  n5 = TextureSampleParameter2D(UVs=@n4.UVs) {
   group=PlayerBanner
   parameterName=Banner_Texture
   samplerSource=SSM_Clamp_WorldGroupSettings
   sortPriority=2
   texture=/Game/UI/Foundation/Textures/Banner/Shapes/banner_shape_standard_Mipmaps
  }
  n6 = ScalarParameter() {defaultValue=1, group=PlayerBanner, parameterName=Banner2_UseSecondBanner}
  n7 = Multiply(
   A=TextureCoordinate()
   B=AppendVector(
    A=ScalarParameter() {defaultValue=1, group=PlayerBanner, parameterName=Banner_MirrorX, sliderMax=1, sliderMin=-1}
    B=ScalarParameter() {defaultValue=1, group=PlayerBanner, parameterName=Banner_MirrorY, sliderMax=1, sliderMin=-1}
   )
  )
  n8 = AppendVector(A=ScalarParameter() {group=PlayerBanner, parameterName=Banner1_X}, B=ScalarParameter() {group=PlayerBanner, parameterName=Banner1_Y})
  n9 = MaterialFunctionCall(
   UVs=MaterialFunctionCall(UVs=Add(A=@n7, B=@n8), "Rotation Angle (0-1)"=ScalarParameter() {group=PlayerBanner, parameterName=Banner1_Rotation}) {materialFunction=&CustomRotator}.Rotated Values
   Texture Scale=ScalarParameter() {defaultValue=0.15, group=PlayerBanner, parameterName=Banner1_Scale}
  ) {materialFunction=&ScaleUVsByCenter}
  n10 = TextureSampleParameter2D(UVs=@n9.UVs) {
   group=PlayerBanner
   parameterName=Banner_Texture
   samplerSource=SSM_Clamp_WorldGroupSettings
   sortPriority=2
   texture=&banner_shape_standard_Mipmaps
  }
  n11 = Max(A=Multiply(A=Multiply(A=@n5.R, B=@n4.0-1 mask), B=@n6), B=Multiply(A=@n10.R, B=@n9.0-1 mask))
  n12 = TextureSampleParameter2D(UVs=LinearInterpolate(A=@n4.UVs, B=TextureCoordinate(), Alpha=ScalarParameter() {group=PlayerBanner, parameterName=Banner1_UseUV0forBackgroundTexture})) {
   group=PlayerBanner
   parameterName=Banner_Background_Texture
   samplerSource=SSM_Clamp_WorldGroupSettings
   sortPriority=3
   texture=&banner_shape_standard_Mipmaps
  }
  n13 = StaticComponentMaskParameter(None=@n12.RGBA) {defaultR=true, group=PlayerBanner, parameterName=Banner1_BackgroundTextureChannel, sortPriority=4}
  n14 = TextureSampleParameter2D(UVs=LinearInterpolate(A=@n9.UVs, B=TextureCoordinate(), Alpha=ScalarParameter() {group=PlayerBanner, parameterName=Banner1_UseUV0forBackgroundTexture})) {
   group=PlayerBanner
   parameterName=Banner_Background_Texture
   samplerSource=SSM_Clamp_WorldGroupSettings
   sortPriority=3
   texture=&banner_shape_standard_Mipmaps
  }
  n15 = Max(
   A=Multiply(A=@n6, B=Multiply(A=@n4.0-1 mask, B=@n13))
   B=Multiply(
    A=@n9.0-1 mask
    B=StaticComponentMaskParameter(None=@n14.RGBA) {defaultR=true, group=PlayerBanner, parameterName=Banner1_BackgroundTextureChannel, sortPriority=4}
   )
  )
  n16 = LinearInterpolate(
   A=Multiply(A=@n11, B=@n15)
   B=Multiply(A=@n15, B=OneMinus(None=@n11))
   Alpha=ScalarParameter() {group=PlayerBanner, parameterName=Banner_z-IconInvertMask}
  )
  n17 = LinearInterpolate(A=@n15, B=@n16, Alpha=ScalarParameter() {defaultValue=1, group=PlayerBanner, parameterName=Banner_z-IconCutout})
  n18 = TextureSampleParameter2D() {group=Diffuse, parameterName=Diffuse, texture=/Engine/EngineMaterials/DefaultDiffuse}
  n19 = FunctionInput() {bUsePreviewValueAsDefault=true, inputName=Gmap, inputType=FunctionInput_Scalar, previewValue=<1 1 1 1>}
  n20 = LinearInterpolate(A=@n18.G, B=@n19, Alpha=ScalarParameter() {defaultValue=1, group=PlayerBanner, parameterName=Banner_ColorGmapOrDiffuse})
  n21 = Saturate(None=LinearInterpolate(
    A=ScalarParameter() {group=PlayerBanner, parameterName=Banner_ColorGmapDark}
    B=ScalarParameter() {defaultValue=1, group=PlayerBanner, parameterName=Banner_ColorGmapLight}
    Alpha=@n20
   ))
  n22 = VectorParameter() {group=PlayerBanner, parameterName=Banner_PrimaryColor, sortPriority=5, defaultValue=<rgba 1 0 0 1>}
  n23 = ScalarParameter() {group=PlayerBanner, parameterName=Banner_z-UseDiffuseAsIconBackground}
  n24 = LinearInterpolate(A=@n22.RGB, B=Multiply(A=@n18.A, B=@n22.RGB), Alpha=@n23)
  n25 = Saturate(None=LinearInterpolate(
    A=ScalarParameter() {group=PlayerBanner, parameterName=Banner_IconGmapDark}
    B=ScalarParameter() {defaultValue=1, group=PlayerBanner, parameterName=Banner_IconGmapLight}
    Alpha=@n20
   ))
  n26 = VectorParameter() {group=PlayerBanner, parameterName=Banner_MaskedColor, sortPriority=5, defaultValue=<rgba 0.65 0.65 0.65 1>}
  n27 = LinearInterpolate(A=@n26.RGB, B=@n22.RGB, Alpha=ScalarParameter() {group=PlayerBanner, parameterName=Banner_z-IconSwapColors})
  n28 = SetMaterialAttributes!(
   MaterialAttributes=@n1.MaterialAttributes
   BaseColor=LinearInterpolate(A=Multiply(A=@n21, B=@n24), B=Multiply(A=@n25, B=@n27), Alpha=LinearInterpolate(A=@n16, B=@n15, Alpha=@n23))
   Metallic=Constant()
   Specular=LinearInterpolate(A=@n1.Specular, B=ScalarParameter() {defaultValue=0.8, group=PlayerBanner, parameterName=Banner_IconSpec}, Alpha=@n17)
   Roughness=LinearInterpolate(A=@n1.Roughness, B=ScalarParameter() {defaultValue=0.4, group=PlayerBanner, parameterName=Banner_IconRoughness}, Alpha=@n17)
  ) {attributeSetTypes [
    #1
    #2
    #3
    #4
   ]}
  n29 = StaticSwitchParameter(True=@n17, False=Max(A=@n17, B=@n18.A)) {desc=by default it will use the alpha of the diffuse map, group=PlayerBanner, parameterName=UseBannerAsBackgroundMask, sortPriority=1}
  FunctionOutput(None=BlendMaterialAttributes(A=@n1.MaterialAttributes, B=@n28, Alpha=@n29))
 ]}
