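# Material function for weapon skins that outputs a Sobel-style edge mask from an input texture.
# It compares a center sample with four diagonally offset samples, optionally collapsing each to a scalar by dot product with DesatColor.
# Absolute differences are summed, scaled and clamped.
# As wired, the fourth offset duplicates the first, leaving one diagonal unsampled, and the negative-offset sample omits the MipBias input.
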MaterialFunction /Game/Weapons/WeaponSkins/MaterialFunction/SobelFilter
graph {
 properties {bExposeToLibrary=true}
 nodes [
  n1 = FunctionInput(Preview=TextureObject() {texture=/Game/Global/Textures/Default/Masks_and_Gradients/16x}) {bUsePreviewValueAsDefault=true, inputName=Texture, inputType=FunctionInput_Texture2D, sortPriority=1}
  n2 = FunctionInput() {bUsePreviewValueAsDefault=true, inputName=MipBias, inputType=FunctionInput_Scalar, previewValue=<-10 0 0 1>}
  n3 = TextureSample!(UVs, Tex=@n1, Bias=@n2, Apply View MipBias) {mipValueMode=TMVM_MipBias, samplerSource=SSM_Wrap_WorldGroupSettings}
  n4 = FunctionInput() {bUsePreviewValueAsDefault=true, inputName=DesatColor, sortPriority=2, previewValue=<1 1 1 1>}
  n5 = FunctionInput(Preview=StaticBool()) {bUsePreviewValueAsDefault=true, inputName=Desaturate?, inputType=FunctionInput_StaticBool, sortPriority=1}
  n6 = StaticSwitch(True=DotProduct(A=@n3.RGB, B=@n4), False=@n3.RGB, Value=@n5)
  n7 = TextureCoordinate()
  n8 = FunctionInput() {bUsePreviewValueAsDefault=true, inputName=OffsetAmount, inputType=FunctionInput_Scalar, previewValue=<0.0005 0 0 1>}
  n9 = TextureSample(UVs=Add(A=@n7, B=@n8), Tex=@n1, Bias=@n2) {mipValueMode=TMVM_MipBias, samplerSource=SSM_Wrap_WorldGroupSettings}
  n10 = Subtract(A=@n6, B=StaticSwitch(True=DotProduct(A=@n9.RGB, B=@n4), False=@n9.RGB, Value=@n5))
  n11 = Multiply(A=@n8) {constB=-1}
  n12 = TextureSample(UVs=Add(A=@n7, B=@n11), Tex=@n1) {mipValueMode=TMVM_MipBias, samplerSource=SSM_Wrap_WorldGroupSettings}
  n13 = Subtract(A=@n6, B=StaticSwitch(True=DotProduct(A=@n12.RGB, B=@n4), False=@n12.RGB, Value=@n5))
  n14 = AppendVector(A=@n11, B=@n8)
  n15 = TextureSample(UVs=Add(A=@n7, B=@n14), Tex=@n1, Bias=@n2) {mipValueMode=TMVM_MipBias, samplerSource=SSM_Wrap_WorldGroupSettings}
  n16 = Subtract(A=@n6, B=StaticSwitch(True=DotProduct(A=@n15.RGB, B=@n4), False=@n15.RGB, Value=@n5))
  n17 = MaterialFunctionCall(Float2=@n14) {materialFunction=&BreakOutFloat2Components}
  n18 = TextureSample(UVs=Add(A=@n7, B=AppendVector(A=@n17.G, B=@n17.G)), Tex=@n1, Bias=@n2) {mipValueMode=TMVM_MipBias, samplerSource=SSM_Wrap_WorldGroupSettings}
  n19 = Subtract(A=@n6, B=StaticSwitch(True=DotProduct(A=@n18.RGB, B=@n4), False=@n18.RGB, Value=@n5))
  n20 = Add(A=Add(A=Abs(None=@n10), B=Abs(None=@n13)), B=Add(A=Abs(None=@n16), B=Abs(None=@n19)))
  n21 = FunctionInput() {
   bUsePreviewValueAsDefault=true
   inputName=Modulate Results Before Clamp
   inputType=FunctionInput_Scalar
   sortPriority=5
   previewValue=<10 0 0 1>
  }
  FunctionOutput(None=Clamp(None=Multiply(A=@n20, B=@n21) {constB=10}))
  Normalize(VectorInput=@n14)
 ]
}
