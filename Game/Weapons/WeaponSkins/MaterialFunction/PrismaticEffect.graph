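# Weapon-skin material function that layers an animated prismatic rainbow glow with sparkles over an input material.
# A panning gradient, warped by a swirl flowmap, adds to base color, while three multiplied screen-aligned noise samples drive its hue shift and clamped emissive.
# Fresnel plus Sobel edges from a normal map set the layer blend, multiplied by the MASK input unless FullPrismaticGlow is on.
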MaterialFunction /Game/Weapons/WeaponSkins/MaterialFunction/PrismaticEffect
graph {nodes [
  Constant()
  Constant() {r=1}
  n1 = MaterialFunctionCall(XY=Multiply(A=TextureCoordinate(), B=ScalarParameter() {defaultValue=1, parameterName=GradientTiling})) {materialFunction=&Swizzle}
  n2 = AppendVector(
   A=LinearInterpolate(
    A=ScalarParameter() {defaultValue=0.1, parameterName=PrismaticSpeed_X}
    B=ScalarParameter() {defaultValue=0.1, parameterName=PrismaticSpeed_X_Pliant_Backpack}
    Alpha=NamedRerouteUsage()
   )
   B=ScalarParameter() {parameterName=PrismaticSpeed_Y}
  )
  n3 = Multiply(A=TextureCoordinate(), B=ScalarParameter() {defaultValue=1, parameterName=GradientWarp_Tiling})
  n4 = ComponentMask(None=TextureSample(UVs=Panner(Coordinate=@n3) {speedY=-0.0125}) {samplerType=SAMPLERTYPE_LinearColor, texture=/Game/Global/Textures/Default/Flowmaps/T_SwirlFlow_01}.RGB) {g=true, r=true}
  n5 = TextureSampleParameter2D(UVs=Add(A=Panner(Coordinate=@n1.YX, Speed=@n2) {speedX=0.1}, B=Multiply(A=@n4) {constB=0.075})) {parameterName=Gradient, texture=/Game/Global/Textures/Default/Masks_and_Gradients/FortniteRainbow}
  n6 = ScalarParameter() {defaultValue=0.5, parameterName=Fresnel Power}
  n7 = ScalarParameter() {parameterName=Sobel mip}
  n8 = TextureObjectParameter() {parameterName=PrismaticSobel_Normal, samplerType=SAMPLERTYPE_Normal, texture=/Game/Weapons/FORT_Shotguns/Textures/T_HammerShotGun_N}
  n9 = MaterialFunctionCall(Float2=TextureProperty(Tex=@n8)) {materialFunction=&BreakOutFloat2Components}
  n10 = Multiply(A=Divide(B=Max(A=@n9.R, B=@n9.R)), B=ScalarParameter() {defaultValue=0.25, parameterName=Sobel offset})
  n11 = StaticBool() {value=true}
  n12 = MaterialFunctionCall!(
   Base=LinearInterpolate(
    A=ScalarParameter() {parameterName=Fresnel Inner}
    Alpha=MaterialFunctionCall(Power=@n6) {materialFunction=&Fresnel_Function}.Result
   )
   Add=Power(
    Base=MaterialFunctionCall!(MipBias=@n7, OffsetAmount=@n10, Texture=@n8, Desaturate?=@n11, DesatColor, Modulate Results Before Clamp) {materialFunction=&/SobelFilter}.Result
    Exp=ScalarParameter() {defaultValue=6, parameterName=Sobel Result Power}
   )
  ) {materialFunction=/Engine/Functions/Engine_MaterialFunctions02/Math/MultiplyAdd}
  Desaturation(None=Multiply(A=@n5.RGB, B=@n12.Result), Fraction=Constant() {r=3})
  n13 = GetMaterialAttributes(None=FunctionInput() {bUsePreviewValueAsDefault=true, inputName=Material, inputType=FunctionInput_MaterialAttributes}) {attributeGetTypes [
    #1
    #2
    #3
    #4
    #5
    #6
   ]}
  n14 = MaterialFunctionCall() {materialFunction=&ComponentPivotLocation}
  n15 = ViewSize()
  n16 = Divide(B=Divide(A=ComponentMask(None=@n15) {r=true}) {constB=2048})
  n17 = MaterialFunctionCall() {materialFunction=/Engine/Functions/Engine_MaterialFunctions02/Utility/FOV}
  n18 = Multiply(
   A=ScalarParameter() {defaultValue=-0.2224, parameterName=Sparkle Scale}
   B=Multiply(
    A=MaterialFunctionCall!(Center Point World Position=@n14.Result) {materialFunction=/Game/Packages/Fortnite_UniqueMaterials_1/Effects/WorldAlignedScreenCoordinates}.Result
    B=Multiply(A=@n16, B=@n17."FOV ")
   )
  )
  n19 = TextureObject() {samplerType=SAMPLERTYPE_LinearColor, texture=/Game/Gadgets/Assets/Slow_Field/Textures/T_Slowfield_SparkleNoise_01}
  n20 = TextureSample(UVs=Multiply(A=@n18) {constB=1.1}, Tex=@n19) {samplerType=SAMPLERTYPE_LinearColor, texture=&T_Slowfield_SparkleNoise_01}
  n21 = TextureSample(UVs=Panner(Coordinate=@n18) {speedX=0.1, speedY=0.1}, Tex=@n19) {samplerType=SAMPLERTYPE_LinearColor, texture=&T_Slowfield_SparkleNoise_01}
  n22 = Panner(Coordinate=Multiply(A=@n18) {constB=0.75}) {speedX=0.1, speedY=0.1}
  n23 = Multiply(
   A=Multiply(
    A=Multiply(A=@n20.R, B=@n21.G)
    B=TextureSample(UVs=@n22, Tex=@n19) {samplerType=SAMPLERTYPE_LinearColor, texture=&T_Slowfield_SparkleNoise_01}.G
   )
   B=ScalarParameter() {defaultValue=25.3628, parameterName=Sparkle Brightness}
  ) {constB=80}
  n24 = Add(A=@n13.BaseColor, B=Multiply(A=@n5.RGB, B=ScalarParameter() {defaultValue=4, parameterName=RainbowIntensity}))
  n25 = Multiply(
   A=MaterialFunctionCall(Hue Shift Percentage=@n23, Texture=@n24) {materialFunction=&HueShift}.Result
   B=ScalarParameter() {defaultValue=0.5, parameterName=BaseColorMult}
  )
  n26 = Constant() {r=0.5}
  n27 = Multiply(A=@n24, B=@n23)
  n28 = Clamp(
   None=Multiply(A=@n27, B=ScalarParameter() {defaultValue=1.5, parameterName=Emissive_Intensity, sortPriority=1})
   Max=ScalarParameter() {defaultValue=25, parameterName=MAX Emissive_Intensity Prismatic, sortPriority=0}
  )
  n29 = Multiply(A=FunctionInput() {bUsePreviewValueAsDefault=true, inputName=MASK, inputType=FunctionInput_Scalar}, B=@n12.Result)
  FunctionOutput(None=MaterialFunctionCall!(
    Base Material=@n13.MaterialAttributes
    Top Material=MakeMaterialAttributes!(
     BaseColor=@n25
     Metallic=@n13.Metallic
     Specular=@n26
     Roughness=@n13.Metallic
     Anisotropy
     EmissiveColor=@n28
     Opacity
     OpacityMask
     Normal=@n13.Metallic
     Tangent
     WorldPositionOffset
     SubsurfaceColor
     ClearCoat
     ClearCoatRoughness
     AmbientOcclusion
     Refraction
     CustomizedUVs_0
     CustomizedUVs_1
     CustomizedUVs_2
     CustomizedUVs_3
     CustomizedUVs_4
     CustomizedUVs_5
     CustomizedUVs_6
     CustomizedUVs_7
     PixelDepthOffset
     ShadingModel
     Displacement
    )
    Alpha=StaticSwitchParameter(True=@n12.Result, False=@n29) {parameterName=FullPrismaticGlow}
   ) {materialFunction=/Engine/Functions/MaterialLayerFunctions/MatLayerBlend_Standard}.Blended Material)
  MaterialFunctionCall(Texture=@n27) {materialFunction=&HueShift}
  NamedRerouteDeclaration(None=FunctionInput() {bUsePreviewValueAsDefault=true, inputName=IsCharacter, inputType=FunctionInput_Scalar}) {name=IsCharacter, nodeColor=<rgba 0 0.9253 1 1>}
  n30 = FunctionInput() {bUsePreviewValueAsDefault=true, inputName=IsStatsB, inputType=FunctionInput_Vector2}
  NamedRerouteDeclaration(None=ComponentMask(None=@n30) {r=true}) {name=IsBackpack, nodeColor=<rgba 0 0.9253 1 1>}
  NamedRerouteDeclaration(None=ComponentMask(None=@n30) {g=true}) {name=IsPickaxe, nodeColor=<rgba 0 0.9253 1 1>}
 ]}
