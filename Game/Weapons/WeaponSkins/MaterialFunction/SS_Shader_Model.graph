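# Weapon skin material function that layers a Subsurface shading model, with tinted subsurface color and opacity, over incoming material attributes.
# Subsurface color optionally multiplies in base color and is scaled by constant or Fresnel-driven strength, which also sets opacity above a minimum.
# A static "Dual SS" (Vitiligo) switch adds a second color and strength set in a secondary mask area.
# Mobile uses a simpler fallback, and masks blend the result with the original attributes.
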MaterialFunction /Game/Weapons/WeaponSkins/MaterialFunction/SS_Shader_Model
graph {nodes [
  n1 = FunctionInput() {inputType=FunctionInput_MaterialAttributes}
  n2 = GetMaterialAttributes(None=@n1)
  n3 = GetMaterialAttributes(None=@n1) {attributeGetTypes [
    #1
    #2
    #3
    #4
   ]}
  n4 = VectorParameter() {group=SS Shader Model, parameterName=SS Color, defaultValue=<rgba 1 1 1 0>}
  n5 = FunctionInput() {bUsePreviewValueAsDefault=true, inputName=Dual SS Secondary Alpha, inputType=FunctionInput_Scalar, sortPriority=1}
  n6 = StaticBoolParameter() {
   desc=Allows for 2 SS Colors and expands frenel options to simulate SSP. Uses Dual SS Secondary Mask to determine secondary area.
   group=SS Shader Model
   parameterName="Use Dual SS Effect (Vitiligo)"
  }
  n7 = StaticSwitch(
   True=LinearInterpolate(A=@n4.RGB, B=VectorParameter() {group=SS Shader Model, parameterName=SS Color 2, defaultValue=<rgba 1 1 1 0>}.RGB, Alpha=@n5)
   False=@n4.RGB
   Value=@n6
  )
  n8 = Multiply(A=@n3.BaseColor, B=@n7)
  n9 = ScalarParameter() {defaultValue=4, group=SS Shader Model, parameterName=SS Fresnel Exponent}
  n10 = LinearInterpolate(
   A=Fresnel(ExponentIn=@n9)
   B=Fresnel(ExponentIn=@n9, Normal=Transform(None=@n3.Normal))
   Alpha=ScalarParameter() {group=SS Shader Model, parameterName=UseObjectNormal}
  )
  n11 = LinearInterpolate(
   A=ScalarParameter() {group=SS Shader Model, parameterName=SS Inner Strength}
   B=ScalarParameter() {defaultValue=0.85, group=SS Shader Model, parameterName=SS Outer Strength}
   Alpha=@n10
  )
  n12 = LinearInterpolate(
   A=ScalarParameter() {group=SS Shader Model, parameterName=SS Inner Strength 2}
   B=VectorParameter() {group=SS Shader Model, parameterName=SS Outer Strength 2, defaultValue=<rgba 0.85 0.85 0.85 0>}.RGB
   Alpha=@n10
  )
  n13 = StaticSwitchParameter(
   True=StaticSwitch(True=LinearInterpolate(A=@n11, B=@n12, Alpha=@n5), False=@n11, Value=@n6)
   False=ScalarParameter() {group=SS Shader Model, parameterName=SS Opacity}
  ) {group=SS Shader Model, parameterName=Use Fresnel for SS Opacity}
  n14 = SetMaterialAttributes!(
   MaterialAttributes=@n3.MaterialAttributes
   SubsurfaceColor=Multiply(A=StaticSwitchParameter(True=@n8, False=@n7) {group=SS Shader Model, parameterName=Use Diffuse in SS Color}, B=@n13)
   Opacity=Max(A=@n13, B=ScalarParameter() {defaultValue=0.05, group=SS Shader Model, parameterName=SS_Opacity_Min})
   ShadingModel=ShadingModel() {shadingModel=MSM_Subsurface}
  ) {attributeSetTypes [#1, #2, #5]}
  n15 = SetMaterialAttributes!(MaterialAttributes=@n2.MaterialAttributes, SubsurfaceColor=@n3.BaseColor, Opacity=Constant() {r=0.5}) {attributeSetTypes [#1, #2]}
  n16 = FunctionInput() {bUsePreviewValueAsDefault=true, inputName=Alpha, inputType=FunctionInput_Scalar, sortPriority=1}
  n17 = Saturate(None=Add(A=@n5, B=@n16))
  n18 = StaticSwitch(True=@n3.Opacity, False=@n16, Value=StaticBoolParameter() {group=SS Shader Model, parameterName=Use Skin FX Mask instead of _M Mask for SSS})
  n19 = StaticBoolParameter() {
   desc=Allows for 2 SS Colors and expands frenel options to simulate SSP. Uses Dual SS Secondary Mask to determine secondary area.
   group=SS Shader Model
   parameterName="Use Dual SS Effect (Vitiligo)"
  }
  FunctionOutput(None=BlendMaterialAttributes(
    A=@n2.MaterialAttributes
    B=ShadingPathSwitch(Default=@n14, Deferred=@n14, Mobile=@n15) {desc=Migrated from FeatureLevelSwitch}
    Alpha=StaticSwitch(True=@n17, False=@n18, Value=@n19)
   ))
 ]}
