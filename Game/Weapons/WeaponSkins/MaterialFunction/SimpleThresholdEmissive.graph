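# Material function for weapon skins that blends an emissive, black-base-color layer into incoming material attributes using a thresholded Sequence input and Alpha mask.
# A SphereMask radius follows either event spectrum times amplitude (or Emote_AudioLow) or a timed LinearSine, via static switches.
# Near-silent audio zeroes the blend even in sine mode.
# "Audio Based Hue Shift" actually lerps two colors by time; the HueShift branch is unconnected.
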MaterialFunction /Game/Weapons/WeaponSkins/MaterialFunction/SimpleThresholdEmissive
graph {nodes [
  n1 = ScalarParameter() {defaultValue=0.2, group=ThresholdEmissive, parameterName=Threshold Radius}
  n2 = Add(A=@n1) {constB=0.1}
  Add(A=Constant() {r=0.1}, B=@n2)
  Add(A=@n2) {constA=1, constB=0.9}
  n3 = FunctionInput() {inputName=Material, inputType=FunctionInput_MaterialAttributes}
  n4 = GetMaterialAttributes(None=@n3) {attributeGetTypes [#1]}
  n5 = VectorParameter() {group=ThresholdEmissive, parameterName=ThresholdEmissive_Color, defaultValue=<rgba 1 1 1 1>}
  n6 = MaterialFunctionCall(
   Value=Multiply(A=Time(), B=ScalarParameter() {defaultValue=1, group=ThresholdEmissive, parameterName=Time Scale})
   Period=ScalarParameter() {defaultValue=1, group=ThresholdEmissive, parameterName=LInear Sine Period}
  ) {materialFunction=&LinearSine}
  n7 = StaticSwitchParameter(
   True=LinearInterpolate(
    A=@n5.RGB
    B=VectorParameter() {group=ThresholdEmissive, parameterName=ThresholdEmissive_AudioTargetColor, defaultValue=<rgba 1 1 1 1>}.RGB
    Alpha=Saturate(None=@n6.Linear Sine)
   )
   False=@n5.RGB
  ) {group=ThresholdEmissive, parameterName=Audio Based Hue Shift}
  n8 = LinearInterpolate(A=@n7, B=Add(A=@n4.EmissiveColor, B=@n7), Alpha=ScalarParameter() {group=ThresholdEmissive, parameterName=Add to Emissive})
  n9 = Constant()
  n10 = CollectionParameter() {collection=&FortniteMaterialParameters, parameterName=EventAudio_Spectrum}
  n11 = MaterialFunctionCall(Float4=@n10) {materialFunction=&BreakOutFloat4Components}
  n12 = Divide(A=Add(A=Add(A=@n11.R, B=@n11.R), B=Add(A=@n11.B, B=@n11.B))) {constB=4}
  n13 = CollectionParameter() {collection=&FortniteMaterialParameters, parameterName=EventAudio_Amplitude}
  n14 = CollectionParameter() {collection=&FortniteMaterialParameters, parameterName=Emote_AudioLow}
  n15 = StaticSwitch(
   True=Power(Base=Multiply(A=@n12, B=@n13), Exp=ScalarParameter() {defaultValue=1, group=ThresholdEmissive, parameterName=FFT Exponent, sortPriority=105})
   False=@n14
   Value=StaticBoolParameter() {group=ThresholdEmissive, parameterName=Use New FFT, sortPriority=100}
  )
  n16 = Saturate(None=@n15)
  n17 = Saturate(None=LinearInterpolate(
    A=ScalarParameter() {defaultValue=1, group=ThresholdEmissive, parameterName=Lerp_Min}
    B=ScalarParameter() {defaultValue=-0.25, group=ThresholdEmissive, parameterName=Lerp_Max}
    Alpha=@n16
   ) {constA=1, constB=-0.25})
  n18 = LinearInterpolate(A=@n1, Alpha=@n6.Linear Sine)
  n19 = Multiply(
   A=SphereMask(
    A=FunctionInput() {inputName=Sequence, inputType=FunctionInput_Scalar}
    B=Constant()
    Radius=StaticSwitchParameter(True=@n17, False=@n18) {group=ThresholdEmissive, parameterName=Sequence_Audio Data}
    Hardness=Constant() {r=1}
   )
   B=FunctionInput() {inputName=Alpha, inputType=FunctionInput_Scalar}
  )
  n20 = If(A=@n15, B=Constant() {r=0.001}, "A > B"=Constant() {r=1}, "A < B"=Constant())
  FunctionOutput(None=BlendMaterialAttributes(
    A=@n3
    B=SetMaterialAttributes!(MaterialAttributes=@n4.MaterialAttributes, EmissiveColor=@n8, BaseColor=@n9) {attributeSetTypes [#1, #2]}
    Alpha=Multiply(A=@n19, B=@n20)
   ))
  LinearInterpolate(Alpha=@n16)
  n21 = CollectionParameter() {collection=&FortniteMaterialParameters, parameterName=Emote_EnvTime}
  n22 = ScalarParameter() {defaultValue=11.28, parameterName=Threshold Hue Divisor}
  MaterialFunctionCall(Hue Shift Percentage=Frac(None=Divide(A=@n21, B=@n22) {constB=88.45}), Texture=@n5.RGB) {materialFunction=&HueShift}
  Multiply()
  Power(Base=@n15)
 ]}
