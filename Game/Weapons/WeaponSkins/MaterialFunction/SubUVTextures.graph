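# Material function that overlays a sub-UV flipbook, such as an animated face or screen, onto incoming material attributes.
# UVs get channel choice, offset, bump offset and centered scaling before sampling CustomNumbers_SubUV.
# Frames come from looping time, a curve-atlas blink, a frame parameter, or character-specific facial pose logic behind static switches (RoboKitty, Banana, WavyMan).
# The tinted result drives emissive or base color, adjusts surface attributes by flipbook alpha, and blends by the Alpha input.
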MaterialFunction /Game/Weapons/WeaponSkins/MaterialFunction/SubUVTextures
graph {nodes [
  n1 = FunctionInput() {inputName=Material, inputType=FunctionInput_MaterialAttributes}
  n2 = GetMaterialAttributes(None=@n1) {attributeGetTypes [
    #1
    #2
    #3
    #4
    #5
    #6
    #7
    #8
    #9
    #10
   ]}
  n3 = Add(
   A=LinearInterpolate(
    A=TextureCoordinate()
    B=TextureCoordinate() {coordinateIndex=1}
    Alpha=ScalarParameter() {group=Flipbook, parameterName=use Second UV Channel}
   )
   B=AppendVector(
    A=ScalarParameter() {defaultValue=0.125, group=Flipbook, parameterName=Flipbook X}
    B=ScalarParameter() {group=Flipbook, parameterName=Flipbook Y}
   )
  )
  n4 = ScalarParameter() {defaultValue=-0.25, group=Flipbook, parameterName=Bump Height}
  n5 = ScalarParameter() {defaultValue=0.5, group=Flipbook, parameterName=Flipbook scale}
  n6 = ScalarParameter() {defaultValue=1, group=Flipbook, parameterName=Flipbook scale V relative}
  n7 = MaterialFunctionCall(
   UVs=LinearInterpolate(
    A=@n3
    B=BumpOffset(Coordinate=@n3, Height=@n4)
    Alpha=ScalarParameter() {defaultValue=1, group=Flipbook, parameterName=BumpOffset Intensity}
   )
   Texture Scale=AppendVector(A=@n5, B=Multiply(A=@n5, B=@n6))
  ) {materialFunction=&ScaleUVsByCenter}
  n8 = AppendVector(
   A=ScalarParameter() {defaultValue=8, group=Flipbook, parameterName=SubImages X}
   B=ScalarParameter() {defaultValue=8, group=Flipbook, parameterName=SubImages Y}
  )
  n9 = Time()
  n10 = ScalarParameter() {defaultValue=0.01, parameterName=BlinkingRate}
  n11 = CurveAtlasRowParameter(CurveTime=Frac(None=Multiply(A=@n9, B=@n10))) {
   atlas=&AthenaCurveAtlas01
   curve=/Game/Characters/Player/Female/Medium/Bodies/F_MED_Angler/Textures/C_BlinkingRate
   defaultValue=236
   parameterName=CurveBlink
  }
  n12 = Round(None=@n11.G)
  n13 = ScalarParameter() {group=Banana, parameterName=Jaw_open_Pose}
  n14 = ScalarParameter() {defaultValue=0.1, group=Banana, parameterName=THreshold}
  n15 = ScalarParameter() {group=Banana, parameterName=Default Open Mouth}
  n16 = Max(
   A=If(
    A=ScalarParameter() {group=Banana, parameterName=R_Frown_pose}
    B=@n14
    "A > B"=ScalarParameter() {defaultValue=10, parameterName=Mouth Open Frown Frame}
    "A == B"=@n15
    "A < B"=@n15
   )
   B=If(
    A=ScalarParameter() {group=Banana, parameterName=R_smile_pose}
    B=@n14
    "A > B"=ScalarParameter() {defaultValue=6, parameterName=Mouth Open Smile Frame}
    "A == B"=@n15
    "A < B"=@n15
   )
  )
  n17 = ScalarParameter() {defaultValue=1, group=Banana, parameterName=Default Face}
  n18 = If(
   A=Max(A=ScalarParameter() {group=Banana, parameterName=R_lip_corner_narrow_pose}, B=ScalarParameter() {parameterName=PeelyBlenderFace})
   B=@n14
   "A > B"=ScalarParameter() {defaultValue=3, parameterName=Narrow Frame}
   "A == B"=@n17
   "A < B"=@n17
  )
  n19 = If(
   A=ScalarParameter() {group=Banana, parameterName=R_frown_pose}
   B=@n14
   "A > B"=ScalarParameter() {defaultValue=4, parameterName=Frown Frame}
   "A == B"=@n18
   "A < B"=@n18
  )
  n20 = If(
   A=ScalarParameter() {group=Banana, parameterName=R_smile_pose}
   B=@n14
   "A > B"=ScalarParameter() {defaultValue=8, parameterName=Smile Mouth animation Frame}
   "A == B"=@n19
   "A < B"=@n19
  )
  n21 = ScalarParameter() {group=Flipbook, parameterName="L_frown_pose "}
  n22 = Constant() {r=0.3}
  n23 = Constant() {r=1}
  n24 = Constant()
  n25 = ScalarParameter() {group=Flipbook, parameterName=jaw_open_pose}
  n26 = Constant() {r=0.3}
  n27 = Constant() {r=2}
  n28 = Constant()
  n29 = Constant()
  n30 = Max(
   A=Max(A=If(A=@n21, B=@n22, "A > B"=@n23, "A == B"=@n24, "A < B"=@n24), B=If(A=@n25, B=@n26, "A > B"=@n27, "A == B"=@n28, "A < B"=@n28))
   B=If(A=ScalarParameter() {group=Flipbook, parameterName=L_wide_pose}, B=Constant() {r=0.3}, "A > B"=Constant() {r=3}, "A == B"=@n29, "A < B"=@n29)
  )
  n31 = ScalarParameter() {group=Flipbook, parameterName=FrameNumber}
  n32 = StaticSwitchParameter(
   True=MaterialFunctionCall() {materialFunction=/Game/Characters/Pets/Robo_kitty/Materials/RoboKittyFaceMath}.Result
   False=StaticSwitchParameter(
    True=If(A=@n13, B=@n14, "A > B"=@n16, "A == B"=@n20, "A < B"=@n20)
    False=StaticSwitchParameter(True=@n30, False=@n31) {group=Flipbook, parameterName=isWavyMan}
   ) {group=Flipbook, parameterName=isBanana}
  ) {group=Flipbook, parameterName=isRoboKitty}
  n33 = Time()
  n34 = ScalarParameter() {defaultValue=0.05, group=Flipbook, parameterName=Animation Speed}
  n35 = LinearInterpolate(B=Multiply(A=@n8, B=@n8), Alpha=Frac(None=Multiply(A=@n33, B=@n34))) {constB=15}
  n36 = MaterialFunctionCall!(
   Texture=TextureObjectParameter() {group=Flipbook, parameterName=Flipbook, texture=/Game/Accessories/FORT_Backpacks/M_LRG_Bob_Backpack/Textures/T_8BitGhoul}
   UVs=@n7.UVs
   SubImages=LinearInterpolate(
    A=ScalarParameter() {defaultValue=8, group=Flipbook, parameterName=SubImages}
    B=@n8
    Alpha=ScalarParameter() {group=Flipbook, parameterName=useNonSquareTexture}
   )
   Frame=LinearInterpolate(
    A=StaticSwitchParameter(True=@n12, False=@n32) {group=Flipbook, parameterName=isSimpleBlink}
    B=Floor(None=@n35)
    Alpha=ScalarParameter() {defaultValue=1, group=Flipbook, parameterName=isAnimation}
   )
  ) {materialFunction=&/CustomNumbers_SubUV}
  n37 = Multiply(
   A=Multiply(A=ScalarParameter() {defaultValue=1, group=Flipbook, parameterName=Scren Brightness}, B=Multiply(A=@n36.RGB, B=@n7.0-1 mask))
   B=VectorParameter() {group=Flipbook, parameterName=FlipbookTint, defaultValue=<rgba 1 1 1 1>}.RGB
  )
  n38 = ScalarParameter() {group=Flipbook, parameterName=Affects Base Color}
  n39 = LinearInterpolate(
   A=LinearInterpolate(A=@n37, B=Constant(), Alpha=@n38)
   Alpha=Fresnel(ExponentIn=ScalarParameter() {defaultValue=2, group=Flipbook, parameterName=Flipbook Fresnel exponent}) {baseReflectFraction=0}
  ) {constB=0}
  n40 = ScalarParameter() {defaultValue=1, desc=1 is passthrough 0 is full roughness, group=Flipbook, parameterName=Multiply Flipbook Roughness}
  n41 = OneMinus(None=@n36.Alpha)
  n42 = OneMinus(None=Saturate(None=Add(A=@n40, B=@n41)))
  n43 = Add(
   A=ScalarParameter() {defaultValue=1, desc=1 is passthrough 0 is no specular, group=Flipbook, parameterName=Multiply Flipbook Specular}
   B=@n41
  )
  n44 = Add(
   A=ScalarParameter() {defaultValue=1, desc=1 is passthrough 0 is no metallic, group=Flipbook, parameterName=Multiply Flipbook Metallic}
   B=@n41
  )
  n45 = ScalarParameter() {group=Flipbook, parameterName=useAlphaAsMask}
  n46 = Add(
   A=ScalarParameter() {defaultValue=1, desc=1 is passthrough 0 is no subsurface, group=Flipbook, parameterName=Multiply Flipbook Subsurface}
   B=@n41
  )
  n47 = SetMaterialAttributes!(
   MaterialAttributes=@n2.MaterialAttributes
   EmissiveColor=LinearInterpolate(
    A=Add(A=@n2.EmissiveColor, B=@n39)
    B=Multiply(A=@n2.EmissiveColor, B=@n39)
    Alpha=ScalarParameter() {group=Flipbook, parameterName=Multiply Flipbook Emissive}
   )
   BaseColor=LinearInterpolate(A=@n2.BaseColor, B=@n37, Alpha=@n38)
   Roughness=Saturate(None=Add(A=@n2.Roughness, B=@n42))
   SubsurfaceColor=Multiply(A=@n2.SubsurfaceColor, B=Saturate(None=@n46))
   Opacity=@n2.MaterialAttributes
   CustomData0=@n2.MaterialAttributes
   CustomData1=@n2.MaterialAttributes
   Specular=Multiply(A=@n2.Specular, B=Saturate(None=@n43))
   Metallic=Multiply(A=@n2.Metallic, B=Saturate(None=@n44))
   OpacityMask=LinearInterpolate(A=@n2.OpacityMask, B=@n36.Alpha, Alpha=@n45)
  ) {attributeSetTypes [
    #1
    #7
    #2
    #6
    #5
    #3
    #4
    #8
    #9
    #10
   ]}
  n48 = FunctionInput() {inputName=Alpha, inputType=FunctionInput_Scalar}
  n49 = LinearInterpolate(A=@n48, B=Multiply(A=@n48, B=Multiply(A=@n36.Alpha, B=@n7.0-1 mask)), Alpha=@n38)
  FunctionOutput(None=BlendMaterialAttributes(A=@n1, B=@n47, Alpha=LinearInterpolate(A=@n49, B=@n48, Alpha=@n45)) {vertexAttributeBlendType=UseA})
  Max(A=Max(A=@n18, B=@n19), B=@n20)
  Reroute(None=@n49)
  StaticSwitch()
 ]}
