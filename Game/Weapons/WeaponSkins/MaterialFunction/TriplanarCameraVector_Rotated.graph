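# Weapon-skin material function that maps one 2D texture onto the camera vector triplanar-style, giving a cubemap-like background with rotation.
# The vector, scaled by Tiling and shifted by Offset, feeds XZ, YZ and XY samples through Rotators driven by Rotation Speed's G, B and R channels respectively.
# SCurve weights from the unscaled vector's absolute X and Z components blend the three samples, with AxisFadeContrast controlling seam fade.
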MaterialFunction /Game/Weapons/WeaponSkins/MaterialFunction/TriplanarCameraVector_Rotated
graph {
 properties {
  bExposeToLibrary=true
  description="Applies a texture to the background, similar to a cubemap but uses the same 2d texture for all sides"
  libraryCategoriesText [Misc,Texturing]
 }
 nodes [
  n1 = FunctionInput(Preview=CameraVectorWS()) {bUsePreviewValueAsDefault=true, desc=Allows you to distort camera vector for special effects, inputName=CameraVector}
  n2 = Add(
   A=FunctionInput() {bUsePreviewValueAsDefault=true, inputName=Offset, sortPriority=2}
   B=Multiply(A=@n1, B=FunctionInput() {bUsePreviewValueAsDefault=true, inputName=Tiling, sortPriority=1, previewValue=<1 1 1 1>})
  )
  n3 = FunctionInput() {inputName=Rotation Speed}
  n4 = FunctionInput(Preview=TextureObject() {texture=/Engine/MapTemplates/TilingNoise05}) {inputName=Texture, inputType=FunctionInput_Texture2D}
  n5 = TextureSample(UVs=Rotator(Coordinate=ComponentMask(None=@n2) {b=true, r=true}, Time=ComponentMask(None=@n3) {g=true}), Tex=@n4) {samplerSource=SSM_Wrap_WorldGroupSettings, texture=&TilingNoise05}
  n6 = TextureSample(UVs=Rotator(Coordinate=ComponentMask(None=@n2) {b=true, g=true}, Time=ComponentMask(None=@n3) {b=true}), Tex=@n4) {samplerSource=SSM_Wrap_WorldGroupSettings, texture=&TilingNoise05}
  n7 = Constant3Vector() {constant=<rgba 1 0 0 0>}
  n8 = FunctionInput() {
   bUsePreviewValueAsDefault=true
   description=Used to fade the seams between the 6 textures
   inputName=AxisFadeContrast
   inputType=FunctionInput_Scalar
   sortPriority=3
   previewValue=<16 0 0 1>
  }
  n9 = MaterialFunctionCall(In=Abs(None=DotProduct(A=@n1, B=@n7)), Power=@n8) {materialFunction=&SCurve}
  n10 = Rotator(Coordinate=ComponentMask(None=@n2) {g=true, r=true}, Time=ComponentMask(None=@n3) {r=true})
  n11 = Abs(None=DotProduct(A=@n1, B=Constant3Vector() {constant=<rgba 0 0 1 0>}))
  FunctionOutput(None=LinearInterpolate(
    A=LinearInterpolate(A=@n5.RGB, B=@n6.RGB, Alpha=@n9.Result)
    B=TextureSample(UVs=@n10, Tex=@n4) {samplerSource=SSM_Wrap_WorldGroupSettings, texture=&TilingNoise05}.RGB
    Alpha=MaterialFunctionCall(In=@n11, Power=@n8) {materialFunction=&SCurve}.Result
   ))
 ]
}
