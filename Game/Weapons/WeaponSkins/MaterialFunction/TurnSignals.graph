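# Material function for vehicle turn-signal and brake lights, blending a glowing taillight over incoming material attributes.
# Rounded local-bounds position on a switchable axis, or a CenterLine input, splits the sides; isTurningLeft selects one, gated by isTurning and an optional time-based blink.
# isBraking lights both sides and shifts taillight color, which is shaped by inverse Fresnel and optionally modulated by base color.
# The saturated turn-plus-brake mask times input Alpha drives the blend.
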MaterialFunction /Game/Weapons/WeaponSkins/MaterialFunction/TurnSignals
graph {nodes [
  n1 = GetMaterialAttributes(None=FunctionInput() {inputType=FunctionInput_MaterialAttributes}) {attributeGetTypes [#1]}
  n2 = Time()
  n3 = ScalarParameter() {defaultValue=1, group=TurnSignal, parameterName=BlinkRate}
  n4 = Round(None=Frac(None=Multiply(A=@n2, B=@n3)))
  n5 = Constant() {r=1}
  n6 = TransformPosition(None=WorldPosition() {worldPositionShaderOffset=WPT_ExcludeAllShaderOffsets}.XYZ) {transformSourceType=TRANSFORMPOSSOURCE_World}
  n7 = MaterialFunctionCall() {materialFunction=&ObjectLocalBounds}
  n8 = ScalarParameter() {desc=3. Tweak the CenterLineOffset which is stable in Game, group=TurnSignal, parameterName=CenterLineOffset}
  n9 = StaticComponentMaskParameter(None=Divide(
    A=Subtract(A=@n6, B=@n7.Local Bounds Minimum)
    B=Add(A=@n7.Local Bounds Size, B=@n8) {desc=3. Tweak the CenterLineOffset which is stable in Game}
   ) {constB=1000}) {defaultR=true, desc=4.Add Channel Switcher for changing the aixs, group=TurnSignal, parameterName=ChannelSwitch}
  n10 = ScalarParameter() {defaultValue=0.18, group=TurnSignal, parameterName=CenterLineOFfsetStablelize}
  n11 = Round(None=FunctionInput(Preview=Add(A=@n9, B=@n10) {desc=5.Skip it for center line stable work}) {bUsePreviewValueAsDefault=true, inputName=CenterLine, inputType=FunctionInput_Scalar})
  n12 = OneMinus(None=@n11)
  n13 = ScalarParameter() {group=TurnSignal, parameterName=isTurningLeft}
  n14 = Multiply(
   A=Multiply(
    A=StaticSwitchParameter(True=@n4, False=@n5) {defaultValue=true, desc=Default is using Materail Blinker, group=TurnSignal, parameterName=UseMaterialBlinker}
    B=LinearInterpolate(A=@n11, B=@n12, Alpha=@n13)
   )
   B=ScalarParameter() {group=TurnSignal, parameterName=isTurning}
  )
  n15 = ScalarParameter() {group=TurnSignal, parameterName=isBraking}
  n16 = Constant() {r=4}
  n17 = Multiply(
   A=LinearInterpolate(
    A=VectorParameter() {group=TurnSignal, parameterName=Taillight Color, defaultValue=<rgba 10 0.4 0.4 1>}.RGB
    B=VectorParameter() {group=TurnSignal, parameterName=TaillightBrakingColor, defaultValue=<rgba 10 0.4 0.4 1>}.RGB
    Alpha=@n15
   )
   B=OneMinus(None=Fresnel(ExponentIn=@n16) {baseReflectFraction=0})
  )
  n18 = StaticSwitchParameter(True=Multiply(A=@n1.BaseColor, B=@n17), False=@n17) {desc=Default is using Materail Blinker, group=TurnSignal, parameterName=ModulateEmissiveAgainstBaseColor}
  n19 = Saturate(None=Add(A=@n14, B=@n15))
  n20 = FunctionInput() {bUsePreviewValueAsDefault=true, inputName=Alpha, inputType=FunctionInput_Scalar}
  FunctionOutput(None=BlendMaterialAttributes(
    A=@n1.MaterialAttributes
    B=SetMaterialAttributes!(MaterialAttributes=@n1.MaterialAttributes, BaseColor=@n14, EmissiveColor=@n18) {attributeSetTypes [#1, #2]}
    Alpha=Multiply(A=@n19, B=@n20)
   ) {vertexAttributeBlendType=UseA})
  MaterialFunctionCall() {materialFunction=&BoundingBoxBased_0-1_UVW}
 ]}
