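# Weapon skin material function that overlays a continuously rotating Diffuse texture onto the BaseColor of incoming material attributes.
# UV1 is offset, center-scaled and spun by CustomRotator at Time times a speed parameter to sample the texture.
# A clamped, multiplied UV1 U gradient masks the lerp.
# Most other rebuilt attributes read the input's Metallic pin, and the Alpha blend and clown-mask samples do not reach the output.
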MaterialFunction /Game/Weapons/WeaponSkins/MaterialFunction/UV_Rotator
graph {nodes [
  n1 = FunctionInput() {inputName=Material, inputType=FunctionInput_MaterialAttributes}
  BlendMaterialAttributes(A=@n1, Alpha=FunctionInput() {bUsePreviewValueAsDefault=true, inputName=Alpha, inputType=FunctionInput_Scalar})
  n2 = BreakMaterialAttributes!(Attr=@n1)
  n3 = AppendVector(
   A=ScalarParameter() {defaultValue=1.1968, group=Rotator, parameterName=Rotator X}
   B=ScalarParameter() {defaultValue=-0.381, group=Rotator, parameterName=Rotator Y}
  )
  n4 = TextureCoordinate() {coordinateIndex=1}
  n5 = MaterialFunctionCall(
   UVs=MaterialFunctionCall(UVs=Add(A=@n3, B=@n4), Texture Scale=ScalarParameter() {defaultValue=5.7313, group=Custom Numbers, parameterName=Number Scale_1}) {materialFunction=&ScaleUVsByCenter}.UVs
   Rotation Center=AppendVector(
    A=ScalarParameter() {defaultValue=0.71, group=Rotator, parameterName=Rotator X_1}
    B=ScalarParameter() {defaultValue=0.44, group=Rotator, parameterName=Rotator Y_1}
   )
   "Rotation Angle (0-1)"=Multiply(A=Time(), B=ScalarParameter() {defaultValue=1.019, group=Rotator, parameterName=Rotator Speed})
  ) {materialFunction=&CustomRotator}
  n6 = TextureSampleParameter2D(UVs=@n5.Rotated Values) {group=Diffuse, parameterName=Diffuse, texture=/Engine/EngineMaterials/DefaultDiffuse}
  n7 = Clamp(None=Multiply(A=ScalarParameter() {defaultValue=14.1186, parameterName=Rotate Mask Multiply}, B=ComponentMask(None=@n4) {desc="1 / ## Horz Images", r=true}))
  FunctionOutput(None=MakeMaterialAttributes!(
    BaseColor=LinearInterpolate(A=@n2.BaseColor, B=@n6.RGB, Alpha=@n7)
    Metallic=@n2.Metallic
    Specular=@n2.Metallic
    Roughness=@n2.Metallic
    Anisotropy
    EmissiveColor=@n2.Metallic
    Opacity=@n2.Metallic
    OpacityMask=@n2.Metallic
    Normal=@n2.Metallic
    Tangent
    WorldPositionOffset=@n2.Metallic
    SubsurfaceColor=@n2.Metallic
    ClearCoat=@n2.Metallic
    ClearCoatRoughness=@n2.Metallic
    AmbientOcclusion=@n2.Metallic
    Refraction=@n2.Metallic
    CustomizedUVs_0=@n2.Metallic
    CustomizedUVs_1=@n2.Metallic
    CustomizedUVs_2=@n2.Metallic
    CustomizedUVs_3=@n2.Metallic
    CustomizedUVs_4=@n2.Metallic
    CustomizedUVs_5=@n2.Metallic
    CustomizedUVs_6=@n2.Metallic
    CustomizedUVs_7=@n2.Metallic
    PixelDepthOffset=@n2.Metallic
    ShadingModel
    Displacement
   ))
  TextureSample(UVs=@n5.Rotated Values) {texture=/Game/Accessories/Hats/M_MED_ClownMask_01/Textures/T_M_Clown_Mask01_D}
  TextureSample() {texture=&T_M_Clown_Mask01_D}
 ]}
