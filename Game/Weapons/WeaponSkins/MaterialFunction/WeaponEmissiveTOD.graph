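# Material function for weapon skins that optionally boosts emissive brightness by time of day.
# It takes a material attributes input, pulls out Emissive Color and runs it through the BoostEmissivebyTOD function.
# A static switch parameter, Use TOD Emissive Boost in the Skin FX group, picks boosted or original emissive before it is written back into the attributes.
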MaterialFunction /Game/Weapons/WeaponSkins/MaterialFunction/WeaponEmissiveTOD
graph {nodes [
  n1 = GetMaterialAttributes(None=FunctionInput() {inputType=FunctionInput_MaterialAttributes}) {attributeGetTypes [#1]}
  n2 = MaterialFunctionCall!(In=@n1.EmissiveColor) {materialFunction=&/BoostEmissivebyTOD}
  FunctionOutput(None=SetMaterialAttributes!(
    MaterialAttributes=@n1.MaterialAttributes
    EmissiveColor=StaticSwitchParameter(True=@n2.Result, False=@n1.EmissiveColor) {group=Skin FX, parameterName=Use TOD Emissive Boost}
   ) {attributeSetTypes [#1]})
 ]}
