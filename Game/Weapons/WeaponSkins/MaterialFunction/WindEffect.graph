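# Weapon-skin material function that adds animated wind sway to input Material Attributes through World Position Offset.
# A rotated, time-panned texture, mapped by either UVs or scaled pre-skinned position, supplies a signed displacement from its red channel.
# Static switches choose that mapping and push along vertex normals or a fixed direction, scaled by intensity and its mask.
# A separate MASK blends the result with the original; Low quality bypasses the effect.
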MaterialFunction /Game/Weapons/WeaponSkins/MaterialFunction/WindEffect
graph {nodes [
  n1 = FunctionInput() {bUsePreviewValueAsDefault=true, inputName=Material, inputType=FunctionInput_MaterialAttributes}
  n2 = GetMaterialAttributes(None=@n1) {attributeGetTypes [#1]}
  n3 = PreSkinnedPosition()
  n4 = ScalarParameter() {defaultValue=400, group=Wind, parameterName=WorldAlignedScale}
  n5 = ScalarParameter() {defaultValue=1, group=Wind, parameterName=WindScale}
  n6 = StaticSwitchParameter(True=Multiply(A=Divide(A=@n3, B=@n4), B=@n5), False=Multiply(A=TextureCoordinate(), B=@n5)) {group=Wind, parameterName=World Aligned Wind}
  n7 = ScalarParameter() {defaultValue=0.25, group=Wind, parameterName=Rotate Wind Texture}
  n8 = Multiply(
   A=Time()
   B=AppendVector(
    A=ScalarParameter() {defaultValue=1, group=Wind, parameterName=Wind Speed X}
    B=ScalarParameter() {defaultValue=-0.5, group=Wind, parameterName=Wind Speed Y}
   )
  )
  n9 = TextureSampleParameter2D(UVs=Add(
    A=MaterialFunctionCall(UVs=@n6, "Rotation Angle (0-1)"=@n7) {materialFunction=&CustomRotator}.Rotated Values
    B=Frac(None=@n8)
   )) {
   group=Wind
   parameterName=Wind Texture
   samplerType=SAMPLERTYPE_LinearColor
   texture=/Game/Global/Textures/Default/Masks_and_Gradients/T_VertGradient
  }
  n10 = LinearInterpolate(Alpha=@n9.R) {constA=-1}
  n11 = Multiply(A=@n10, B=VertexNormalWS())
  n12 = Multiply(A=@n10, B=Normalize(VectorInput=VectorParameter() {group=Wind, parameterName=Wind Sway Direction, defaultValue=<rgba 0 0 1 1>}.RGB))
  n13 = Multiply(
   A=Multiply(
    A=StaticSwitchParameter(True=@n11, False=@n12) {group=Wind, parameterName=Use Vertex Normal Direction}
    B=ScalarParameter() {defaultValue=10, group=Wind, parameterName=Wind Intensity}
   )
   B=FunctionInput() {bUsePreviewValueAsDefault=true, inputName=Wind Intensity Mask, inputType=FunctionInput_Scalar}
  )
  n14 = BlendMaterialAttributes(
   A=@n1
   B=SetMaterialAttributes!(MaterialAttributes=@n2.MaterialAttributes, WorldPositionOffset=Add(A=@n2.WorldPositionOffset, B=@n13)) {attributeSetTypes [#1]}
   Alpha=FunctionInput() {bUsePreviewValueAsDefault=true, inputName=MASK, inputType=FunctionInput_Scalar}
  )
  FunctionOutput(None=QualitySwitch(Default=@n14, Low=@n1, High=@n14, Medium=@n14))
  MaterialFunctionCall!(VectorToRotate, LookAtVector, RestingVector) {materialFunction=/Engine/Functions/Engine_MaterialFunctions01/Vectors/RotateVector}
 ]}
