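# Material function that layers optional Skin FX onto incoming material attributes for weapon skins.
# A long serial chain of static switches passes the attributes through effect subfunctions such as thin film, crystal, galaxy, slurp and cloth fuzz.
# Most effects read a shared mask (SkinFX_Mask texture or Customization Mask input), many through selectable channels.
# Optional time-of-day emissive boost and distance-based vertex scaling finish the chain, which feeds both outputs.
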MaterialFunction /Game/Weapons/WeaponSkins/MaterialFunction/Weapon_SkinFXFunction
graph {nodes [
  FunctionInput() {bUsePreviewValueAsDefault=true, inputName=Pre-Skinned World Normal}
  n1 = TextureSampleParameter2D() {group=Skin FX, parameterName=SkinFX_Mask, samplerType=SAMPLERTYPE_Masks, texture=&T_BlackMask}
  n2 = StaticSwitchParameter(True=FunctionInput() {inputName=Customization Mask, inputType=FunctionInput_Vector4}, False=AppendVector(A=@n1.RGB, B=@n1.RGB)) {group=Skin FX, parameterName=useCustomizationMaskToDrive_SkinFX}
  n3 = ChannelMaskParameter(None=@n2) {group=Skin FX, maskChannel=Green, parameterName=Panning Emissive Offset Channel, defaultValue=<rgba 0 1 0 0>}
  n4 = PreSkinnedNormal()
  n5 = PreSkinnedPosition()
  n6 = ChannelMaskParameter(None=@n2) {group=Skin FX, maskChannel=Green, parameterName=CrystalEdgeMaskChannel, defaultValue=<rgba 0 1 0 0>}
  n7 = FunctionInput() {inputName=Material Attributes, inputType=FunctionInput_MaterialAttributes}
  n8 = StaticSwitchParameter(
   True=MaterialFunctionCall!(Alpha=StaticComponentMaskParameter(None=@n2) {defaultR=true, group=Skin FX, parameterName=ThinFilm_Channel}, Material Attributes=@n7) {materialFunction=&/ThinFilmFunction}.Result
   False=@n7
  ) {group=Skin FX, parameterName=Use Thin Film}
  n9 = StaticSwitchParameter(
   True=MaterialFunctionCall!(Material=@n8, Alpha=StaticComponentMaskParameter(None=@n2) {defaultG=true, group=Skin FX, parameterName=DiscoLights_Channel}) {materialFunction=&/DiscoFloorFunction}.Result
   False=@n8
  ) {group=Skin FX, parameterName=Use Disco Lights}
  n10 = StaticSwitchParameter(
   True=MaterialFunctionCall!(Edge Map=@n6.R, Material=@n9, Alpha=ChannelMaskParameter(None=@n2) {group=Skin FX, parameterName=CrystalMaskChannel}.R) {materialFunction=&/Crystal}.Result
   False=@n9
  ) {group=Skin FX, parameterName=UseCrystal}
  n11 = StaticSwitchParameter(
   True=MaterialFunctionCall!(In=@n10) {materialFunction=/Game/Gadgets/Assets/VinderTech_GliderChute/Glider_Holiday_01/Materials/MF_AnimatedSweaterGlider}.Result
   False=@n10
  ) {group=Skin FX, parameterName=Use AnimatedHolidaySweater}
  n12 = StaticSwitchParameter(
   True=MaterialFunctionCall!(GlowingElementMask=StaticComponentMaskParameter(None=@n2) {defaultR=true, group=Skin FX, parameterName=GlowingElementChannel}, Material=@n11) {materialFunction=&/AnimatedElementGLow}.Result
   False=@n11
  ) {group=Skin FX, parameterName=Use AnimatedElementGlow}
  n13 = StaticSwitchParameter(
   True=MaterialFunctionCall!(Mask=@n2, Materials=@n12) {materialFunction=&/MF_ParallaxGlow}.Result
   False=@n12
  ) {parameterName=UseParallaxGlow}
  n14 = StaticSwitchParameter(
   True=MaterialFunctionCall!(In=@n13, Alpha=StaticComponentMaskParameter(None=@n2) {defaultR=true, group=Skin FX, parameterName="Channel Embracing T H E C U B E "}, AO) {materialFunction=&/TheCUBE}.Result
   False=@n13
  ) {group=Skin FX, parameterName=Use Cube Effect}
  n15 = StaticSwitchParameter(
   True=MaterialFunctionCall!(
    MASK=StaticComponentMaskParameter(None=@n2) {defaultR=true, group=Skin FX, parameterName=PrismaticEffect_Channel}
    IsCharacter=NamedRerouteUsage()
    IsStatsB=NamedRerouteUsage()
    Material=@n14
   ) {materialFunction=&/PrismaticEffect}.Result
   False=@n14
  ) {group=Skin FX, parameterName=UsePrismaticEffect}
  n16 = StaticSwitchParameter(
   True=MaterialFunctionCall!(
    MASK=StaticComponentMaskParameter(None=@n2) {defaultR=true, group=Skin FX, parameterName=WindEffect_Channel}
    Wind Intensity Mask=StaticComponentMaskParameter(None=@n2) {defaultR=true, group=Skin FX, parameterName=WindEffect Intensity_Channel}
    Material=@n15
   ) {materialFunction=&/WindEffect}.Result
   False=@n15
  ) {group=Skin FX, parameterName=UseWindEffect}
  n17 = StaticSwitchParameter(
   True=MaterialFunctionCall!(MASK=StaticComponentMaskParameter(None=@n2) {defaultR=true, group=Skin FX, parameterName=Anisotropic_Channel}, In=@n16) {materialFunction=&/Anisotropic}.Result
   False=@n16
  ) {group=Skin FX, parameterName=Use Antisotropic Shading}
  n18 = StaticSwitchParameter(
   True=MaterialFunctionCall!(Alpha=StaticComponentMaskParameter(None=@n2) {defaultR=true, group=Skin FX, parameterName=Hop Rock Gem Channel}, Material=@n17) {materialFunction=&/HopRockGem}.Result
   False=@n17
  ) {group=Skin FX, parameterName=UseHopRockGem}
  n19 = StaticSwitchParameter(
   True=MaterialFunctionCall!(
    Alpha=StaticComponentMaskParameter(None=@n2) {defaultR=true, group=Skin FX, parameterName=Panning Emissive Channel}
    isVehicle
    Gmap=FunctionInput() {bUsePreviewValueAsDefault=true, inputName=Gmap, inputType=FunctionInput_Scalar}
    UVOffset=@n3.R
    Local Normal=VertexInterpolator(VS=@n4).PS
    Local Position=VertexInterpolator(VS=@n5).PS
    EmissiveColorOverride
    Material=@n18
   ) {materialFunction=&/PanningEmissive}.Result
   False=@n18
  ) {group=Skin FX, parameterName=Use Panning Emissive}
  n20 = StaticSwitchParameter(
   True=MaterialFunctionCall!(
    Alpha=StaticComponentMaskParameter(None=@n2) {defaultR=true, group=Skin FX, parameterName=Marshmello Channel}
    Offset=StaticComponentMaskParameter(None=@n2) {defaultG=true, group=Skin FX, parameterName=Marshmello Offset Channel}
    EQ Low=StaticComponentMaskParameter(None=@n2) {defaultR=true, group=Skin FX, parameterName=Marshmello Low Channel}
    EQ High=StaticComponentMaskParameter(None=@n2) {defaultG=true, group=Skin FX, parameterName=Marshmello High Channel}
    Material=@n19
   ) {materialFunction=&/Marshmello}.Result
   False=@n19
  ) {group=Skin FX, parameterName=UseMarshmelloLights}
  n21 = StaticSwitchParameter(
   True=MaterialFunctionCall!(Alpha=StaticComponentMaskParameter(None=@n2) {defaultR=true, group=Skin FX, parameterName=Clearcoat Channel}, In=@n20) {materialFunction=&/Clearcoat}.Result
   False=@n20
  ) {group=Skin FX, parameterName=Use Clear Coat}
  n22 = StaticSwitchParameter(
   True=MaterialFunctionCall!(
    MASK=StaticComponentMaskParameter(None=@n2) {defaultR=true, group=Skin FX, parameterName=Generic Gem Effect Channel}
    Color Change Sequence
    Material=@n21
   ) {materialFunction=&/GenericGem}.Result
   False=@n21
  ) {group=Skin FX, parameterName=Use Generic Gem}
  n23 = StaticSwitchParameter(
   True=MaterialFunctionCall!(Mask=@n2, Material=@n22) {materialFunction=&/MF_GemPattern}.Result
   False=@n22
  ) {group=Skin FX, parameterName=UseGemPattern}
  n24 = StaticSwitchParameter(
   True=MaterialFunctionCall!(Alpha=ChannelMaskParameter(None=@n2) {group=Skin FX, parameterName=Sparkle Remix Channel}.R, Material=@n23) {materialFunction=&/MF_SparkleRemix_Burst}.Result
   False=@n23
  ) {group=Skin FX, parameterName=Use Sparkle Remix Burst}
  n25 = StaticSwitchParameter(
   True=MaterialFunctionCall!(Mask=@n2, Material=@n24) {materialFunction=&/MF_SleepyTime}.Result
   False=@n24
  ) {parameterName=UseSleepyTIme}
  n26 = StaticSwitchParameter(
   True=MaterialFunctionCall!(FX_Mask_Vectoir=@n2, MA_Input=@n25, FX_MaskInput) {materialFunction=&/MF_BlinkLight}.MA_Output
   False=@n25
  ) {parameterName=UseBlinkingLight}
  n27 = StaticSwitchParameter(
   True=MaterialFunctionCall!(Alpha=StaticComponentMaskParameter(None=@n2) {defaultR=true, group=Skin FX, parameterName=Automatic Headlights}, In=@n26) {materialFunction=&/NightTime_AutomaticHeadlights}.Result
   False=@n26
  ) {group=Skin FX, parameterName=UseAutomaticHeadlights}
  n28 = StaticSwitchParameter(
   True=MaterialFunctionCall!(In=@n27, CenterLine, Alpha=StaticComponentMaskParameter(None=@n2) {defaultA=true, group=Skin FX, parameterName=Taillight Channel}) {materialFunction=&/TurnSignals}.Result
   False=@n27
  ) {group=Skin FX, parameterName=Use TurnSignals}
  n29 = StaticSwitchParameter(
   True=MaterialFunctionCall!(
    In=@n28
    Warp Mask
    GalaxyAlternateFresnel
    Alpha=StaticComponentMaskParameter(None=@n2) {defaultA=true, group=Skin FX, parameterName=Galaxy Channel}
   ) {materialFunction=&/Galaxy}.Result
   False=@n28
  ) {group=Skin FX, parameterName=Use Galaxy}
  n30 = StaticSwitchParameter(
   True=MaterialFunctionCall!(
    Material=@n29
    EdgeMap=StaticComponentMaskParameter(None=@n2) {defaultA=true, group=Skin FX, parameterName=Ice Edge Map Channel}
    Alpha=StaticComponentMaskParameter(None=@n2) {defaultA=true, group=Skin FX, parameterName=Ice Channel}
   ) {materialFunction=&/Ice}.Result
   False=@n29
  ) {group=Skin FX, parameterName=Use Ice Effect}
  n31 = StaticSwitchParameter(
   True=MaterialFunctionCall!(Material=@n30, In=StaticComponentMaskParameter(None=@n2) {defaultR=true, group=Skin FX, parameterName=Alpha Mask Channel}) {materialFunction=&/AlphaMask}.Result
   False=@n30
  ) {group=Skin FX, parameterName=use Opacity Mask}
  n32 = StaticSwitchParameter(
   True=MaterialFunctionCall!(Material=@n31, Mask for Fade=VertexColor().G) {materialFunction=/Game/Packages/Fortress_SharedMaterials/Base_Material/MF_Fade_Along_UV}.Fade in OpacityMask
   False=@n31
  ) {group=Skin FX, parameterName=Use Fade Along U in OpacityMask}
  n33 = StaticSwitchParameter(
   True=MaterialFunctionCall!(
    Alpha=StaticComponentMaskParameter(None=@n2) {defaultR=true, group=Skin FX, parameterName=Cloth Channel}
    Gmap=Constant() {r=1}
    Material=@n32
   ) {materialFunction=&/ClothFuzz}.Result
   False=@n32
  ) {group=Skin FX, parameterName=Use Cloth fuzz}
  n34 = StaticSwitchParameter(
   True=MaterialFunctionCall!(In=@n33, Dual SS Secondary Alpha, Alpha=StaticComponentMaskParameter(None=@n2) {defaultG=true, group=Skin FX, parameterName=SS Masks}) {materialFunction=&/SS_Shader_Model}.Result
   False=@n33
  ) {group=Skin FX, parameterName="Uses Subsurface Shading Model (NOT Subsurface Profile)"}
  n35 = StaticSwitchParameter(
   True=MaterialFunctionCall!(Mask=ChannelMaskParameter(None=@n2) {group=Skin FX, parameterName=PanAndDistortChannel}.R, Material=@n34) {materialFunction=&/MF_PanAndDistort}.Result
   False=@n34
  ) {group=Skin FX, parameterName=UsePanAndDistort}
  n36 = StaticSwitchParameter(
   True=MaterialFunctionCall!(Mask=@n2, Material=@n35) {materialFunction=&/MF_DistortionFresnel}.Result
   False=@n35
  ) {parameterName=UseDistortionFresnel}
  n37 = StaticSwitchParameter(
   True=MaterialFunctionCall!(Gmap, Materials=@n36) {materialFunction=&/PlayerBanner}.Result
   False=@n36
  ) {group=Skin FX, parameterName=UsePlayerBanner}
  n38 = StaticSwitchParameter(
   True=MaterialFunctionCall!(TransitionGradient=ChannelMaskParameter(None=@n2) {group=Skin FX, parameterName=SprayPaint Greadient Channel}.R, In=@n37) {materialFunction=&/MF_SprayPaintFunction}.Result
   False=@n37
  ) {group=Skin FX, parameterName=UseSprayPaint}
  n39 = StaticSwitchParameter(
   True=MaterialFunctionCall!(PreSkinnedPosition, Mask=@n2, Material=@n38) {materialFunction=&/MF_BlackMondayCapeMovement}.Result
   False=@n38
  ) {group=Skin FX, parameterName=UseBlackMondayCapeMovement}
  n40 = StaticSwitchParameter(
   True=MaterialFunctionCall!(Material=@n39, Mask=@n2) {materialFunction=&/MF_ViperGlitchWeapon}.Result
   False=@n39
  ) {parameterName=UseViperGlitch}
  n41 = StaticSwitchParameter(
   True=MaterialFunctionCall!(
    Alpha=StaticComponentMaskParameter(None=@n2) {defaultR=true, group=Skin FX, parameterName=Slurp Channel}
    AO=Constant() {r=1}
    BubblesMask=StaticComponentMaskParameter(None=@n2) {defaultG=true, group=Skin FX, parameterName=Slurp Bubbles Channel}
    SlurpNormalDistortionMask=StaticComponentMaskParameter(None=@n2) {defaultB=true, group=Skin FX, parameterName=Slurp Normal Distortion Channel}
    Pre Skinned Position
    Material=@n40
   ) {materialFunction=&/MF_Slurp}.Result
   False=@n40
  ) {group=Skin FX, parameterName=UseSlurp}
  n42 = StaticSwitchParameter(
   True=MaterialFunctionCall!(Alpha=StaticComponentMaskParameter(None=@n2) {defaultR=true, group=Skin FX, parameterName=RandomTileAnimatedPattern Channel}, Material=@n41) {materialFunction=&/MF_RandomTileAnimatedPattern}.Result
   False=@n41
  ) {group=Skin FX, parameterName=UseRandomTileAnimatedPattern}
  n43 = GetMaterialAttributes(None=@n42) {attributeGetTypes [#1]}
  n44 = StaticSwitchParameter(
   True=SetMaterialAttributes!(
    MaterialAttributes=@n43.MaterialAttributes
    EmissiveColor=MaterialFunctionCall!(In=@n43.EmissiveColor) {materialFunction=&/BoostEmissivebyTOD}.Result
   ) {attributeSetTypes [#1]}
   False=@n42
  ) {group=Skin FX, parameterName=Boost Emissive by TOD}
  n45 = StaticSwitchParameter(
   True=MaterialFunctionCall!(MA_In=@n44) {materialFunction=&/MF_ScaleVertsByDistance}.Result
   False=@n44
  ) {group=Skin FX, parameterName=ScaleVertsByDistance}
  FunctionOutput(None=@n45)
  FunctionOutput(None=@n45) {outputName=Wrap_OutPut}
  NamedRerouteDeclaration(None=FunctionInput() {bUsePreviewValueAsDefault=true, inputName=IsCharacter, inputType=FunctionInput_Scalar}) {name=IsCharacter, nodeColor=<rgba 0 0.9253 1 1>}
  n46 = FunctionInput() {bUsePreviewValueAsDefault=true, inputName=IsStatsB, inputType=FunctionInput_Vector2}
  NamedRerouteDeclaration(None=ComponentMask(None=@n46) {r=true}) {name=IsBackpack, nodeColor=<rgba 0 0.9253 1 1>}
  NamedRerouteDeclaration(None=ComponentMask(None=@n46) {g=true}) {name=IsPickaxe, nodeColor=<rgba 0 0.9253 1 1>}
  NamedRerouteDeclaration(None=@n46) {name=IsStatsB, nodeColor=<rgba 0 0.9253 1 1>}
 ]}
